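# Niagara system for the GlowFang emote despawn effect, a short, delayed burst of glow, sparks and lightning around the staff prop.
# Flash001 adds a glow sprite and light, while GPU particulate bursts sample the staff's skeletal mesh or a tall box with curl noise.
# CPU particulates and velocity-aligned lightning meshes add detail.
# System-level ElimColor tints most emitters, SMTOD scales flash brightness, and Ice_Mesh and StaffChargeup001 are disabled.
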NiagaraSystem /BRCosmetics/Animation/Game/MainPlayer/Emotes/GlowFang/VFX/NS_Emote_GlowFang_Despawn
dependencies {
 usedDataInterfaces [
  &NiagaraDataInterfaceSkeletalMesh
  &NiagaraDataInterfaceCurve
  &NiagaraDataInterfaceVector2DCurve
  &NiagaraDataInterfaceArrayFloat
 ]
 usedDynamicInputs [
  &UniformRangedVector
  &SimulationPosition
  &FNTimeOfDayScalar
  &InvertBool
 ]
 usedModules [
  &SystemState
  &self:SystemSpawnScript.SystemScriptSource.SystemScriptGraph.NiagaraNodeAssignment_0.SetVariables_#1
  &self:SystemSpawnScript.SystemScriptSource.SystemScriptGraph.NiagaraNodeAssignment_1.SetVariables_#2
  &self:SystemSpawnScript.SystemScriptSource.SystemScriptGraph.NiagaraNodeAssignment_2.SetVariables_#3
  &self:SystemSpawnScript.SystemScriptSource.SystemScriptGraph.NiagaraNodeAssignment_3.SetVariables_#4
  &self:SystemSpawnScript.SystemScriptSource.SystemScriptGraph.NiagaraNodeAssignment_4.SetVariables_#5
 ]
}
emitters {
 Flash001 {
  emitterUpdateScript [
   &EmitterState(
    Life Cycle Mode=::NewEnumerator1
    Loop Behavior=::NewEnumerator1
    Loop Delay=0.2
    Loop Duration=$System.CurrentLoopDuration
    UseLoopDelay=true
   )
   &SpawnBurst_Instantaneous(Spawn Count=1, Use Spawn Probability=false)
  ]
  particleSpawnScript [
   &InitializeParticle(
    Color=$System.ElimColor
    Lifetime=0.6
    Mass Mode=::NewEnumerator1
    Mesh Scale Mode=::NewEnumerator3
    Position=<0 0 15>
    Position Mode=::NewEnumerator1
    Sprite Size=<200 175>
    Sprite Size Mode=::NewEnumerator4
    UsePositionOffset=false
    Write Material Random=true
   ) {module=InitializeParticle001}
   &CameraOffset(Camera Offset Amount=10)
  ]
  particleUpdateScript [
   &SolveForcesAndVelocity(Clamp Velocity=false, Limit Acceleration=false)
   &ParticleState()
   &ScaleSpriteSize(
    Scale Sprite Size Mode=::NewEnumerator0
    Uniform Scale Factor=&FloatFromCurve(FloatCurve=DataInterface() {propertyValues {Curve {keys [
         {arriveTangent=9.0725, interpMode=RCIM_Cubic, leaveTangent=9.0725, tangentMode=RCTM_User, time=0, value=0}
         {arriveTangent=0.1968, interpMode=RCIM_Cubic, leaveTangent=0.1968, tangentMode=RCTM_User, time=0.1408, value=0.9847}
         {arriveTangent=-3.1577, interpMode=RCIM_Cubic, leaveTangent=-3.1577, tangentMode=RCTM_User, time=1, value=0}
        ]}}})
   )
   &FN_Sprite_Attributes(
    Create SpriteSize Attribute=true
    SpriteSize=&Multiply_Vector2D(
     A=<40 40>
     B=&Vector2DFromFloat(Value=&FloatFromCurve(FloatCurve=DataInterface() {propertyValues {Curve {keys [<at 0 0 RCIM_Cubic>,<at 0.5 1 RCIM_Cubic>]}}}))
    )
    Write to Module Namespace=true
   ) {module=FN_Sprite}
   /SetVariables(Particles.SpriteColor=&MultiplyLinearColor(
     Linear Color=$Particles.Color
     Scale Factor=&MakeVector4(
      X=$System.SMTOD
      Y=$System.SMTOD
      Z=$System.SMTOD
      W=&MakeCustomFloatFromBool(Bool=$User.bIsFrontEndPreview, True Float=0.06, False Float=0.13)
     )
    )) {isSetParameters=true}
   &DynamicMaterialParameters(Write Parameter Index 0=false)
   &FN_Light_Attributes(
    Color=<rgba 0 0.5942 1 0.2>
    Create Color Attribute=true
    Create Exponent Attribute=true
    Create Position Attribute=true
    Create Radius Attribute=true
    Enable Particle Light Enable=true
    Exponent=3
    Light Enabled=true
    Position:NiagaraPosition=$Engine.Emitter.SimulationPosition
    Radius=55
    Scale Color=&FloatFromCurve(
     FloatCurve=DataInterface() {propertyValues {Curve {keys [{interpMode=RCIM_Cubic, tangentMode=RCTM_Break, time=0, value=1}, {arriveTangent=-1.1726, tangentMode=RCTM_User, time=1, value=0}]}}}
     Scale Curve=10
    )
   )
   &ScaleColor(Scale RGB=&VectorFromFloat(Value=$System.SMTOD), ScaleA=true, ScaleRGB=true) {module=ScaleColor002}
   &ScaleColor(
    Scale Alpha=&FloatFromCurve(
     FloatCurve=DataInterface() {propertyValues {Curve {keys [
         {arriveTangent=9.0725, interpMode=RCIM_Cubic, leaveTangent=9.0725, tangentMode=RCTM_User, time=0, value=0}
         {arriveTangent=-2.5249, interpMode=RCIM_Cubic, leaveTangent=-2.5249, tangentMode=RCTM_User, time=0.1408, value=0.9847}
         {arriveTangent=-0.0405, interpMode=RCIM_Cubic, leaveTangent=-0.0405, tangentMode=RCTM_User, time=1, value=0}
        ]}}}
     Scale Curve=0.3
    )
    Scale RGB=&VectorFromFloat(Value=&FloatFromCurve(FloatCurve=DataInterface() {propertyValues {Curve {keys [<at 0 1>,<at 0.1 1.5>,<at 0.2 1>,<at 0.6 1>,<at 0.7 1.5>,<at 0.8 1>,<at 1 1>]}}}))
    ScaleA=true
    ScaleRGB=true
   )
  ]
  rendererClasses [&NiagaraLightRendererProperties, &NiagaraSpriteRendererProperties]
  renderers [
   {rendererClass=&NiagaraLightRendererProperties, rendererIndex=0}
   {rendererClass=&NiagaraSpriteRendererProperties, rendererIndex=1}
  ]
 }
 HangingParticulates {
  simTarget=GPUComputeSim
  emitterUpdateScript [
   &EmitterState(Life Cycle Mode=::NewEnumerator1, Loop Behavior=::NewEnumerator1, Loop Delay=0.5, Loop Duration=0.8, UseLoopDelay=true)
   &SpawnBurst_Instantaneous(Spawn Count=50, Use Spawn Probability=false, UseLoopCountLimit=false)
  ]
  particleSpawnScript [
   &InitializeParticle(
    Color=&Multiply_LinearColorByFloat(Linear Color=$System.ElimColor, Float=5)
    Lifetime Max=0.2
    Lifetime Min=0.15
    Lifetime Mode=::NewEnumerator1
    Sprite Rotation Mode=::NewEnumerator2
    Sprite Size Mode=::NewEnumerator1
    Sprite UV Mode=::NewEnumerator3
    Uniform Sprite Size Max=5
    Uniform Sprite Size Min=1
    UsePositionOffset=false
    Write Material Random=true
   )
   &SkeletalMeshLocation(
    MaskSpawnGroup=false
    Mesh Sampling Type=::NewEnumerator7
    OffsetAlongNormal=false
    OverrideRandomSeed=false
    Position Sampling=::NewEnumerator1
    Sampled Position Offset=&RandomRangeVector(Minimum=<0 0 0>, Maximum=<3 3 5>)
    Skeletal Mesh=DataInterface() {propertyValues {
      DefaultMesh=null
      ExcludeBoneName=root
      OptionalComponentClass=null
      PreviewMesh=/BRCosmetics/Animation/Props/Gadgets_GlowFang_Staff/Meshes/GlowFang_Staff_Gadget
      SkinningMode=SkinOnTheFly
      SoftSourceActor=null
      SourceMode=DefaultMeshOnly
      UvSetIndex=0
      WholeMeshLOD=-1
      bExcludeBone=false
      bReadDeformedGeometry=true
      bRequireCurrentFrameData=true
      ComponentTags []
      FilteredBones []
      FilteredSockets []
      MeshUserParameter {parameter {name=None, typeDefHandle {registeredTypeIndex=1}, varData []}}
      SamplingRegions []
     }}
   )
  ]
  particleUpdateScript [
   &ParticleState()
   &ScaleSpriteSize(Scale Factor=&Vector2DFromFloat(Value=&FloatFromCurve(FloatCurve=DataInterface() {propertyValues {Curve {keys [
          <at 0 0 RCIM_Cubic>
          <at 0.112 1.0173>
          <at 0.2 0.6>
          {arriveTangent=-6.9697, interpMode=RCIM_Cubic, leaveTangent=-6.9697, tangentMode=RCTM_User, time=0.3, value=2}
          {arriveTangent=-2.4779, interpMode=RCIM_Cubic, leaveTangent=-2.4779, tangentMode=RCTM_User, time=0.5, value=0.375}
          {arriveTangent=0.3304, tangentMode=RCTM_User, time=1.0011, value=0.002}
         ]}}})))
   &ScaleColor(
    Scale Alpha=&FloatFromCurve(FloatCurve=DataInterface() {propertyValues {Curve {keys [
         {arriveTangent=9.0725, interpMode=RCIM_Cubic, leaveTangent=9.0725, tangentMode=RCTM_User, time=0, value=0}
         {arriveTangent=-2.5249, interpMode=RCIM_Cubic, leaveTangent=-2.5249, tangentMode=RCTM_User, time=0.1408, value=0.9847}
         {arriveTangent=-0.5654, interpMode=RCIM_Cubic, leaveTangent=-0.5654, time=0.5, value=0.35}
         {arriveTangent=-1.25, interpMode=RCIM_Cubic, leaveTangent=-1.25, time=0.6, value=0.725}
         {arriveTangent=-3.375, interpMode=RCIM_Cubic, leaveTangent=-0.1412, tangentMode=RCTM_Break, time=0.7, value=0.1}
         {arriveTangent=-0.7119, interpMode=RCIM_Cubic, leaveTangent=1.625, tangentMode=RCTM_Break, time=0.8, value=0.05}
         {arriveTangent=-0.25, interpMode=RCIM_Cubic, leaveTangent=-0.25, time=0.9, value=0.425}
         {arriveTangent=-0.0405, interpMode=RCIM_Cubic, leaveTangent=-0.0405, tangentMode=RCTM_User, time=1, value=0}
        ]}}})
    Scale RGB=&VectorFromFloat(Value=&FloatFromCurve(FloatCurve=DataInterface() {propertyValues {Curve {keys [<at 0 1>,<at 0.1 5>,<at 0.2 1>,<at 0.7 1>,<at 0.8 5>,<at 0.9 1>,<at 1 1>]}}}))
    ScaleA=true
    ScaleRGB=true
   )
   &CurlNoiseForce(
    Bias Noise=false
    Noise Frequency=500
    Noise Strength=&FloatFromCurve(
     FloatCurve=DataInterface() {propertyValues {Curve {keys [
         {arriveTangent=4.1785, interpMode=RCIM_Cubic, leaveTangent=4.1785, tangentMode=RCTM_User, time=0, value=0}
         {arriveTangent=1.1179, interpMode=RCIM_Cubic, leaveTangent=1.1179, tangentMode=RCTM_User, time=0.2042, value=0.6228}
         {arriveTangent=0.0224, interpMode=RCIM_Cubic, leaveTangent=0.0224, tangentMode=RCTM_User, time=1, value=1}
        ]}}}
     Scale Curve=500
    )
    Pan Noise=true
    Pan Noise Field=<0.5 0 1>
   )
   &Drag(
    Drag=&FloatFromCurve(
     FloatCurve=DataInterface() {propertyValues {Curve {keys [
         {arriveTangent=-6.9709, interpMode=RCIM_Cubic, leaveTangent=-6.9709, tangentMode=RCTM_User, time=0, value=1}
         {arriveTangent=0.0301, interpMode=RCIM_Cubic, leaveTangent=-0.1449, tangentMode=RCTM_Break, time=0.3, value=0}
         {interpMode=RCIM_Cubic, tangentMode=RCTM_User, time=1, value=0}
        ]}}}
     Scale Curve=&RandomRangeFloat(Minimum=1, Maximum=15)
    )
    Use Linear Drag=true
    Use Rotational Drag=false
   )
   &InheritVelocity(
    Inherited Velocity Amount Scale=<0.01 0.01 0.01>
    Inherited Velocity Speed Limit:NiagaraFloat=&FloatFromCurve(
     FloatCurve=DataInterface() {propertyValues {Curve {keys [
         {interpMode=RCIM_Cubic, tangentMode=RCTM_User, time=0.0017, value=1}
         {interpMode=RCIM_Cubic, tangentMode=RCTM_User, time=0.5, value=1}
         {tangentMode=RCTM_User, time=1, value=0}
        ]}}}
     Scale Curve=100
    )
    Source Speed Threshold=15
    Speed Limit=true
    UseSpeedThreshold=false
   )
   &SolveForcesAndVelocity(Clamp Velocity=false, Limit Acceleration=false)
   &ScaleSpriteSizeBySpeed(
    Max Scale Factor=<0.75 5>
    Sample Scale Factor By Curve=true
    Scale Factor Curve=DataInterface() {propertyValues {Curve {keys [{interpMode=RCIM_Cubic, tangentMode=RCTM_User, time=0, value=0}, {interpMode=RCIM_Cubic, tangentMode=RCTM_User, time=1, value=1}]}}}
   )
  ]
  rendererClasses [&NiagaraSpriteRendererProperties]
  renderers [{rendererClass=&NiagaraSpriteRendererProperties}]
 }
 HangingParticulates001 {
  simTarget=GPUComputeSim
  emitterUpdateScript [
   &EmitterState(Life Cycle Mode=::NewEnumerator1, Loop Behavior=::NewEnumerator1, Loop Delay=0.45, Loop Duration=0.8, UseLoopDelay=true)
   &SpawnBurst_Instantaneous(Spawn Count=50, Use Spawn Probability=false, UseLoopCountLimit=false)
  ]
  particleSpawnScript [
   &InitializeParticle(
    Color=&Multiply_LinearColorByFloat(Linear Color=$System.ElimColor, Float=5)
    Lifetime Max=0.2
    Lifetime Min=0.1
    Lifetime Mode=::NewEnumerator1
    Sprite Rotation Mode=::NewEnumerator2
    Sprite Size Mode=::NewEnumerator1
    Sprite UV Mode=::NewEnumerator3
    Uniform Sprite Size Max=5
    Uniform Sprite Size Min=1
    UsePositionOffset=false
    Write Material Random=true
   )
   &ShapeLocation(
    Box Midpoint=&RandomRangeVector(Minimum=<0 0 0>, Maximum=<1 1 1>)
    Box Size=<3 3 75>
    MaskSpawnGroup=false
    Override Seed=false
    Shape Primitive=::NewEnumerator2
    Surface Only=false
    UseOwnerScale=true
   )
  ]
  particleUpdateScript [
   &ParticleState()
   &ScaleSpriteSize(Scale Factor=&Vector2DFromFloat(Value=&FloatFromCurve(FloatCurve=DataInterface() {propertyValues {Curve {keys [
          <at 0 0 RCIM_Cubic>
          <at 0.112 1.0173>
          <at 0.2 0.6>
          {arriveTangent=-6.9697, interpMode=RCIM_Cubic, leaveTangent=-6.9697, tangentMode=RCTM_User, time=0.3, value=2}
          {arriveTangent=-2.4779, interpMode=RCIM_Cubic, leaveTangent=-2.4779, tangentMode=RCTM_User, time=0.5, value=0.375}
          {arriveTangent=0.3304, tangentMode=RCTM_User, time=1.0011, value=0.002}
         ]}}})))
   &ScaleColor(
    Scale Alpha=&FloatFromCurve(FloatCurve=DataInterface() {propertyValues {Curve {keys [
         {arriveTangent=9.0725, interpMode=RCIM_Cubic, leaveTangent=9.0725, tangentMode=RCTM_User, time=0, value=0}
         {arriveTangent=-2.5249, interpMode=RCIM_Cubic, leaveTangent=-2.5249, tangentMode=RCTM_User, time=0.1408, value=0.9847}
         {arriveTangent=-0.5654, interpMode=RCIM_Cubic, leaveTangent=-0.5654, time=0.5, value=0.35}
         {arriveTangent=-1.25, interpMode=RCIM_Cubic, leaveTangent=-1.25, time=0.6, value=0.725}
         {arriveTangent=-3.375, interpMode=RCIM_Cubic, leaveTangent=-0.1412, tangentMode=RCTM_Break, time=0.7, value=0.1}
         {arriveTangent=-0.7119, interpMode=RCIM_Cubic, leaveTangent=1.625, tangentMode=RCTM_Break, time=0.8, value=0.05}
         {arriveTangent=-0.25, interpMode=RCIM_Cubic, leaveTangent=-0.25, time=0.9, value=0.425}
         {arriveTangent=-0.0405, interpMode=RCIM_Cubic, leaveTangent=-0.0405, tangentMode=RCTM_User, time=1, value=0}
        ]}}})
    Scale RGB=&VectorFromFloat(Value=&FloatFromCurve(FloatCurve=DataInterface() {propertyValues {Curve {keys [<at 0 1>,<at 0.1 5>,<at 0.2 1>,<at 0.7 1>,<at 0.8 5>,<at 0.9 1>,<at 1 1>]}}}))
    ScaleA=true
    ScaleRGB=true
   )
   &CurlNoiseForce(
    Bias Noise=false
    Noise Frequency=500
    Noise Strength=&FloatFromCurve(
     FloatCurve=DataInterface() {propertyValues {Curve {keys [
         {arriveTangent=4.1785, interpMode=RCIM_Cubic, leaveTangent=4.1785, tangentMode=RCTM_User, time=0, value=0}
         {arriveTangent=1.1179, interpMode=RCIM_Cubic, leaveTangent=1.1179, tangentMode=RCTM_User, time=0.2042, value=0.6228}
         {arriveTangent=0.0224, interpMode=RCIM_Cubic, leaveTangent=0.0224, tangentMode=RCTM_User, time=1, value=1}
        ]}}}
     Scale Curve=500
    )
    Pan Noise=false
   )
   &Drag(
    Drag=&FloatFromCurve(
     FloatCurve=DataInterface() {propertyValues {Curve {keys [
         {arriveTangent=-6.9709, interpMode=RCIM_Cubic, leaveTangent=-6.9709, tangentMode=RCTM_User, time=0, value=1}
         {arriveTangent=0.0301, interpMode=RCIM_Cubic, leaveTangent=-0.1449, tangentMode=RCTM_Break, time=0.3, value=0}
         {interpMode=RCIM_Cubic, tangentMode=RCTM_User, time=1, value=0}
        ]}}}
     Scale Curve=&RandomRangeFloat(Minimum=1, Maximum=15)
    )
    Use Linear Drag=true
    Use Rotational Drag=false
   )
   &InheritVelocity(
    Inherited Velocity Speed Limit:NiagaraFloat=&FloatFromCurve(
     FloatCurve=DataInterface() {propertyValues {Curve {keys [
         {interpMode=RCIM_Cubic, tangentMode=RCTM_User, time=0, value=1}
         {interpMode=RCIM_Cubic, tangentMode=RCTM_User, time=0.5, value=1}
         {tangentMode=RCTM_User, time=1, value=0}
        ]}}}
     Scale Curve=100
    )
    Source Speed Threshold=15
    Speed Limit=true
    UseSpeedThreshold=false
   ) {enabled=false}
   &GravityForce(Gravity=<0 0 50>)
   &SolveForcesAndVelocity(Clamp Velocity=false, Limit Acceleration=false)
   &ScaleSpriteSizeBySpeed(
    Max Scale Factor=<0.75 5>
    Sample Scale Factor By Curve=true
    Scale Factor Curve=DataInterface() {propertyValues {Curve {keys [{interpMode=RCIM_Cubic, tangentMode=RCTM_User, time=0, value=0}, {interpMode=RCIM_Cubic, tangentMode=RCTM_User, time=1, value=1}]}}}
   )
  ]
  rendererClasses [&NiagaraSpriteRendererProperties]
  renderers [{rendererClass=&NiagaraSpriteRendererProperties}]
 }
 HangingParticulates002 {
  emitterUpdateScript [
   &EmitterState(Life Cycle Mode=::NewEnumerator1, Loop Behavior=::NewEnumerator1, Loop Delay=0.2, Loop Duration=0.3, UseLoopDelay=true)
   &SpawnBurst_Instantaneous(Spawn Count=10, Use Spawn Probability=false, UseLoopCountLimit=false)
  ]
  particleSpawnScript [
   /SetVariables(Particles.VisibilityTag=&SpawnGroup()) {isSetParameters=true}
   &InitializeParticle(
    Color=&MaskLinearColorBySpawnGroup(
     Masked Linear Color=&Multiply_LinearColorByFloat(Linear Color=$System.ElimColor, Float=1.5)
     Passthrough Linear Color=<rgba 0.1938 0.9043 1 1>
    )
    Lifetime=&MaskFloatBySpawnGroup(Masked Float=&MaskFloatBySpawnGroup(Masked Float=0.4, Passthrough Float=0.7), Passthrough Float=0.6)
    Mesh Scale=<1.05 1.05 1.001>
    Mesh Scale Mode=::NewEnumerator4
    Position Offset=<0 0 7>
    Sprite Size Max=<40 20>
    Sprite Size Min=<30 10>
    Sprite Size Mode=::NewEnumerator2
    UsePositionOffset=true
    Write Material Random=true
   )
   &SystemLocation(MaskSpawnGroup=false, Offset=&MaskVectorBySpawnGroup(Masked Vector=<0 0 0>, Passthrough Vector=<-9.3 0 0>))
   &ShapeLocation(
    Box Midpoint=&MaskVectorBySpawnGroup(Masked Vector=&RandomRangeVector(Minimum=<0 0 0>, Maximum=<1 1 1>), Passthrough Vector=<0.5 0.5 0.5>)
    Box Size=&MaskVectorBySpawnGroup(Masked Vector=&Multiply_VectorByFloat(Vector=<5 5 30>, Float=1), Passthrough Vector=<0 0 0>)
    MaskSpawnGroup=false
    Override Seed=false
    Shape Primitive=::NewEnumerator2
    Surface Only=false
    UseOwnerScale=true
   )
   &AddVelocity(
    Module.Use Falloff Distance=false
    Override Seed=false
    Velocity Mode=::NewEnumerator1
    Velocity Speed=&MaskFloatBySpawnGroup(Masked Float=&RandomRangeFloat(Minimum=10, Maximum=25), Passthrough Float=0)
   )
   &InitialMeshOrientation(Mesh Orientation Mode=::NewEnumerator7, Rotation=<0 0 -0.25>, Use Rotation Vector=true)
   &DynamicMaterialParameters(
    Index 0 Param 1=&RandomRangeFloat(Minimum=0, Maximum=0.2)
    Module.Index0Float0Write=true
    Module.Index0Float1Write=true
    Module.Index0Float2Write=true
    Module.Index0Float3Write=true
   )
   &CameraOffset(Camera Offset Amount=15)
  ]
  particleUpdateScript [
   &ParticleState()
   &ScaleSpriteSize(Scale Factor=&Vector2DFromFloat(Value=&FloatFromCurve(FloatCurve=DataInterface() {propertyValues {Curve {keys [
          <at 0 0 RCIM_Cubic>
          <at 0.112 1.0173>
          <at 0.2 0.6>
          <at 0.3 0.9>
          {arriveTangent=-1.2808, interpMode=RCIM_Cubic, leaveTangent=-1.2808, time=0.7, value=0.65}
          <at 1.0011 0.002>
         ]}}})))
   &CurlNoiseForce(
    Bias Noise=false
    Noise Frequency=&MaskFloatBySpawnGroup(Masked Float=50, Passthrough Float=0)
    Noise Strength=&FloatFromCurve(
     FloatCurve=DataInterface() {propertyValues {Curve {keys [{interpMode=RCIM_Cubic, tangentMode=RCTM_User, time=0, value=0}, {interpMode=RCIM_Cubic, tangentMode=RCTM_User, time=1, value=1}]}}}
     Scale Curve=&MaskFloatBySpawnGroup(Masked Float=500, Passthrough Float=0)
    )
    Pan Noise=false
   )
   &SpriteRotationRate(Rotation Rate=&RandomRangeFloat(Minimum=-70, Maximum=70))
   &ScaleColor(
    Scale Alpha=&FloatFromCurve(FloatCurve=DataInterface() {propertyValues {Curve {keys [
         {interpMode=RCIM_Cubic, tangentMode=RCTM_User, time=0, value=0}
         {interpMode=RCIM_Cubic, tangentMode=RCTM_User, time=0.25, value=0.7}
         {tangentMode=RCTM_User, time=1, value=0}
        ]}}})
    ScaleA=true
    ScaleRGB=true
   )
   &Drag(
    Drag=&FloatFromCurve(
     FloatCurve=DataInterface() {propertyValues {Curve {keys [{interpMode=RCIM_Cubic, tangentMode=RCTM_User, time=0, value=0}, {interpMode=RCIM_Cubic, tangentMode=RCTM_User, time=1, value=1}]}}}
     Scale Curve=&RandomRangeFloat(Minimum=5, Maximum=15)
    )
    Use Linear Drag=true
    Use Rotational Drag=false
   )
   &SolveForcesAndVelocity(Clamp Velocity=false, Limit Acceleration=false)
   &ScaleSpriteSizeBySpeed(
    Max Scale Factor=&MaskVector2dBySpawnGroup(Masked Vector2D=<0.75 5>, Passthrough Vector2D=<1 1>, Spawn Group Mask=0)
    Sample Scale Factor By Curve=true
    Scale Factor Curve=DataInterface() {propertyValues {Curve {keys [{interpMode=RCIM_Cubic, tangentMode=RCTM_User, time=0, value=0}, {interpMode=RCIM_Cubic, tangentMode=RCTM_User, time=1, value=1}]}}}
   )
  ]
  rendererClasses [&NiagaraSpriteRendererProperties]
  renderers [{rendererClass=&NiagaraSpriteRendererProperties}]
 }
 Ice_Mesh {
  bEnabled=false
  emitterUpdateScript [
   &EmitterState()
   &SpawnBurst_Instantaneous(Spawn Count=1, Use Spawn Probability=false)
  ]
  particleSpawnScript [
   &InitializeParticle(Color=<rgba 1 0.1938 0.1938 1>, Lifetime=0.2, Mesh Scale Mode=::NewEnumerator4, UsePositionOffset=false, Write Material Random=true) {module=InitializeParticle001}
   &SystemLocation(MaskSpawnGroup=false, Offset=<8 -6 0>)
   &InitialMeshOrientation(Mesh Orientation Mode=::NewEnumerator7, Rotation=<0 0 -0.25>, Use Rotation Vector=true)
  ]
  particleUpdateScript [
   &ParticleState()
   &SolveForcesAndVelocity(Clamp Velocity=false, Limit Acceleration=false)
   &ScaleColor(
    Scale Alpha=&FloatFromCurve(FloatCurve=DataInterface() {propertyValues {Curve {keys [
         {interpMode=RCIM_Cubic, tangentMode=RCTM_User, time=0, value=0}
         {arriveTangent=70.6593, interpMode=RCIM_Cubic, leaveTangent=-48.9815, tangentMode=RCTM_Break, time=0.2, value=15}
         {tangentMode=RCTM_User, time=1, value=0}
        ]}}})
    Scale RGB=&VectorFromFloat(Value=&FloatFromCurve(FloatCurve=DataInterface() {propertyValues {Curve {keys [
          {interpMode=RCIM_Cubic, tangentMode=RCTM_User, time=0, value=0}
          {arriveTangent=45.4089, interpMode=RCIM_Cubic, leaveTangent=-22.7897, tangentMode=RCTM_Break, time=0.2, value=5}
          {tangentMode=RCTM_User, time=1, value=0}
         ]}}}))
    ScaleA=true
    ScaleRGB=true
   )
   &DynamicMaterialParameters(
    Index 0 Param 1=&FloatFromCurve(FloatCurve=DataInterface() {propertyValues {Curve {keys [
         {interpMode=RCIM_Cubic, leaveTangent=-79.1318, tangentMode=RCTM_Break, time=0, value=1}
         {arriveTangent=-2.6873, tangentMode=RCTM_User, time=0.5, value=-10}
         {arriveTangent=14.6846, interpMode=RCIM_Cubic, leaveTangent=14.6846, tangentMode=RCTM_User, time=0.9, value=-10}
        ]}}})
    Module.Param0WriteEnabled=true
    Module.Param1WriteEnabled=false
    Module.Param2WriteEnabled=false
    Module.Param3WriteEnabled=false
   )
  ]
  rendererClasses [&NiagaraMeshRendererProperties]
  renderers [{rendererClass=&NiagaraMeshRendererProperties, rendererIndex=0}]
 }
 LightningDirectional001 {
  emitterUpdateScript [
   &EmitterState(Life Cycle Mode=::NewEnumerator1, Loop Behavior=::NewEnumerator1, Loop Delay=0.3, Loop Duration=0.4, UseLoopDelay=true)
   &SpawnBurst_Instantaneous(Spawn Count=7, Use Spawn Probability=false, UseLoopCountLimit=false)
  ]
  particleSpawnScript [
   &PartitionParticles()
   &InitializeParticle(
    Color=&Multiply_LinearColorByFloat(Linear Color=$System.ElimColor, Float=0.4)
    Lifetime=&RandomRangeFloat(Minimum=0.15, Maximum=0.2)
    Mesh Scale Mode=::NewEnumerator3
    Mesh Uniform Scale=&RandomRangeFloat(Minimum=0.12, Maximum=0.15)
    Position Mode=::NewEnumerator0
    Write Material Random=true
   )
   &CameraOffset(Camera Offset Amount=2)
   &AddVelocity(Velocity=&Multiply_Vector(
     A=<0 0 1>
     B=&VectorFromFloat(Value=&SelectFloatFromArray(Direct Array Index=$Particles.PartitionParticles.Partition, Float Selection Array=DataInterface() {propertyValues {FloatData [-1,1]}}))
    )) {module=AddVelocity002}
   &ShapeLocation(MaskSpawnGroup=false, Override Seed=false, Sphere Radius=5, UseOwnerScale=true) {module=ShapeLocation001}
  ]
  particleUpdateScript [
   &ParticleState()
   &ScaleMeshSize(Scale Factor=&RandomRangeVector(Minimum=<2.5 2.5 5>, Maximum=<4.3 4.3 5>)) {module=ScaleMeshSize001}
   &ScaleMeshSize(Scale Factor=&MakeVector(
     X=1
     Y=1
     Z=&FloatFromCurve(
      FloatCurve=DataInterface() {propertyValues {Curve {keys [
          {interpMode=RCIM_Cubic, leaveTangent=5.4877, tangentMode=RCTM_Break, time=0.2, value=0}
          {interpMode=RCIM_Cubic, leaveTangent=-0.1745, time=0.7665, value=1}
         ]}}}
      Scale Curve=1.25
     )
    ))
   &DynamicMaterialParameters(
    Index 0 Param 1=&MakeFloatFromVector(Vector=$Particles.Scale, Channel=::NewEnumerator0)
    Index 0 Param 2=&RandomRangeFloat(Minimum=0, Maximum=10)
    Index 0 Param 3=&FloatFromCurve(
     FloatCurve=DataInterface() {propertyValues {Curve {keys [{interpMode=RCIM_Cubic, leaveTangent=8.1085, tangentMode=RCTM_Break, time=0.8568, value=1}, {tangentMode=RCTM_User, time=1, value=0}]}}}
     Scale Curve=0.05
    )
    Index 0 Param 4=0.8
    Index 1 Param 1=&FloatFromCurve(FloatCurve=DataInterface() {propertyValues {Curve {keys [
         {interpMode=RCIM_Cubic, leaveTangent=1.1034, tangentMode=RCTM_Break, time=0, value=0}
         {arriveTangent=2.2399, tangentMode=RCTM_User, time=0.5, value=1}
        ]}}})
    Index 1 Param 2=&FloatFromCurve(FloatCurve=DataInterface() {propertyValues {Curve {keys [<at 0.5 0>, {arriveTangent=3.2053, tangentMode=RCTM_User, time=0.9, value=1}]}}})
    Index 1 Param 3=&FloatFromCurve(FloatCurve=DataInterface() {propertyValues {Curve {keys [<at 0 0 RCIM_Cubic>, {arriveTangent=3.4223, interpMode=RCIM_Cubic, leaveTangent=3.4223, tangentMode=RCTM_User, time=1, value=1}]}}})
    Index 1 Param 4=&FloatFromCurve(
     FloatCurve=DataInterface() {propertyValues {Curve {keys [
         {interpMode=RCIM_Cubic, leaveTangent=0.1169, tangentMode=RCTM_Break, time=0, value=0}
         {arriveTangent=2.5708, interpMode=RCIM_Cubic, leaveTangent=2.5708, tangentMode=RCTM_User, time=1, value=1}
        ]}}}
     Scale Curve=&RandomRangeFloat(Minimum=30, Maximum=40)
    )
    Index 2 Param 1=&FloatFromCurve(FloatCurve=DataInterface() {propertyValues {Curve {keys [
         {interpMode=RCIM_Cubic, leaveTangent=-0.0025, tangentMode=RCTM_Break, time=0, value=0.5}
         {arriveTangent=-0.0316, interpMode=RCIM_Cubic, tangentMode=RCTM_Break, time=1, value=-0.2}
        ]}}})
    Index 2 Param 2=&FloatFromCurve(FloatCurve=DataInterface() {propertyValues {Curve {keys [
         {interpMode=RCIM_Cubic, tangentMode=RCTM_User, time=0, value=0}
         {arriveTangent=2.0522, interpMode=RCIM_Cubic, leaveTangent=2.0522, tangentMode=RCTM_User, time=1, value=1}
        ]}}})
    Module.Index0Param0Write=true
    Module.Index0Param1Write=true
    Module.Index0Param2Write=true
    Module.Index0Param3Write=true
    Module.Index1Param0Write=true
    Module.Index1Param1Write=true
    Module.Index1Param2Write=true
    Module.Index1Param3Write=true
    Module.Index2Param0Write=true
    Module.Index2Param1Write=true
    Module.Index2Param2Write=false
    Module.Index2Param3Write=false
    Module.Index3Param0Write=false
    Module.Index3Param1Write=false
    Module.Index3Param2Write=false
    Module.Index3Param3Write=true
    Write Parameter Index 1=true
    Write Parameter Index 2=true
    Write Parameter Index 3=true
   )
   &SolveForcesAndVelocity(Clamp Velocity=false, Limit Acceleration=false)
   &ScaleColor(
    Scale Alpha=&FloatFromCurve(FloatCurve=DataInterface() {propertyValues {Curve {keys [
         {tangentMode=RCTM_User, time=0, value=1}
         {arriveTangent=-0.0405, interpMode=RCIM_Cubic, leaveTangent=-0.0405, tangentMode=RCTM_User, time=0.4, value=1}
         {arriveTangent=-3.5786, interpMode=RCIM_Cubic, leaveTangent=-3.5786, tangentMode=RCTM_User, time=1, value=0}
        ]}}})
    Scale RGB=&Lerp_Vector(
     A=&VectorFromFloat(Value=&FloatFromCurve(
       FloatCurve=DataInterface() {propertyValues {Curve {keys [<at 0 1>,<at 0.1 1>,<at 0.2 2>,<at 0.3 1>,<at 0.6 1>,<at 0.7 2>,<at 0.8 1>,<at 1 1>]}}}
       Scale Curve=10
      ))
     B=&VectorFromFloat(Value=&RandomRangeFloat(Minimum=0.3, Maximum=5, Evaluation Type=::NewEnumerator1))
     Alpha=&FloatFromCurve(FloatCurve=DataInterface() {propertyValues {Curve {keys [
          {interpMode=RCIM_Cubic, leaveTangent=2.5294, tangentMode=RCTM_Break, time=0, value=0}
          {arriveTangent=-0.2166, interpMode=RCIM_Cubic, tangentMode=RCTM_Break, time=0.4, value=1}
         ]}}})
    )
    ScaleA=true
    ScaleRGB=true
   )
   /SetVariables(Particles.MeshOrientation=&MultiplyQuaternion(
     Quaternion A=&MultiplyQuaternion(
      Quaternion A=&MakeQuatFromAxisAngle(Axis=<0 0 1>, AngleInDegrees=0)
      Quaternion B=&FindQuatBetweenTwoVectors(From Vector=<0 0 1>, To Vector=$Particles.Velocity)
     )
     Quaternion B=<0 0 0 1>
    )) {isSetParameters=true}
  ]
  rendererClasses [&NiagaraMeshRendererProperties]
  renderers [
   {rendererClass=&NiagaraMeshRendererProperties, rendererIndex=0}
   {rendererClass=&NiagaraMeshRendererProperties, rendererIndex=1}
  ]
 }
 StaffChargeup001 {
  bEnabled=false
  emitterUpdateScript [
   &EmitterState(
    Life Cycle Mode=::NewEnumerator1
    Loop Behavior=::NewEnumerator1
    Loop Delay=&MaskFloat(
     Mask Parameter=$Emitter.SpawnGroup
     Mask=0
     Masked Float=0
     Passthrough Float=&MaskFloat(
      Mask Parameter=$Emitter.SpawnGroup
      Mask=1
      Masked Float=0.2
      Passthrough Float=&MaskFloat(
       Mask Parameter=$Emitter.SpawnGroup
       Mask=2
       Masked Float=0.7
       Passthrough Float=&MaskFloat(
        Mask Parameter=$Emitter.SpawnGroup
        Mask=3
        Masked Float=1
        Passthrough Float=&MaskFloat(
         Mask Parameter=$Emitter.SpawnGroup
         Mask=4
         Masked Float=2.2
         Passthrough Float=&MaskFloat(Mask Parameter=$Emitter.SpawnGroup, Mask=5, Masked Float=0.9, Passthrough Float=0.9)
        )
       )
      )
     )
    )
    Loop Duration=&MaskFloat(
     Mask Parameter=$Emitter.SpawnGroup
     Mask=0
     Masked Float=2.75
     Passthrough Float=&MaskFloat(
      Mask Parameter=$Emitter.SpawnGroup
      Mask=1
      Masked Float=2.75
      Passthrough Float=&MaskFloat(
       Mask Parameter=$Emitter.SpawnGroup
       Mask=2
       Masked Float=2.75
       Passthrough Float=&MaskFloat(
        Mask Parameter=$Emitter.SpawnGroup
        Mask=3
        Masked Float=2.75
        Passthrough Float=&MaskFloat(
         Mask Parameter=$Emitter.SpawnGroup
         Mask=4
         Masked Float=1
         Passthrough Float=&MaskFloat(Mask Parameter=$Emitter.SpawnGroup, Mask=5, Masked Float=2.75, Passthrough Float=1)
        )
       )
      )
     )
    )
    UseLoopDelay=true
   )
   &SpawnRate(SpawnGroup=5, SpawnRate=&Multiply_Float(A=15, B=$System.FadeOut), Use Spawn Probability=false) {enabled=false, module=SpawnRate001}
   &SpawnBurst_Instantaneous(
    Spawn Count=&RandomRangeInt(Minimum=9, Maximum=11)
    Spawn Group=4
    Use Spawn Probability=false
    UseLoopCountLimit=false
   ) {enabled=false, module=SpawnBurst_Instantaneous003}
   &SpawnBurst_Instantaneous(Spawn Group=3, Spawn Time=1, Use Spawn Probability=false, UseLoopCountLimit=false) {module=SpawnBurst_Instantaneous002}
   &SpawnRate(SpawnRate=&Multiply_Float(A=25, B=$System.FadeOut), Use Spawn Probability=false) {enabled=false}
   &SpawnBurst_Instantaneous(Spawn Group=1, Spawn Time=0.2, Use Spawn Probability=false, UseLoopCountLimit=false) {enabled=false}
  ]
  particleSpawnScript [
   /SetVariables(Particles.VisibilityTag=&SpawnGroup()) {isSetParameters=true}
   &InitializeParticle(
    Color=&MaskLinearColorBySpawnGroup(
     Masked Linear Color=&Multiply_LinearColorByFloat(Linear Color=$System.ElimColor, Float=0.4)
     Passthrough Linear Color=&MaskLinearColorBySpawnGroup(
      Masked Linear Color=&Multiply_LinearColorByFloat(Linear Color=$System.ElimColor, Float=0.5)
      Passthrough Linear Color=&MaskLinearColorBySpawnGroup(
       Masked Linear Color=$System.ElimColor
       Passthrough Linear Color=&MaskLinearColorBySpawnGroup(
        Masked Linear Color=$System.ElimColor
        Passthrough Linear Color=&MaskLinearColorBySpawnGroup(
         Masked Linear Color=&Multiply_LinearColorByFloat(Linear Color=$System.ElimColor, Float=&RandomRangeFloat(Minimum=0.8, Maximum=25))
         Passthrough Linear Color=&Multiply_LinearColorByFloat(Linear Color=$System.ElimColor, Float=0.4)
         Spawn Group Mask=4
        )
        Spawn Group Mask=3
       )
       Spawn Group Mask=2
      )
      Spawn Group Mask=1
     )
    )
    Lifetime=&MaskFloatBySpawnGroup(
     Masked Float=&RandomRangeFloat(Minimum=0.175, Maximum=0.35)
     Passthrough Float=&MaskFloatBySpawnGroup(
      Masked Float=1.25
      Passthrough Float=&MaskFloatBySpawnGroup(
       Masked Float=2.75
       Passthrough Float=&MaskFloatBySpawnGroup(
        Masked Float=1
        Passthrough Float=&MaskFloatBySpawnGroup(
         Masked Float=&RandomRangeFloat(Minimum=0.1, Maximum=1)
         Passthrough Float=&RandomRangeFloat(Minimum=0.175, Maximum=0.3)
         Spawn Group Mask=4
        )
        Spawn Group Mask=3
       )
       Spawn Group Mask=2
      )
      Spawn Group Mask=1
     )
    )
    Mesh Scale=&MaskVectorBySpawnGroup(
     Masked Vector=&RandomRangeVector(Minimum=<0.75 0.75 0.75>, Maximum=<2 2 2>)
     Passthrough Vector=&MaskVectorBySpawnGroup(
      Masked Vector=<0 0 0>
      Passthrough Vector=&MaskVectorBySpawnGroup(
       Masked Vector=<1.05 1.05 1.001>
       Passthrough Vector=&MaskVectorBySpawnGroup(
        Masked Vector=<0 0 0>
        Passthrough Vector=&MaskVectorBySpawnGroup(
         Masked Vector=<0 0 0>
         Passthrough Vector=&RandomRangeVector(Minimum=<0.3 0.3 0.5>, Maximum=<0.5 0.5 1>)
         Spawn Group Mask=4
        )
        Spawn Group Mask=3
       )
       Spawn Group Mask=2
      )
      Spawn Group Mask=1
     )
    )
    Mesh Scale Mode=::NewEnumerator4
    Position Offset=&MaskVectorBySpawnGroup(
     Masked Vector=<0 0 0>
     Passthrough Vector=&MaskVectorBySpawnGroup(
      Masked Vector=<0 0 25>
      Passthrough Vector=&MaskVectorBySpawnGroup(
       Masked Vector=<0 0 0>
       Passthrough Vector=&MaskVectorBySpawnGroup(Masked Vector=<0 0 0.3>, Passthrough Vector=<0 0 0>, Spawn Group Mask=3)
       Spawn Group Mask=2
      )
      Spawn Group Mask=1
     )
    )
    Sprite Size Mode=::NewEnumerator3
    Uniform Sprite Size=&MaskFloatBySpawnGroup(
     Masked Float=0
     Passthrough Float=&MaskFloatBySpawnGroup(
      Masked Float=50
      Passthrough Float=&MaskFloatBySpawnGroup(
       Masked Float=0
       Passthrough Float=&MaskFloatBySpawnGroup(Masked Float=100, Passthrough Float=&RandomRangeFloat(Minimum=1, Maximum=5), Spawn Group Mask=3)
       Spawn Group Mask=2
      )
      Spawn Group Mask=1
     )
    )
    UsePositionOffset=true
    Write Material Random=true
   )
   &InitialMeshOrientation(
    Mesh Orientation Mode=::NewEnumerator2
    Rotation=&MaskVectorBySpawnGroup(
     Masked Vector=&RandomRangeVector(Minimum=<0 0 -0.5>, Maximum=<0 0 0.5>)
     Passthrough Vector=&MaskVectorBySpawnGroup(
      Masked Vector=<0 0 0>
      Passthrough Vector=&MaskVectorBySpawnGroup(
       Masked Vector=<0 0 -0.25>
       Passthrough Vector=&MaskVectorBySpawnGroup(
        Masked Vector=<0 0 0>
        Passthrough Vector=&MaskVectorBySpawnGroup(
         Masked Vector=<0 0 0>
         Passthrough Vector=&RandomRangeVector(Minimum=<0 0 -1>, Maximum=<0.2 0.2 1>)
         Spawn Group Mask=4
        )
        Spawn Group Mask=3
       )
       Spawn Group Mask=2
      )
      Spawn Group Mask=1
     )
    )
    Use Rotation Vector=true
   )
   &ShapeLocation(
    Box Midpoint=&MaskVectorBySpawnGroup(
     Masked Vector=<0.5 0.5 0.5>
     Passthrough Vector=&MaskVectorBySpawnGroup(
      Masked Vector=<0.5 0.5 0.5>
      Passthrough Vector=&MaskVectorBySpawnGroup(
       Masked Vector=<0.5 0.5 0.5>
       Passthrough Vector=&MaskVectorBySpawnGroup(
        Masked Vector=<0.5 0.5 0.5>
        Passthrough Vector=&MaskVectorBySpawnGroup(Masked Vector=<0.5 0.5 0.5>, Passthrough Vector=<0.5 0.5 1>, Spawn Group Mask=4)
        Spawn Group Mask=3
       )
       Spawn Group Mask=2
      )
      Spawn Group Mask=1
     )
    )
    Box Size=&MaskVectorBySpawnGroup(
     Masked Vector=<1 1 70>
     Passthrough Vector=&MaskVectorBySpawnGroup(
      Masked Vector=<0 0 0>
      Passthrough Vector=&MaskVectorBySpawnGroup(
       Masked Vector=<0 0 0>
       Passthrough Vector=&MaskVectorBySpawnGroup(
        Masked Vector=<1 1 35>
        Passthrough Vector=&MaskVectorBySpawnGroup(Masked Vector=<0 0 0>, Passthrough Vector=<10 10 130>, Spawn Group Mask=4)
        Spawn Group Mask=3
       )
       Spawn Group Mask=2
      )
      Spawn Group Mask=1
     )
    )
    MaskSpawnGroup=false
    Offset=&MaskVectorBySpawnGroup(
     Masked Vector=<0 0 -30>
     Passthrough Vector=&MaskVectorBySpawnGroup(
      Masked Vector=<0 0 0>
      Passthrough Vector=&MaskVectorBySpawnGroup(Masked Vector=<0 0 0>, Passthrough Vector=<0 0 0>, Spawn Group Mask=2)
      Spawn Group Mask=1
     )
    )
    Offset Mode=::NewEnumerator0
    Override Seed=false
    Shape Primitive=::NewEnumerator2
    Surface Only=false
    UseOwnerScale=true
   )
   &SystemLocation(
    MaskSpawnGroup=false
    Offset=&MaskVectorBySpawnGroup(
     Masked Vector=$Particles.ShapeLocation.ShapePosition
     Passthrough Vector=&MaskVectorBySpawnGroup(
      Masked Vector=<0 0 25>
      Passthrough Vector=&MaskVectorBySpawnGroup(
       Masked Vector=<-5.3 0 -55>
       Passthrough Vector=&MaskVectorBySpawnGroup(
        Masked Vector=<0 0 20>
        Passthrough Vector=&MaskVectorBySpawnGroup(Masked Vector=<0 0 0>, Passthrough Vector=$Particles.ShapeLocation.ShapePosition, Spawn Group Mask=4)
        Spawn Group Mask=3
       )
       Spawn Group Mask=2
      )
      Spawn Group Mask=1
     )
    )
   ) {enabled=false}
   &CameraOffset(Camera Offset Amount=&MaskFloatBySpawnGroup(
     Masked Float=0
     Passthrough Float=&MaskFloatBySpawnGroup(Masked Float=10, Passthrough Float=0, Spawn Group Mask=1)
    ))
   &AddVelocity(
    Velocity=&MaskVectorBySpawnGroup(
     Masked Vector=<0 0 0>
     Passthrough Vector=&MaskVectorBySpawnGroup(
      Masked Vector=<0 0 0>
      Passthrough Vector=&MaskVectorBySpawnGroup(
       Masked Vector=<0 0 0>
       Passthrough Vector=&MaskVectorBySpawnGroup(
        Masked Vector=<-1 0 1>
        Passthrough Vector=&MaskVectorBySpawnGroup(Masked Vector=<0 0 0>, Passthrough Vector=<0 0 0>, Spawn Group Mask=4)
        Spawn Group Mask=3
       )
       Spawn Group Mask=2
      )
      Spawn Group Mask=1
     )
    )
    Velocity Speed Scale=15
   )
   &AddVelocity(
    Cone Angle=&MaskFloatBySpawnGroup(Masked Float=20, Passthrough Float=95, Spawn Group Mask=5)
    Cone Axis=<0 0 1>
    Distribution Along Cone Axis=0.2
    Module.UseConeFalloff=false
    Module.UseVelDistribution=true
    Override Seed=false
    Velocity Mode=::NewEnumerator2
    Velocity Speed=&MaskFloatBySpawnGroup(
     Masked Float=0
     Passthrough Float=&MaskFloatBySpawnGroup(
      Masked Float=0
      Passthrough Float=&MaskFloatBySpawnGroup(
       Masked Float=0
       Passthrough Float=&MaskFloatBySpawnGroup(
        Masked Float=0
        Passthrough Float=&MaskFloatBySpawnGroup(Masked Float=&RandomRangeFloat(Minimum=150, Maximum=300), Passthrough Float=0, Spawn Group Mask=4)
        Spawn Group Mask=3
       )
       Spawn Group Mask=2
      )
      Spawn Group Mask=1
     )
    )
   ) {module=AddVelocity001}
  ]
  particleUpdateScript [
   &ParticleState()
   &ScaleSpriteSize(
    Scale Sprite Size Mode=::NewEnumerator0
    Uniform Scale Factor=&MaskFloatBySpawnGroup(
     Masked Float=0
     Passthrough Float=&MaskFloatBySpawnGroup(
      Masked Float=&FloatFromCurve(
       FloatCurve=DataInterface() {propertyValues {Curve {keys [
           <at 0 0.3 RCIM_Cubic>
           {arriveTangent=5.3971, interpMode=RCIM_Cubic, leaveTangent=5.3971, tangentMode=RCTM_User, time=0.6462, value=1.5162}
           {arriveTangent=13.7263, interpMode=RCIM_Cubic, leaveTangent=13.7263, time=0.8013, value=1.0461}
           {arriveTangent=-57.6853, interpMode=RCIM_Cubic, leaveTangent=-57.6853, tangentMode=RCTM_User, time=0.9, value=5}
           <at 1 0 RCIM_Cubic>
          ]}}}
       Scale Curve=0.85
      )
      Passthrough Float=&MaskFloatBySpawnGroup(
       Masked Float=0
       Passthrough Float=&FloatFromCurve(FloatCurve=DataInterface() {propertyValues {Curve {keys [
            {arriveTangent=3.4683, interpMode=RCIM_Cubic, leaveTangent=3.4683, tangentMode=RCTM_User, time=0, value=0}
            {interpMode=RCIM_Cubic, tangentMode=RCTM_User, time=1, value=1.15}
           ]}}})
       Spawn Group Mask=2
      )
      Spawn Group Mask=1
     )
    )
   )
   &InheritVelocity(
    Inherited Velocity Amount Scale=<0.01 0.01 0.01>
    Inherited Velocity Speed Limit:NiagaraFloat=&FloatFromCurve(
     FloatCurve=DataInterface() {propertyValues {Curve {keys [
         {interpMode=RCIM_Cubic, tangentMode=RCTM_User, time=0.0017, value=1}
         {interpMode=RCIM_Cubic, tangentMode=RCTM_User, time=0.5, value=1}
         {tangentMode=RCTM_User, time=1, value=0}
        ]}}}
     Scale Curve=100
    )
    Source Speed Threshold=15
    Speed Limit=true
    UseSpeedThreshold=false
   ) {enabled=false}
   &ScaleMeshSize(Scale Factor=&MaskVectorBySpawnGroup(
     Masked Vector=&Multiply_VectorByFloat(Vector=<1 1 1>, Float=0.2)
     Passthrough Vector=&MaskVectorBySpawnGroup(
      Masked Vector=<0 0 0>
      Passthrough Vector=&MaskVectorBySpawnGroup(
       Masked Vector=<1 1 1>
       Passthrough Vector=&MaskVectorBySpawnGroup(
        Masked Vector=<1 1 1>
        Passthrough Vector=&MaskVectorBySpawnGroup(
         Masked Vector=<1 1 1>
         Passthrough Vector=&Multiply_VectorByFloat(
          Vector=&MakeCustomVectorFromBool(
           Bool=&SetBoolByFloatComparison(
            A=&VectorLength(Vector=Linked() {linkedVariable {name=Particles.Position, type=NiagaraPosition}})
            B=&VectorLength(Vector=<0 0 50>)
            Comparison Type=::NewEnumerator7
           )
           True Vector=<1 1 1>
           False Vector=<0.7 0.7 0.7>
          )
          Float=&Multiply_Float(
           A=0.2
           B=&FloatFromCurve(
            FloatCurve=DataInterface() {propertyValues {Curve {keys [
                <at 0 1 RCIM_Cubic>
                {arriveTangent=-0.1283, interpMode=RCIM_Cubic, leaveTangent=0.4938, tangentMode=RCTM_Break, time=0.9, value=1}
                {arriveTangent=7.2371, interpMode=RCIM_Cubic, leaveTangent=7.2371, tangentMode=RCTM_User, time=1, value=2}
               ]}}}
            CurveIndex=$System.LoopedAge
           )
          )
         )
         Spawn Group Mask=4
        )
        Spawn Group Mask=3
       )
       Spawn Group Mask=2
      )
      Spawn Group Mask=1
     )
    ))
   &DynamicMaterialParameters(
    Index 0 Param 1=&MakeFloatFromVector(Vector=$Particles.Scale, Channel=::NewEnumerator0)
    Index 0 Param 2=&RandomRangeFloat(Minimum=0, Maximum=10)
    Index 0 Param 3=&FloatFromCurve(
     FloatCurve=DataInterface() {propertyValues {Curve {keys [{interpMode=RCIM_Cubic, leaveTangent=8.1085, tangentMode=RCTM_Break, time=0.8568, value=1}, {tangentMode=RCTM_User, time=1, value=0}]}}}
     Scale Curve=0.07
    )
    Index 0 Param 4=0.8
    Index 1 Param 1=&FloatFromCurve(FloatCurve=DataInterface() {propertyValues {Curve {keys [
         {interpMode=RCIM_Cubic, leaveTangent=1.1034, tangentMode=RCTM_Break, time=0, value=0}
         {arriveTangent=2.2399, tangentMode=RCTM_User, time=0.5, value=1}
        ]}}})
    Index 1 Param 2=&FloatFromCurve(FloatCurve=DataInterface() {propertyValues {Curve {keys [<at 0.5 0>, {arriveTangent=3.2053, tangentMode=RCTM_User, time=0.9, value=1}]}}})
    Index 1 Param 3=&FloatFromCurve(FloatCurve=DataInterface() {propertyValues {Curve {keys [<at 0 0 RCIM_Cubic>, {arriveTangent=3.4223, interpMode=RCIM_Cubic, leaveTangent=3.4223, tangentMode=RCTM_User, time=1, value=1}]}}})
    Index 1 Param 4=&FloatFromCurve(
     FloatCurve=DataInterface() {propertyValues {Curve {keys [
         {interpMode=RCIM_Cubic, leaveTangent=0.1169, tangentMode=RCTM_Break, time=0, value=0}
         {arriveTangent=2.5708, interpMode=RCIM_Cubic, leaveTangent=2.5708, tangentMode=RCTM_User, time=1, value=1}
        ]}}}
     Scale Curve=&RandomRangeFloat(Minimum=1, Maximum=40)
    )
    Index 2 Param 1=&FloatFromCurve(FloatCurve=DataInterface() {propertyValues {Curve {keys [
         {interpMode=RCIM_Cubic, leaveTangent=-0.0025, tangentMode=RCTM_Break, time=0, value=0.5}
         {arriveTangent=-0.0316, interpMode=RCIM_Cubic, tangentMode=RCTM_Break, time=1, value=-0.2}
        ]}}})
    Index 2 Param 2=&FloatFromCurve(
     FloatCurve=DataInterface() {propertyValues {Curve {keys [
         {interpMode=RCIM_Cubic, tangentMode=RCTM_User, time=0, value=0}
         {arriveTangent=2.0522, interpMode=RCIM_Cubic, leaveTangent=2.0522, tangentMode=RCTM_User, time=1, value=1}
        ]}}}
     Scale Curve=2
    )
    Module.Index0Param0Write=true
    Module.Index0Param1Write=true
    Module.Index0Param2Write=true
    Module.Index0Param3Write=true
    Module.Index1Param0Write=true
    Module.Index1Param1Write=true
    Module.Index1Param2Write=true
    Module.Index1Param3Write=true
    Module.Index2Param0Write=true
    Module.Index2Param1Write=true
    Module.Index2Param2Write=false
    Module.Index2Param3Write=false
    Module.Index3Param0Write=false
    Module.Index3Param1Write=false
    Module.Index3Param2Write=false
    Module.Index3Param3Write=true
    Write Parameter Index 1=true
    Write Parameter Index 2=true
    Write Parameter Index 3=true
   )
   &UpdateMeshOrientation(
    Rotation Rate=&MaskFloatBySpawnGroup(
     Masked Float=&RandomRangeFloat(Minimum=-0.5, Maximum=0.5)
     Passthrough Float=&MaskFloatBySpawnGroup(Masked Float=0, Passthrough Float=0, Spawn Group Mask=1)
    )
    Rotation Vector=<0 0 1>
   )
   &ScaleColor(
    Scale Alpha=&MaskFloatBySpawnGroup(
     Masked Float=&FloatFromCurve(FloatCurve=DataInterface() {propertyValues {Curve {keys [
          {tangentMode=RCTM_User, time=0, value=1}
          {arriveTangent=-0.0405, interpMode=RCIM_Cubic, leaveTangent=-0.0405, tangentMode=RCTM_User, time=0.7009, value=1}
          {arriveTangent=-8.0089, interpMode=RCIM_Cubic, leaveTangent=-8.0089, tangentMode=RCTM_User, time=1, value=0}
         ]}}})
     Passthrough Float=&MaskFloatBySpawnGroup(
      Masked Float=&FloatFromCurve(FloatCurve=DataInterface() {propertyValues {Curve {keys [
           {interpMode=RCIM_Cubic, leaveTangent=0.2193, tangentMode=RCTM_Break, time=0, value=0}
           {arriveTangent=2.5, interpMode=RCIM_Cubic, leaveTangent=2.5, time=0.1691, value=0.1683}
           {arriveTangent=-0.2166, interpMode=RCIM_Cubic, tangentMode=RCTM_Break, time=0.4, value=1}
          ]}}})
      Passthrough Float=&MaskFloatBySpawnGroup(
       Masked Float=&FloatFromCurve(
        FloatCurve=DataInterface() {propertyValues {Curve {keys [
            {arriveTangent=-0.141, interpMode=RCIM_Cubic, leaveTangent=-0.141, tangentMode=RCTM_User, time=0, value=0}
            {arriveTangent=3.8262, interpMode=RCIM_Cubic, leaveTangent=21.5493, tangentMode=RCTM_Break, time=0.5, value=1}
            {arriveTangent=351.5977, interpMode=RCIM_Cubic, leaveTangent=-207.9236, tangentMode=RCTM_Break, time=0.5234, value=11.2794}
            {arriveTangent=-5.5932, tangentMode=RCTM_User, time=0.7, value=0}
           ]}}}
        Scale Curve=2
       )
       Passthrough Float=&MaskFloatBySpawnGroup(
        Masked Float=&FloatFromCurve(
         FloatCurve=DataInterface() {propertyValues {Curve {keys [
             {interpMode=RCIM_Cubic, tangentMode=RCTM_User, time=0, value=0}
             <at 0.7 1 RCIM_Cubic>
             {interpMode=RCIM_Cubic, tangentMode=RCTM_User, time=1, value=0}
            ]}}}
         Scale Curve=0.32
        )
        Passthrough Float=&MaskFloatBySpawnGroup(
         Masked Float=&FloatFromCurve(FloatCurve=DataInterface() {propertyValues {Curve {keys [
              {arriveTangent=9.0725, interpMode=RCIM_Cubic, leaveTangent=9.0725, tangentMode=RCTM_User, time=0, value=0}
              {arriveTangent=-2.5249, interpMode=RCIM_Cubic, leaveTangent=-2.5249, tangentMode=RCTM_User, time=0.1408, value=0.9847}
              {arriveTangent=-0.5654, interpMode=RCIM_Cubic, leaveTangent=-0.5654, time=0.5, value=0.35}
              {arriveTangent=-1.25, interpMode=RCIM_Cubic, leaveTangent=-1.25, time=0.6, value=0.725}
              {arriveTangent=-3.375, interpMode=RCIM_Cubic, leaveTangent=-0.1412, tangentMode=RCTM_Break, time=0.7, value=0.1}
              {arriveTangent=-0.7119, interpMode=RCIM_Cubic, leaveTangent=1.625, tangentMode=RCTM_Break, time=0.8, value=0.05}
              {arriveTangent=-0.25, interpMode=RCIM_Cubic, leaveTangent=-0.25, time=0.9, value=0.425}
              {arriveTangent=-0.0405, interpMode=RCIM_Cubic, leaveTangent=-0.0405, tangentMode=RCTM_User, time=1, value=0}
             ]}}})
         Passthrough Float=&FloatFromCurve(FloatCurve=DataInterface() {propertyValues {Curve {keys [
              {tangentMode=RCTM_User, time=0, value=1}
              {arriveTangent=-0.0405, interpMode=RCIM_Cubic, leaveTangent=-0.0405, tangentMode=RCTM_User, time=0.7009, value=1}
              {arriveTangent=-8.0089, interpMode=RCIM_Cubic, leaveTangent=-8.0089, tangentMode=RCTM_User, time=1, value=0}
             ]}}})
         Spawn Group Mask=4
        )
        Spawn Group Mask=3
       )
       Spawn Group Mask=2
      )
      Spawn Group Mask=1
     )
    )
    Scale RGB=&MaskVectorBySpawnGroup(
     Masked Vector=&Lerp_Vector(
      A=&VectorFromFloat(Value=&FloatFromCurve(
        FloatCurve=DataInterface() {propertyValues {Curve {keys [<at 0 1>,<at 0.1 1>,<at 0.2 2>,<at 0.3 1>,<at 0.6 1>,<at 0.7 2>,<at 0.8 1>,<at 1 1>]}}}
        Scale Curve=10
       ))
      B=&VectorFromFloat(Value=&RandomRangeFloat(Minimum=0.3, Maximum=5, Evaluation Type=::NewEnumerator1))
      Alpha=&FloatFromCurve(FloatCurve=DataInterface() {propertyValues {Curve {keys [
           {interpMode=RCIM_Cubic, leaveTangent=2.5294, tangentMode=RCTM_Break, time=0, value=0}
           {arriveTangent=-0.2166, interpMode=RCIM_Cubic, tangentMode=RCTM_Break, time=0.4, value=1}
          ]}}})
     )
     Passthrough Vector=&MaskVectorBySpawnGroup(
      Masked Vector=&Lerp_Vector(
       A=&VectorFromFloat(Value=&FloatFromCurve(
         FloatCurve=DataInterface() {propertyValues {Curve {keys [{tangentMode=RCTM_User, time=0, value=1}, {tangentMode=RCTM_User, time=1, value=1}]}}}
         Scale Curve=10
        ))
       B=&VectorFromFloat(Value=&RandomRangeFloat(Minimum=4, Maximum=5, Evaluation Type=::NewEnumerator1))
       Alpha=&FloatFromCurve(FloatCurve=DataInterface() {propertyValues {Curve {keys [
            {interpMode=RCIM_Cubic, leaveTangent=2.5294, tangentMode=RCTM_Break, time=0, value=0}
            {arriveTangent=-0.2166, interpMode=RCIM_Cubic, tangentMode=RCTM_Break, time=0.4, value=1}
           ]}}})
      )
      Passthrough Vector=&MaskVectorBySpawnGroup(
       Masked Vector=&Lerp_Vector(
        A=&VectorFromFloat(Value=&FloatFromCurve(FloatCurve=DataInterface() {propertyValues {Curve {keys [<at 0 1>,<at 0.3 1>,<at 0.7 1>,<at 0.8 4>,<at 0.9 1>,<at 1 1>]}}}, Scale Curve=10))
        B=&VectorFromFloat(Value=&RandomRangeFloat(Minimum=8.5, Maximum=10, Evaluation Type=::NewEnumerator1))
        Alpha=&FloatFromCurve(FloatCurve=DataInterface() {propertyValues {Curve {keys [
             {interpMode=RCIM_Cubic, leaveTangent=0.06, tangentMode=RCTM_Break, time=0.1, value=0}
             {arriveTangent=2.6781, interpMode=RCIM_Cubic, leaveTangent=2.6781, tangentMode=RCTM_User, time=0.8, value=0.8}
             {arriveTangent=2.5, interpMode=RCIM_Cubic, leaveTangent=-22.9319, tangentMode=RCTM_Break, time=0.9, value=1.5}
             <at 1 0 RCIM_Cubic>
            ]}}})
       )
       Passthrough Vector=&MaskVectorBySpawnGroup(
        Masked Vector=&Lerp_Vector(
         A=&VectorFromFloat(Value=&FloatFromCurve(FloatCurve=DataInterface() {propertyValues {Curve {keys [<at 0 1>,<at 0.1 1>,<at 0.2 4>,<at 0.3 1>,<at 0.6 1>,<at 0.7 2>,<at 0.8 1>,<at 1 1>]}}}))
         B=&VectorFromFloat(Value=&RandomRangeFloat(Minimum=0.5, Maximum=1, Evaluation Type=::NewEnumerator1))
         Alpha=&FloatFromCurve(FloatCurve=DataInterface() {propertyValues {Curve {keys [
              {interpMode=RCIM_Cubic, leaveTangent=2.5294, tangentMode=RCTM_Break, time=0, value=0}
              {arriveTangent=-0.2166, interpMode=RCIM_Cubic, tangentMode=RCTM_Break, time=0.4, value=1}
             ]}}})
        )
        Passthrough Vector=&MaskVectorBySpawnGroup(
         Masked Vector=&VectorFromFloat(Value=&FloatFromCurve(FloatCurve=DataInterface() {propertyValues {Curve {keys [<at 0 1>,<at 0.2 1>,<at 0.3 5>,<at 0.4 1>,<at 0.7 1>,<at 0.8 5>,<at 0.9 1>,<at 1 1>]}}}))
         Passthrough Vector=&Lerp_Vector(
          A=&VectorFromFloat(Value=&FloatFromCurve(
            FloatCurve=DataInterface() {propertyValues {Curve {keys [<at 0 1>,<at 0.1 1>,<at 0.2 2>,<at 0.3 1>,<at 0.6 1>,<at 0.7 2>,<at 0.8 1>,<at 1 1>]}}}
            Scale Curve=10
           ))
          B=&VectorFromFloat(Value=&RandomRangeFloat(Minimum=0.3, Maximum=5, Evaluation Type=::NewEnumerator1))
          Alpha=&FloatFromCurve(FloatCurve=DataInterface() {propertyValues {Curve {keys [
               {interpMode=RCIM_Cubic, leaveTangent=2.5294, tangentMode=RCTM_Break, time=0, value=0}
               {arriveTangent=-0.2166, interpMode=RCIM_Cubic, tangentMode=RCTM_Break, time=0.4, value=1}
              ]}}})
         )
         Spawn Group Mask=4
        )
        Spawn Group Mask=3
       )
       Spawn Group Mask=2
      )
      Spawn Group Mask=1
     )
    )
    ScaleA=true
    ScaleRGB=true
   )
   &Drag(
    Drag=&MaskFloatBySpawnGroup(
     Masked Float=0
     Passthrough Float=&MaskFloatBySpawnGroup(
      Masked Float=0
      Passthrough Float=&MaskFloatBySpawnGroup(
       Masked Float=0
       Passthrough Float=&MaskFloatBySpawnGroup(
        Masked Float=&FloatFromCurve(
         FloatCurve=DataInterface() {propertyValues {Curve {keys [
             {interpMode=RCIM_Cubic, tangentMode=RCTM_User, time=0, value=0}
             {arriveTangent=1.6092, interpMode=RCIM_Cubic, leaveTangent=1.6092, tangentMode=RCTM_User, time=1, value=1}
            ]}}}
         Scale Curve=7
        )
        Passthrough Float=&RandomRangeFloat(Minimum=0.8, Maximum=1.2)
        Spawn Group Mask=3
       )
       Spawn Group Mask=2
      )
      Spawn Group Mask=1
     )
    )
    Use Linear Drag=true
    Use Rotational Drag=false
   )
   &GravityForce(Gravity=&MaskVectorBySpawnGroup(
     Masked Vector=<0 0 0>
     Passthrough Vector=&MaskVectorBySpawnGroup(
      Masked Vector=<0 0 0>
      Passthrough Vector=&MaskVectorBySpawnGroup(
       Masked Vector=<0 0 0>
       Passthrough Vector=&MaskVectorBySpawnGroup(
        Masked Vector=<0 0 0>
        Passthrough Vector=&MaskVectorBySpawnGroup(Masked Vector=<0 0 -700>, Passthrough Vector=<0 0 0>, Spawn Group Mask=4)
        Spawn Group Mask=3
       )
       Spawn Group Mask=2
      )
      Spawn Group Mask=1
     )
    ))
   &SolveForcesAndVelocity(Clamp Velocity=false, Limit Acceleration=false)
   &ScaleSpriteSizeBySpeed(
    Max Scale Factor=&MaskVector2dBySpawnGroup(
     Masked Vector2D=<1 1>
     Passthrough Vector2D=&MaskVector2dBySpawnGroup(
      Masked Vector2D=<1 1>
      Passthrough Vector2D=&MaskVector2dBySpawnGroup(
       Masked Vector2D=<1 1>
       Passthrough Vector2D=&MaskVector2dBySpawnGroup(Masked Vector2D=<1 1>, Passthrough Vector2D=<0.5 7>, Spawn Group Mask=3)
       Spawn Group Mask=2
      )
      Spawn Group Mask=1
     )
     Spawn Group Mask=0
    )
    Sample Scale Factor By Curve=true
    Velocity Threshold=500
   )
   /SetVariables(Particles.StaffDMP:Vector4f=&MakeVector4(
     X=&FloatFromCurve(FloatCurve=DataInterface() {propertyValues {Curve {keys [
          {interpMode=RCIM_Cubic, leaveTangent=-79.1318, tangentMode=RCTM_Break, time=0, value=1}
          {arriveTangent=-2.6873, tangentMode=RCTM_User, time=0.5, value=-10}
          {arriveTangent=14.6846, interpMode=RCIM_Cubic, leaveTangent=14.6846, tangentMode=RCTM_User, time=0.9, value=-10}
          <at 1 1>
         ]}}})
     Y=0
     Z=0
     W=0
    )) {isSetParameters=true}
  ]
  rendererClasses [&NiagaraMeshRendererProperties, &NiagaraSpriteRendererProperties]
  renderers [
   {rendererClass=&NiagaraMeshRendererProperties, rendererIndex=0}
   {rendererClass=&NiagaraMeshRendererProperties, rendererIndex=1}
   {rendererClass=&NiagaraSpriteRendererProperties, rendererIndex=2}
   {rendererClass=&NiagaraSpriteRendererProperties, rendererIndex=3}
   {rendererClass=&NiagaraSpriteRendererProperties, rendererIndex=4}
   {rendererClass=&NiagaraSpriteRendererProperties, rendererIndex=5}
   {rendererClass=&NiagaraMeshRendererProperties, rendererIndex=6}
   {rendererClass=&NiagaraMeshRendererProperties, rendererIndex=7}
  ]
 }
}
summary {
 systemName=NS_Emote_GlowFang_Despawn
 emitters [
  {emitterName=HangingParticulates, simTarget=GPUComputeSim, rendererClasses [&NiagaraSpriteRendererProperties]}
  {emitterName=HangingParticulates001, simTarget=GPUComputeSim, rendererClasses [&NiagaraSpriteRendererProperties]}
  {emitterName=Flash001, rendererClasses [&NiagaraLightRendererProperties, &NiagaraSpriteRendererProperties]}
  {emitterName=LightningDirectional001, rendererClasses [&NiagaraMeshRendererProperties]}
  {bEnabled=false, emitterName=Ice_Mesh, rendererClasses [&NiagaraMeshRendererProperties]}
  {
   bEnabled=false
   emitterName=StaffChargeup001
   rendererClasses [&NiagaraMeshRendererProperties, &NiagaraSpriteRendererProperties]
  }
  {emitterName=HangingParticulates002, rendererClasses [&NiagaraSpriteRendererProperties]}
 ]
 userVariables [
  User.bIsFrontEndPreview {type=NiagaraBool}
  User.bKillStage {type=NiagaraInt32}
  User.NormalizedNotifyProgress {type=NiagaraFloat, defaultValue=0, description=""}
 ]
}
