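# Niagara system for the GlowFang emote's ground impact: a blast with beams, shockwave, ground pulse, light and lightning-like zaps.
# CentralBlast packs these layers into one emitter, using spawn-group masks to vary behavior and a visibility tag feeding many sprite, mesh and light renderers.
# System variables supply ElimColor tinting and fade out continuous aura and lightning spawns; other layers use hardcoded reds.
# The Beams001 and CentralBlast001 emitters are disabled.
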NiagaraSystem /BRCosmetics/Animation/Game/MainPlayer/Emotes/GlowFang/VFX/NS_Emote_GlowFang_GroundImpact
dependencies {
 usedDataInterfaces [
  &NiagaraDataInterfaceCurve
  &NiagaraDataInterfaceColorCurve
  &NiagaraDataInterfaceVector2DCurve
  &NiagaraDataInterfaceArrayInt32
  &NiagaraDataInterfaceArrayFloat3
  &NiagaraDataInterfaceArrayColor
  &NiagaraDataInterfaceMeshRendererInfo
 ]
 usedDynamicInputs [&InvertBool]
 usedModules [
  &SystemState
  &self:SystemSpawnScript.SystemScriptSource.SystemScriptGraph.NiagaraNodeAssignment_0.SetVariables_#1
  &self:SystemSpawnScript.SystemScriptSource.SystemScriptGraph.NiagaraNodeAssignment_1.SetVariables_#2
  &self:SystemSpawnScript.SystemScriptSource.SystemScriptGraph.NiagaraNodeAssignment_2.SetVariables_#3
 ]
}
emitters {
 Beams001 {
  bEnabled=false
  emitterUpdateScript [
   &EmitterState()
   &SpawnBurst_Instantaneous(Spawn Count=25, Use Spawn Probability=false, UseLoopCountLimit=false)
  ]
  particleSpawnScript [
   /SetVariables(Particles.Rando:NiagaraFloat=&FloatFromCurve(
     FloatCurve=DataInterface() {propertyValues {Curve {keys [
         {interpMode=RCIM_Cubic, tangentMode=RCTM_User, time=0, value=0}
         {arriveTangent=2.3793, interpMode=RCIM_Cubic, leaveTangent=2.3793, tangentMode=RCTM_User, time=1, value=1}
        ]}}}
     CurveIndex=&RandomRangeFloat(Minimum=0, Maximum=1)
    )) {isSetParameters=true}
   &InitializeParticle(
    Color=&LerpLinearColors(StartColor=<rgba 0.634 1.8967 3 0.3>, EndColor=<rgba 0 5.8116 6 0.4>, LerpFactor=$Particles.Rando)
    Lifetime Max=0.4
    Lifetime Min=0.35
    Lifetime Mode=::NewEnumerator1
    Mesh Scale Mode=::NewEnumerator3
    Mesh Uniform Scale=&Lerp_Float(A=1.5, B=2.5, Alpha=$Particles.Rando)
    UsePositionOffset=false
    Write Material Random=true
   )
   &InitialMeshOrientation(Use Rotation Vector=true)
  ]
  particleUpdateScript [
   &ParticleState()
   &DynamicMaterialParameters(
    Index 0 Param 1=0.1
    Index 0 Param 2=&FloatFromCurve(
     FloatCurve=DataInterface() {propertyValues {Curve {keys [
         {arriveTangent=5.1604, interpMode=RCIM_Cubic, leaveTangent=5.1604, tangentMode=RCTM_User, time=0, value=0}
         {interpMode=RCIM_Cubic, tangentMode=RCTM_User, time=0.5, value=1}
         <at 1 0 RCIM_Cubic>
        ]}}}
     Scale Curve=&Lerp_Float(A=1.75, B=1.5, Alpha=$Particles.Rando)
    )
    Module.Param0WriteEnabled=true
    Module.Param1WriteEnabled=true
    Module.Param2WriteEnabled=false
    Module.Param3WriteEnabled=false
   ) {module=DynamicMaterialParameters001}
   &ScaleMeshSize(Scale Factor=&VectorFromFloat(Value=&FloatFromCurve(
      FloatCurve=DataInterface() {propertyValues {Curve {keys [
          {arriveTangent=3.3092, interpMode=RCIM_Cubic, leaveTangent=3.3092, tangentMode=RCTM_User, time=0, value=0.4}
          {arriveTangent=0.048, interpMode=RCIM_Cubic, leaveTangent=0.048, tangentMode=RCTM_User, time=0.5, value=1}
          {arriveTangent=-2.6178, interpMode=RCIM_Cubic, leaveTangent=-2.6178, tangentMode=RCTM_User, time=1, value=0.5}
         ]}}}
      Scale Curve=0.75
     )))
   &ScaleColor(
    Scale Alpha=&FloatFromCurve(
     FloatCurve=DataInterface() {propertyValues {Curve {keys [
         {arriveTangentWeight=0.0275, interpMode=RCIM_Cubic, leaveTangentWeight=0.3161, tangentMode=RCTM_User, time=0, value=1}
         {arriveTangent=-2.6497, interpMode=RCIM_Cubic, leaveTangent=-2.6497, tangentMode=RCTM_User, time=0.25, value=1}
         {arriveTangent=-0.0405, interpMode=RCIM_Cubic, leaveTangent=-0.0405, tangentMode=RCTM_User, time=1, value=0}
        ]}}}
     Scale Curve=0.25
    )
    Scale RGB=&MakeVectorFromLinearColorRGB(Color=&ColorFromCurve(ColorCurve=DataInterface() {propertyValues {
        AlphaCurve {keys [<at 0 1 RCIM_Cubic>]}
        BlueCurve {keys [<at 0 1 RCIM_Cubic>,<at 1 1 RCIM_Cubic>]}
        GreenCurve {keys [<at 0 1 RCIM_Cubic>,<at 1 1 RCIM_Cubic>]}
        RedCurve {keys [<at 0 1 RCIM_Cubic>,<at 1 1 RCIM_Cubic>]}
       }}))
    ScaleA=true
    ScaleRGB=true
   )
   &ScaleColor(Scale Alpha=$System.TimeOfDayBrightnessMultiplier, ScaleA=true, ScaleRGB=true) {enabled=false, module=ScaleColor001}
  ]
  rendererClasses [&NiagaraMeshRendererProperties]
  renderers [{rendererClass=&NiagaraMeshRendererProperties, rendererIndex=0}]
 }
 CentralBlast {
  emitterUpdateScript [
   &EmitterState()
   &SpawnBurst_Instantaneous(Use Spawn Probability=false, UseLoopCountLimit=false)
   &SpawnBurst_Instantaneous(Spawn Count=25, Spawn Group=1, Use Spawn Probability=false, UseLoopCountLimit=false) {module=SpawnBurst_Instantaneous001}
   &SpawnBurst_Instantaneous(Spawn Count=2, Spawn Group=2, Use Spawn Probability=false, UseLoopCountLimit=false) {module=SpawnBurst_Instantaneous002}
   &SpawnRate(
    SpawnGroup=3
    SpawnRate=&Multiply_Float(A=&RandomRangeFloat(Minimum=5, Maximum=6), B=$System.FadeOutAura)
    Use Spawn Probability=false
   )
   &SpawnBurst_Instantaneous(
    Spawn Count=&RandomRangeInt(Minimum=4, Maximum=7)
    Spawn Group=4
    Use Spawn Probability=false
    UseLoopCountLimit=false
   ) {module=SpawnBurst_Instantaneous005}
   &SpawnBurst_Instantaneous(Spawn Count=8, Spawn Group=5, Use Spawn Probability=false, UseLoopCountLimit=false) {module=SpawnBurst_Instantaneous003}
   &SpawnRate(SpawnGroup=5, SpawnRate=&Multiply_Float(A=9, B=$System.FadeOutBottomLightning), Use Spawn Probability=false) {module=SpawnRate001}
   &SpawnBurst_Instantaneous(Spawn Count=4, Spawn Group=6, Use Spawn Probability=false, UseLoopCountLimit=false) {module=SpawnBurst_Instantaneous004}
   &SpawnRate(SpawnGroup=6, SpawnRate=&Multiply_Float(A=15, B=$System.FadeOutAura), Use Spawn Probability=false) {module=SpawnRate002}
  ]
  particleSpawnScript [
   &PartitionParticles(Number Of Partitions=&MaskIntBySpawnGroup(
     Masked Int=0
     Passthrough Int=&MaskIntBySpawnGroup(
      Masked Int=0
      Passthrough Int=&MaskIntBySpawnGroup(
       Masked Int=1
       Passthrough Int=&MaskIntBySpawnGroup(
        Masked Int=0
        Passthrough Int=&MaskIntBySpawnGroup(
         Masked Int=0
         Passthrough Int=&MaskIntBySpawnGroup(Masked Int=0, Passthrough Int=0, Spawn Group Mask=5)
         Spawn Group Mask=4
        )
        Spawn Group Mask=3
       )
       Spawn Group Mask=2
      )
      Spawn Group Mask=1
     )
     Spawn Group Mask=0
    ))
   /SetVariables(Particles.VisibilityTag=&SpawnGroup()) {isSetParameters=true, module=SetVariables_#4}
   /SetVariables(
    Particles.BeamOrientation:Quat4f=/Niagara/DynamicInputs/Rotation/RandomQuaternion()
    Particles.Rando:NiagaraFloat=&FloatFromCurve(
     FloatCurve=DataInterface() {propertyValues {Curve {keys [
         {interpMode=RCIM_Cubic, tangentMode=RCTM_User, time=0, value=0}
         {arriveTangent=2.3793, interpMode=RCIM_Cubic, leaveTangent=2.3793, tangentMode=RCTM_User, time=1, value=1}
        ]}}}
     CurveIndex=&RandomRangeFloat(Minimum=0, Maximum=1)
    )
   ) {isSetParameters=true, module=SetVariables_#5}
   &InitializeParticle(
    Color=&MaskLinearColorBySpawnGroup(
     Masked Linear Color=&Multiply_LinearColorByFloat(Linear Color=$System.ElimColor, Float=2)
     Passthrough Linear Color=&MaskLinearColorBySpawnGroup(
      Masked Linear Color=&Multiply_LinearColorByFloat(Linear Color=$System.ElimColor, Float=2)
      Passthrough Linear Color=&MaskLinearColorBySpawnGroup(
       Masked Linear Color=&Multiply_LinearColorByFloat(Linear Color=$System.ElimColor, Float=2)
       Passthrough Linear Color=&MaskLinearColorBySpawnGroup(
        Masked Linear Color=&Multiply_LinearColorByFloat(Linear Color=$System.ElimColor, Float=2)
        Passthrough Linear Color=&MaskLinearColorBySpawnGroup(
         Masked Linear Color=&Multiply_LinearColorByFloat(Linear Color=$System.ElimColor, Float=0.4)
         Passthrough Linear Color=&MaskLinearColorBySpawnGroup(
          Masked Linear Color=&Multiply_LinearColorByFloat(Linear Color=$System.ElimColor, Float=0.4)
          Passthrough Linear Color=&Multiply_LinearColorByFloat(Linear Color=$System.ElimColor, Float=0.4)
          Spawn Group Mask=5
         )
         Spawn Group Mask=4
        )
        Spawn Group Mask=3
       )
       Spawn Group Mask=2
      )
      Spawn Group Mask=1
     )
    )
    Lifetime=&MaskFloatBySpawnGroup(
     Masked Float=0.4
     Passthrough Float=&MaskFloatBySpawnGroup(
      Masked Float=&RandomRangeFloat(Minimum=0.35, Maximum=0.4)
      Passthrough Float=&MaskFloatBySpawnGroup(
       Masked Float=1.2
       Passthrough Float=&MaskFloatBySpawnGroup(
        Masked Float=&RandomRangeFloat(Minimum=0.4, Maximum=0.7)
        Passthrough Float=&MaskFloatBySpawnGroup(
         Masked Float=&RandomRangeFloat(Minimum=0.2, Maximum=0.3)
         Passthrough Float=&MaskFloatBySpawnGroup(
          Masked Float=&RandomRangeFloat(Minimum=0.175, Maximum=0.35)
          Passthrough Float=&RandomRangeFloat(Minimum=0.2, Maximum=0.3)
          Spawn Group Mask=5
         )
         Spawn Group Mask=4
        )
        Spawn Group Mask=3
       )
       Spawn Group Mask=2
      )
      Spawn Group Mask=1
     )
    )
    Mass Mode=::NewEnumerator3
    Mesh Scale=&MaskVectorBySpawnGroup(
     Masked Vector=<2.5 2.5 2.5>
     Passthrough Vector=&MaskVectorBySpawnGroup(
      Masked Vector=<2.5 2.5 2.5>
      Passthrough Vector=&MaskVectorBySpawnGroup(
       Masked Vector=<2.5 2.5 2.5>
       Passthrough Vector=&MaskVectorBySpawnGroup(
        Masked Vector=<2.5 2.5 2.5>
        Passthrough Vector=&MaskVectorBySpawnGroup(
         Masked Vector=&RandomRangeVector(Minimum=<0.01 0.01 0.01>, Maximum=<0.03 0.03 0.03>)
         Passthrough Vector=&MaskVectorBySpawnGroup(
          Masked Vector=&RandomRangeVector(Minimum=<1.2 1.2 1.5>, Maximum=<1.4 1.4 2.2>)
          Passthrough Vector=<0.4 0.4 0.4>
          Spawn Group Mask=5
         )
         Spawn Group Mask=4
        )
        Spawn Group Mask=3
       )
       Spawn Group Mask=2
      )
      Spawn Group Mask=1
     )
    )
    Mesh Scale Mode=::NewEnumerator4
    Position Offset=&MaskVectorBySpawnGroup(
     Masked Vector=<0 0 0>
     Passthrough Vector=&MaskVectorBySpawnGroup(
      Masked Vector=<0 0 0>
      Passthrough Vector=&MaskVectorBySpawnGroup(
       Masked Vector=<0 0 0>
       Passthrough Vector=&MaskVectorBySpawnGroup(
        Masked Vector=<0 0 0>
        Passthrough Vector=&MaskVectorBySpawnGroup(
         Masked Vector=<0 0 0>
         Passthrough Vector=&MaskVectorBySpawnGroup(
          Masked Vector=&RandomRangeVector(Minimum=<0 0 15>, Maximum=<0 0 25>)
          Passthrough Vector=<0 0 0>
          Spawn Group Mask=5
         )
         Spawn Group Mask=4
        )
        Spawn Group Mask=3
       )
       Spawn Group Mask=2
      )
      Spawn Group Mask=1
     )
    )
    Sprite Size Mode=::NewEnumerator3
    Uniform Sprite Size=250
    UsePositionOffset=true
    Write Material Random=true
   )
   &ShapeLocation(
    Box Size=&MaskVectorBySpawnGroup(
     Masked Vector=<0 0 0>
     Passthrough Vector=&MaskVectorBySpawnGroup(
      Masked Vector=<0 0 0>
      Passthrough Vector=&MaskVectorBySpawnGroup(
       Masked Vector=<0 0 0>
       Passthrough Vector=&MaskVectorBySpawnGroup(
        Masked Vector=<75 75 5>
        Passthrough Vector=&MaskVectorBySpawnGroup(
         Masked Vector=<0 0 0>
         Passthrough Vector=&MaskVectorBySpawnGroup(Masked Vector=<0 0 0>, Passthrough Vector=<75 75 5>, Spawn Group Mask=5)
         Spawn Group Mask=4
        )
        Spawn Group Mask=3
       )
       Spawn Group Mask=2
      )
      Spawn Group Mask=1
     )
    )
    MaskSpawnGroup=false
    Override Seed=false
    Shape Primitive=::NewEnumerator2
    Surface Only=false
    UseOwnerScale=true
   )
   &ShapeLocation(
    MaskSpawnGroup=false
    Override Seed=false
    Sphere Radius=&MaskFloatBySpawnGroup(
     Masked Float=0
     Passthrough Float=&MaskFloatBySpawnGroup(
      Masked Float=0
      Passthrough Float=&MaskFloatBySpawnGroup(
       Masked Float=0
       Passthrough Float=&MaskFloatBySpawnGroup(
        Masked Float=0
        Passthrough Float=&MaskFloatBySpawnGroup(
         Masked Float=10
         Passthrough Float=&MaskFloatBySpawnGroup(Masked Float=0, Passthrough Float=0, Spawn Group Mask=5)
         Spawn Group Mask=4
        )
        Spawn Group Mask=3
       )
       Spawn Group Mask=2
      )
      Spawn Group Mask=1
     )
    )
    UseOwnerScale=true
   ) {module=ShapeLocation001}
   &InitialMeshOrientation(
    Orientation Axis=&MaskVectorBySpawnGroup(
     Masked Vector=<0 0 1>
     Passthrough Vector=&MaskVectorBySpawnGroup(
      Masked Vector=<0 0 1>
      Passthrough Vector=&MaskVectorBySpawnGroup(
       Masked Vector=<0 0 1>
       Passthrough Vector=&MaskVectorBySpawnGroup(
        Masked Vector=<0 0 0>
        Passthrough Vector=&MaskVectorBySpawnGroup(
         Masked Vector=<1 0 0>
         Passthrough Vector=&MaskVectorBySpawnGroup(Masked Vector=<0 0 1>, Passthrough Vector=<1 1 0>, Spawn Group Mask=5)
         Spawn Group Mask=4
        )
        Spawn Group Mask=3
       )
       Spawn Group Mask=2
      )
      Spawn Group Mask=1
     )
    )
    Orientation Vector=&MaskVectorBySpawnGroup(
     Masked Vector=<0 0 1>
     Passthrough Vector=&MaskVectorBySpawnGroup(
      Masked Vector=<0 0 1>
      Passthrough Vector=&MaskVectorBySpawnGroup(
       Masked Vector=<0 0 1>
       Passthrough Vector=&MaskVectorBySpawnGroup(
        Masked Vector=<0 0 0>
        Passthrough Vector=&MaskVectorBySpawnGroup(
         Masked Vector=<1 0 0>
         Passthrough Vector=&MaskVectorBySpawnGroup(Masked Vector=<1 0 0>, Passthrough Vector=<1 0 0>, Spawn Group Mask=5)
         Spawn Group Mask=4
        )
        Spawn Group Mask=3
       )
       Spawn Group Mask=2
      )
      Spawn Group Mask=1
     )
    )
    Rotation=&MaskVectorBySpawnGroup(
     Masked Vector=<0 0.25 0>
     Passthrough Vector=&MaskVectorBySpawnGroup(
      Masked Vector=<0 0.25 0>
      Passthrough Vector=&MaskVectorBySpawnGroup(
       Masked Vector=<0 0.25 0>
       Passthrough Vector=&MaskVectorBySpawnGroup(
        Masked Vector=<0 0 0>
        Passthrough Vector=&MaskVectorBySpawnGroup(
         Masked Vector=&RandomRangeVector(Minimum=<-0.1 -0.1 0>, Maximum=<0.1 0.1 0>)
         Passthrough Vector=&MaskVectorBySpawnGroup(
          Masked Vector=&RandomRangeVector(Minimum=<0 0 0.1>, Maximum=<1 0 0.3>)
          Passthrough Vector=&RandomRangeVector(Minimum=<0 0 0>, Maximum=<1 1 1>)
          Spawn Group Mask=5
         )
         Spawn Group Mask=4
        )
        Spawn Group Mask=3
       )
       Spawn Group Mask=2
      )
      Spawn Group Mask=1
     )
    )
    Use Orientation Vector=true
    Use Rotation Vector=true
   )
   /SetVariables(Particles.SpriteAlignment=<1 0 0.4>) {isSetParameters=true, module=SetVariables_#6}
   /SetVariables(
    Particles.GroundPulseCO=10
    Particles.GroundPulseDMP:Vector4f=&MakeVector4(X=0.6, Y=-0.05, Z=0, W=0)
    Particles.GroundPulseOrientation:Quat4f=<0 1 0 1>
   ) {isSetParameters=true, module=SetVariables_#7}
   /SetVariables(Particles.GroundZapCO=15) {isSetParameters=true, module=SetVariables_#8}
   &CameraOffset(Camera Offset Amount=&MaskFloatBySpawnGroup(
     Masked Float=5
     Passthrough Float=&MaskFloatBySpawnGroup(Masked Float=0, Passthrough Float=10, Spawn Group Mask=5)
     Spawn Group Mask=4
    )) {module=CameraOffset001}
  ]
  particleUpdateScript [
   &ParticleState()
   &ScaleMeshSize(Scale Factor=&MaskVectorBySpawnGroup(
     Masked Vector=<1 1 1>
     Passthrough Vector=&MaskVectorBySpawnGroup(
      Masked Vector=<1 1 1>
      Passthrough Vector=&MaskVectorBySpawnGroup(
       Masked Vector=<1 1 1>
       Passthrough Vector=&MaskVectorBySpawnGroup(
        Masked Vector=<1 1 1>
        Passthrough Vector=&MaskVectorBySpawnGroup(
         Masked Vector=&MakeVector(
          X=2
          Y=8
          Z=&FloatFromCurve(FloatCurve=DataInterface() {propertyValues {Curve {keys [
               {interpMode=RCIM_Cubic, leaveTangent=0.0966, tangentMode=RCTM_Break, time=0, value=1}
               {arriveTangent=0.0951, interpMode=RCIM_Cubic, leaveTangent=-0.1745, tangentMode=RCTM_Break, time=1, value=1.07}
              ]}}})
         )
         Passthrough Vector=&MaskVectorBySpawnGroup(Masked Vector=<1 1 1>, Passthrough Vector=<1 1 1>, Spawn Group Mask=5)
         Spawn Group Mask=4
        )
        Spawn Group Mask=3
       )
       Spawn Group Mask=2
      )
      Spawn Group Mask=1
     )
    ))
   &ScaleMeshSize(Scale Factor=&MaskVectorBySpawnGroup(
     Masked Vector=<1 1 1>
     Passthrough Vector=&MaskVectorBySpawnGroup(
      Masked Vector=<1 1 1>
      Passthrough Vector=&MaskVectorBySpawnGroup(
       Masked Vector=<1 1 1>
       Passthrough Vector=&MaskVectorBySpawnGroup(
        Masked Vector=<1 1 1>
        Passthrough Vector=&MaskVectorBySpawnGroup(
         Masked Vector=&RandomRangeVector(Minimum=<3 3 30>, Maximum=<4 4 60>)
         Passthrough Vector=&MaskVectorBySpawnGroup(
          Masked Vector=&Multiply_VectorByFloat(Vector=<1 1 1>, Float=0.22)
          Passthrough Vector=&Multiply_VectorByFloat(
           Vector=&RandomRangeVector(Minimum=<1 1 2.2>, Maximum=<2 2 2.2>)
           Float=&FloatFromCurve(
            FloatCurve=DataInterface() {propertyValues {Curve {keys [<at 0 1>,<at 1 1.5>]}}}
            Scale Curve=&RandomRangeFloat(Minimum=0.05, Maximum=0.07)
           )
          )
          Spawn Group Mask=5
         )
         Spawn Group Mask=4
        )
        Spawn Group Mask=3
       )
       Spawn Group Mask=2
      )
      Spawn Group Mask=1
     )
    )) {module=ScaleMeshSize001}
   &UpdateMeshOrientation(
    Rotation Rate=&MaskFloatBySpawnGroup(
     Masked Float=0
     Passthrough Float=&MaskFloatBySpawnGroup(
      Masked Float=0
      Passthrough Float=&MaskFloatBySpawnGroup(
       Masked Float=0
       Passthrough Float=&MaskFloatBySpawnGroup(
        Masked Float=0
        Passthrough Float=&MaskFloatBySpawnGroup(
         Masked Float=0
         Passthrough Float=&MaskFloatBySpawnGroup(Masked Float=&RandomRangeFloat(Minimum=-1, Maximum=1), Passthrough Float=0, Spawn Group Mask=5)
         Spawn Group Mask=4
        )
        Spawn Group Mask=3
       )
       Spawn Group Mask=2
      )
      Spawn Group Mask=1
     )
    )
    Rotation Vector=&MaskVectorBySpawnGroup(
     Masked Vector=<0 0 0>
     Passthrough Vector=&MaskVectorBySpawnGroup(
      Masked Vector=<0 0 0>
      Passthrough Vector=&MaskVectorBySpawnGroup(
       Masked Vector=<0 0 0>
       Passthrough Vector=&MaskVectorBySpawnGroup(
        Masked Vector=<0 0 0>
        Passthrough Vector=&MaskVectorBySpawnGroup(
         Masked Vector=<0 0 0>
         Passthrough Vector=&MaskVectorBySpawnGroup(Masked Vector=<0 0 1>, Passthrough Vector=<0 0 0>, Spawn Group Mask=5)
         Spawn Group Mask=4
        )
        Spawn Group Mask=3
       )
       Spawn Group Mask=2
      )
      Spawn Group Mask=1
     )
    )
   )
   &SpriteRotationRate(Rotation Rate=&FloatFromCurve(
     FloatCurve=DataInterface() {propertyValues {Curve {keys [
         {arriveTangent=1.4306, interpMode=RCIM_Cubic, leaveTangent=1.4306, tangentMode=RCTM_User, time=0.2, value=0}
         {arriveTangent=1.1179, interpMode=RCIM_Cubic, leaveTangent=1.1179, tangentMode=RCTM_User, time=0.4, value=0.6228}
         {arriveTangent=0.0224, interpMode=RCIM_Cubic, leaveTangent=0.0224, tangentMode=RCTM_User, time=1, value=1}
        ]}}}
     Scale Curve=0
    ))
   &ScaleColor(
    Scale Alpha=&MaskFloatBySpawnGroup(
     Masked Float=&FloatFromCurve(FloatCurve=DataInterface() {propertyValues {Curve {keys [
          {arriveTangent=-1.5276, interpMode=RCIM_Cubic, leaveTangent=-1.5276, tangentMode=RCTM_User, time=0, value=1}
          {interpMode=RCIM_Cubic, tangentMode=RCTM_User, time=1, value=0}
         ]}}})
     Passthrough Float=&MaskFloatBySpawnGroup(
      Masked Float=&FloatFromCurve(FloatCurve=DataInterface() {propertyValues {Curve {keys [
           {arriveTangent=-1.5276, interpMode=RCIM_Cubic, leaveTangent=-1.5276, tangentMode=RCTM_User, time=0, value=1}
           {interpMode=RCIM_Cubic, tangentMode=RCTM_User, time=1, value=0}
          ]}}})
      Passthrough Float=&MaskFloatBySpawnGroup(
       Masked Float=&FloatFromCurve(FloatCurve=DataInterface() {propertyValues {Curve {keys [
            {arriveTangent=-1.5276, interpMode=RCIM_Cubic, leaveTangent=-1.5276, tangentMode=RCTM_User, time=0, value=1}
            {interpMode=RCIM_Cubic, tangentMode=RCTM_User, time=1, value=0}
           ]}}})
       Passthrough Float=&MaskFloatBySpawnGroup(
        Masked Float=&FloatFromCurve(FloatCurve=DataInterface() {propertyValues {Curve {keys [
             {arriveTangent=-1.5276, interpMode=RCIM_Cubic, leaveTangent=-1.5276, tangentMode=RCTM_User, time=0, value=1}
             {interpMode=RCIM_Cubic, tangentMode=RCTM_User, time=1, value=0}
            ]}}})
        Passthrough Float=&MaskFloatBySpawnGroup(
         Masked Float=&FloatFromCurve(FloatCurve=DataInterface() {propertyValues {Curve {keys [
              {tangentMode=RCTM_User, time=0, value=1}
              {arriveTangent=-0.0405, interpMode=RCIM_Cubic, leaveTangent=-0.0405, tangentMode=RCTM_User, time=0.7009, value=1}
              {arriveTangent=-8.0089, interpMode=RCIM_Cubic, leaveTangent=-8.0089, tangentMode=RCTM_User, time=1, value=0}
             ]}}})
         Passthrough Float=&MaskFloatBySpawnGroup(
          Masked Float=&FloatFromCurve(FloatCurve=DataInterface() {propertyValues {Curve {keys [
               {tangentMode=RCTM_User, time=0, value=1}
               {arriveTangent=-0.0405, interpMode=RCIM_Cubic, leaveTangent=-0.0405, tangentMode=RCTM_User, time=0.7009, value=1}
               {arriveTangent=-8.0089, interpMode=RCIM_Cubic, leaveTangent=-8.0089, tangentMode=RCTM_User, time=1, value=0}
              ]}}})
          Passthrough Float=2
          Spawn Group Mask=5
         )
         Spawn Group Mask=4
        )
        Spawn Group Mask=3
       )
       Spawn Group Mask=2
      )
      Spawn Group Mask=1
     )
    )
    Scale RGB=&MaskVectorBySpawnGroup(
     Masked Vector=<1 1 1>
     Passthrough Vector=&MaskVectorBySpawnGroup(
      Masked Vector=<1 1 1>
      Passthrough Vector=&MaskVectorBySpawnGroup(
       Masked Vector=<1 1 1>
       Passthrough Vector=&MaskVectorBySpawnGroup(
        Masked Vector=<1 1 1>
        Passthrough Vector=&MaskVectorBySpawnGroup(
         Masked Vector=&Lerp_Vector(
          A=&VectorFromFloat(Value=&FloatFromCurve(
            FloatCurve=DataInterface() {propertyValues {Curve {keys [<at 0 1>,<at 0.1 1>,<at 0.2 2>,<at 0.3 1>,<at 0.6 1>,<at 0.7 2>,<at 0.8 1>,<at 1 1>]}}}
            Scale Curve=&MaskFloatBySpawnGroup(
             Masked Float=25
             Passthrough Float=&MaskFloatBySpawnGroup(Masked Float=10, Passthrough Float=25, Spawn Group Mask=5)
             Spawn Group Mask=4
            )
           ))
          B=&VectorFromFloat(Value=&RandomRangeFloat(Minimum=0.3, Maximum=5, Evaluation Type=::NewEnumerator1))
          Alpha=&FloatFromCurve(FloatCurve=DataInterface() {propertyValues {Curve {keys [
               {interpMode=RCIM_Cubic, leaveTangent=2.5294, tangentMode=RCTM_Break, time=0, value=0}
               {arriveTangent=-0.2166, interpMode=RCIM_Cubic, tangentMode=RCTM_Break, time=0.4, value=1}
              ]}}})
         )
         Passthrough Vector=&MaskVectorBySpawnGroup(
          Masked Vector=&Lerp_Vector(
           A=&VectorFromFloat(Value=&FloatFromCurve(
             FloatCurve=DataInterface() {propertyValues {Curve {keys [<at 0 1>,<at 0.1 1>,<at 0.2 2>,<at 0.3 1>,<at 0.6 1>,<at 0.7 2>,<at 0.8 1>,<at 1 1>]}}}
             Scale Curve=&MaskFloatBySpawnGroup(
              Masked Float=25
              Passthrough Float=&MaskFloatBySpawnGroup(Masked Float=10, Passthrough Float=25, Spawn Group Mask=5)
              Spawn Group Mask=4
             )
            ))
           B=&VectorFromFloat(Value=&RandomRangeFloat(Minimum=0.3, Maximum=5, Evaluation Type=::NewEnumerator1))
           Alpha=&FloatFromCurve(FloatCurve=DataInterface() {propertyValues {Curve {keys [
                {interpMode=RCIM_Cubic, leaveTangent=2.5294, tangentMode=RCTM_Break, time=0, value=0}
                {arriveTangent=-0.2166, interpMode=RCIM_Cubic, tangentMode=RCTM_Break, time=0.4, value=1}
               ]}}})
          )
          Passthrough Vector=&Lerp_Vector(
           A=&VectorFromFloat(Value=&FloatFromCurve(
             FloatCurve=DataInterface() {propertyValues {Curve {keys [<at 0 1>,<at 0.1 1>,<at 0.2 2>,<at 0.3 1>,<at 0.6 1>,<at 0.7 2>,<at 0.8 1>,<at 1 1>]}}}
             Scale Curve=&MaskFloatBySpawnGroup(
              Masked Float=25
              Passthrough Float=&MaskFloatBySpawnGroup(Masked Float=10, Passthrough Float=25, Spawn Group Mask=5)
              Spawn Group Mask=4
             )
            ))
           B=&VectorFromFloat(Value=&RandomRangeFloat(Minimum=0.3, Maximum=5, Evaluation Type=::NewEnumerator1))
           Alpha=&FloatFromCurve(FloatCurve=DataInterface() {propertyValues {Curve {keys [
                {interpMode=RCIM_Cubic, leaveTangent=2.5294, tangentMode=RCTM_Break, time=0, value=0}
                {arriveTangent=-0.2166, interpMode=RCIM_Cubic, tangentMode=RCTM_Break, time=0.4, value=1}
               ]}}})
          )
          Spawn Group Mask=5
         )
         Spawn Group Mask=4
        )
        Spawn Group Mask=3
       )
       Spawn Group Mask=2
      )
      Spawn Group Mask=1
     )
    )
    ScaleA=true
    ScaleRGB=true
   ) {module=ScaleColor001}
   &ScaleSpriteSize(
    Scale Sprite Size Mode=::NewEnumerator0
    Uniform Scale Factor=&FloatFromCurve(FloatCurve=DataInterface() {propertyValues {Curve {keys [
         {arriveTangent=2.278, interpMode=RCIM_Cubic, leaveTangent=2.278, tangentMode=RCTM_User, time=0, value=0}
         {interpMode=RCIM_Cubic, tangentMode=RCTM_User, time=1, value=1}
        ]}}})
   )
   /SetVariables(Particles.WaveDynMat:Vector4f=&MakeVector4(
     X=&FloatFromCurve(
      FloatCurve=DataInterface() {propertyValues {Curve {keys [{arriveTangent=3.726, interpMode=RCIM_Cubic, leaveTangent=3.726, tangentMode=RCTM_User, time=0, value=-0.3}, <at 1 1>]}}}
      Scale Curve=0.6
     )
     Y=&FloatFromCurve(FloatCurve=DataInterface() {propertyValues {Curve {keys [{arriveTangent=-0.5818, interpMode=RCIM_Cubic, leaveTangent=-0.5818, tangentMode=RCTM_User, time=0, value=1}, <at 1 0.8>]}}})
     Z=&FloatFromCurve(FloatCurve=DataInterface() {propertyValues {Curve {keys [<at 0 0>,<at 0.15 1>,<at 1 1.2>]}}}, Scale Curve=10)
     W=&FloatFromCurve(FloatCurve=DataInterface() {propertyValues {Curve {keys [<at 0 0>,<at 1 1>]}}}, Scale Curve=0)
    )) {isSetParameters=true, module=SetVariables_#9}
   /SetVariables(Particles.GlowSize=&Multiply_Vector2D(A=$Particles.SpriteSize, B=<3 3>)) {isSetParameters=true, module=SetVariables_#10}
   /SetVariables(Particles.GlowColor=&MultiplyLinearColor(Linear Color=$Particles.Color, Scale Factor=<1 1 1 0.08>)) {isSetParameters=true, module=SetVariables_#11}
   /SetVariables(
    Particles.ShockColor:LinearColor=&MultiplyLinearColor(
     Linear Color=&ColorFromCurve(ColorCurve=DataInterface() {propertyValues {
        AlphaCurve {keys [<at 0 1 RCIM_Cubic>]}
        BlueCurve {keys [<at 0 0 RCIM_Cubic>,<at 0.6348 0>]}
        GreenCurve {keys [<at 0 0.5 RCIM_Cubic>,<at 0.6348 0.01>]}
        RedCurve {keys [<at 0 20 RCIM_Cubic>,<at 0.6348 8>]}
       }})
     Scale Factor=&MakeVector4_FromFloat(Float=&FloatFromCurve(
       FloatCurve=DataInterface() {propertyValues {Curve {keys [
           {arriveTangent=-1.5276, interpMode=RCIM_Cubic, leaveTangent=-1.5276, tangentMode=RCTM_User, time=0, value=1}
           {interpMode=RCIM_Cubic, tangentMode=RCTM_User, time=1, value=0}
          ]}}}
       Scale Curve=0.5
      ))
    )
    Particles.ShockwaveScale:Vector3f=&MaskVectorBySpawnGroup(
     Masked Vector=&Multiply_VectorByFloat(
      Vector=$Particles.Initial.Scale
      Float=&FloatFromCurve(FloatCurve=DataInterface() {propertyValues {Curve {keys [
           {arriveTangent=1.8207, interpMode=RCIM_Cubic, leaveTangent=1.8207, tangentMode=RCTM_User, time=0, value=0.2}
           {interpMode=RCIM_Cubic, tangentMode=RCTM_User, time=1, value=1.2}
          ]}}})
     )
     Passthrough Vector=$Particles.Initial.Scale
    )
   ) {isSetParameters=true, module=SetVariables_#12}
   /SetVariables(
    Particles.BeamColor:LinearColor=&MultiplyLinearColor(
     Linear Color=&LerpLinearColors(StartColor=<rgba 3 0.6 0.6 0.3>, EndColor=<rgba 6 0 0 0.4>, LerpFactor=$Particles.Rando)
     Scale Factor=&MakeVector4_FromFloat(Float=&FloatFromCurve(
       FloatCurve=DataInterface() {propertyValues {Curve {keys [
           {arriveTangentWeight=0.0275, interpMode=RCIM_Cubic, leaveTangentWeight=0.3161, tangentMode=RCTM_User, time=0, value=1}
           {arriveTangent=-2.6497, interpMode=RCIM_Cubic, leaveTangent=-2.6497, tangentMode=RCTM_User, time=0.25, value=1}
           {arriveTangent=-0.0405, interpMode=RCIM_Cubic, leaveTangent=-0.0405, tangentMode=RCTM_User, time=1, value=0}
          ]}}}
       Scale Curve=0.25
      ))
    )
    Particles.BeamDMP:Vector4f=&MakeVector4(
     X=0.1
     Y=&FloatFromCurve(
      FloatCurve=DataInterface() {propertyValues {Curve {keys [
          {arriveTangent=5.1604, interpMode=RCIM_Cubic, leaveTangent=5.1604, tangentMode=RCTM_User, time=0, value=0}
          {interpMode=RCIM_Cubic, tangentMode=RCTM_User, time=0.5, value=1}
          <at 1 0 RCIM_Cubic>
         ]}}}
      Scale Curve=&Lerp_Float(A=1.75, B=1.5, Alpha=$Particles.Rando)
     )
     Z=0
     W=0
    )
    Particles.BeamScale:Vector3f=&Multiply_VectorByFloat(
     Vector=&VectorFromFloat(Value=&Lerp_Float(A=1.5, B=2.5, Alpha=$Particles.Rando))
     Float=&FloatFromCurve(
      FloatCurve=DataInterface() {propertyValues {Curve {keys [
          {arriveTangent=3.3092, interpMode=RCIM_Cubic, leaveTangent=3.3092, tangentMode=RCTM_User, time=0, value=0.4}
          {arriveTangent=0.048, interpMode=RCIM_Cubic, leaveTangent=0.048, tangentMode=RCTM_User, time=0.5, value=1}
          {arriveTangent=-2.6178, interpMode=RCIM_Cubic, leaveTangent=-2.6178, tangentMode=RCTM_User, time=1, value=0.5}
         ]}}}
      Scale Curve=0.75
     )
    )
   ) {isSetParameters=true, module=SetVariables_#13}
   /SetVariables(
    Particles.GroundPulseColor:LinearColor=&MultiplyLinearColor(
     Linear Color=&ColorFromCurve(ColorCurve=DataInterface() {propertyValues {
        AlphaCurve {keys [<at 0 0.2 RCIM_Cubic>,<at 1 0 RCIM_Cubic>]}
        BlueCurve {keys [<at 0 0 RCIM_Cubic>,<at 0.1757 0>,<at 1 0 RCIM_Cubic>]}
        GreenCurve {keys [<at 0 0 RCIM_Cubic>,<at 0.1757 0>,<at 1 0.0571 RCIM_Cubic>]}
        RedCurve {keys [<at 0 1 RCIM_Cubic>,<at 0.1757 0.6435>,<at 1 1 RCIM_Cubic>]}
       }})
     Scale Factor=&MakeVector4_FromFloat(Float=&FloatFromCurve(
       FloatCurve=DataInterface() {propertyValues {Curve {keys [
           {arriveTangent=7.3302, interpMode=RCIM_Cubic, leaveTangent=7.3302, tangentMode=RCTM_User, time=0, value=0}
           {arriveTangent=122.0485, interpMode=RCIM_Cubic, leaveTangent=-11.619, tangentMode=RCTM_Break, time=0.05, value=5}
           {arriveTangent=-0.6259, tangentMode=RCTM_User, time=1, value=0}
          ]}}}
       Scale Curve=0.75
      ))
    )
    Particles.GroundPulseScale:Vector3f=&Multiply_VectorByFloat(
     Vector=<0.02 0.6 0.6>
     Float=&FloatFromCurve(
      FloatCurve=DataInterface() {propertyValues {Curve {keys [
          {arriveTangent=4.1785, interpMode=RCIM_Cubic, leaveTangent=4.1785, tangentMode=RCTM_User, time=0, value=0}
          {arriveTangent=1.1179, interpMode=RCIM_Cubic, leaveTangent=1.1179, tangentMode=RCTM_User, time=0.2042, value=0.6228}
          {arriveTangent=-0.2023, interpMode=RCIM_Cubic, leaveTangent=-0.2023, tangentMode=RCTM_User, time=0.9586, value=1}
          {arriveTangent=0.0224, interpMode=RCIM_Cubic, leaveTangent=0.0224, tangentMode=RCTM_User, time=1, value=0}
         ]}}}
      Scale Curve=1.2
     )
    )
   ) {isSetParameters=true, module=SetVariables_#14}
   /SetVariables(
    Particles.GroundZapColor:LinearColor=&Multiply_LinearColorByFloat(
     Linear Color=$System.ElimColor
     Float=&FloatFromCurve(
      FloatCurve=DataInterface() {propertyValues {Curve {keys [
          {tangentMode=RCTM_User, time=0, value=0}
          <at 0.1 1.2>
          <at 0.2 0.7>
          <at 0.6 0.6>
          <at 0.7 0.875>
          <at 0.8 0.35>
          {tangentMode=RCTM_User, time=1, value=0}
         ]}}}
      Scale Curve=&RandomRangeFloat(Minimum=0.2, Maximum=0.4)
     )
    )
    Particles.GroundZapSize:Vector2f=&Multiply_Vector2D_ByFloat(
     Float=&FloatFromCurve(FloatCurve=DataInterface() {propertyValues {Curve {keys [
          <at 0 0 RCIM_Cubic>
          <at 0.112 1.0173>
          {arriveTangent=-1.142, interpMode=RCIM_Cubic, leaveTangent=-1.142, time=0.7001, value=0.872}
          <at 1.0011 0.002>
         ]}}})
     Vector2D=<275 125>
    )
   ) {isSetParameters=true, module=SetVariables_#15}
   &FN_Light_Attributes(
    Color=<rgba 1 0 0 0.7>
    Create Color Attribute=true
    Create Exponent Attribute=true
    Create Radius Attribute=true
    Exponent=1
    Radius=280
    Scale Color=&FloatFromCurve(FloatCurve=DataInterface() {propertyValues {Curve {keys [<at 0 1>,<at 1 0>]}}}, Scale Curve=10)
   )
   &CameraOffset(Camera Offset Amount=&MaskFloatBySpawnGroup(
     Masked Float=&Lerp_Float(
      A=30
      B=-5
      Alpha=&FloatFromCurve(FloatCurve=DataInterface() {propertyValues {Curve {keys [<at 0 0>,<at 0.25 1>]}}}, CurveIndex=$Engine.System.Age)
     )
     Passthrough Float=&MaskFloatBySpawnGroup(
      Masked Float=&Lerp_Float(
       A=30
       B=-5
       Alpha=&FloatFromCurve(FloatCurve=DataInterface() {propertyValues {Curve {keys [<at 0 0>,<at 0.25 1>]}}}, CurveIndex=$Engine.System.Age)
      )
      Passthrough Float=&MaskFloatBySpawnGroup(
       Masked Float=&Lerp_Float(
        A=30
        B=-5
        Alpha=&FloatFromCurve(FloatCurve=DataInterface() {propertyValues {Curve {keys [<at 0 0>,<at 0.25 1>]}}}, CurveIndex=$Engine.System.Age)
       )
       Passthrough Float=&MaskFloatBySpawnGroup(
        Masked Float=&Lerp_Float(
         A=30
         B=-5
         Alpha=&FloatFromCurve(FloatCurve=DataInterface() {propertyValues {Curve {keys [<at 0 0>,<at 0.25 1>]}}}, CurveIndex=$Engine.System.Age)
        )
        Passthrough Float=&MaskFloatBySpawnGroup(
         Masked Float=0
         Passthrough Float=&MaskFloatBySpawnGroup(Masked Float=0, Passthrough Float=0, Spawn Group Mask=5)
         Spawn Group Mask=4
        )
        Spawn Group Mask=3
       )
       Spawn Group Mask=2
      )
      Spawn Group Mask=1
     )
    ))
   &DynamicMaterialParameters(
    Index 0 Param 1=&MakeFloatFromVector(Vector=$Particles.Scale, Channel=::NewEnumerator0)
    Index 0 Param 2=&RandomRangeFloat(
     Minimum=0
     Maximum=&MaskFloatBySpawnGroup(
      Masked Float=10
      Passthrough Float=&MaskFloatBySpawnGroup(Masked Float=10, Passthrough Float=35, Spawn Group Mask=5)
      Spawn Group Mask=4
     )
    )
    Index 0 Param 3=&FloatFromCurve(
     FloatCurve=DataInterface() {propertyValues {Curve {keys [{interpMode=RCIM_Cubic, leaveTangent=8.1085, tangentMode=RCTM_Break, time=0.8568, value=1}, {tangentMode=RCTM_User, time=1, value=0}]}}}
     Scale Curve=0.07
    )
    Index 0 Param 4=&MaskFloatBySpawnGroup(
     Masked Float=0.8
     Passthrough Float=&MaskFloatBySpawnGroup(Masked Float=0.8, Passthrough Float=&RandomRangeFloat(Minimum=0.8, Maximum=1), Spawn Group Mask=5)
     Spawn Group Mask=4
    )
    Index 1 Param 1=&FloatFromCurve(FloatCurve=DataInterface() {propertyValues {Curve {keys [
         {interpMode=RCIM_Cubic, leaveTangent=1.1034, tangentMode=RCTM_Break, time=0, value=0}
         {arriveTangent=2.2399, tangentMode=RCTM_User, time=0.5, value=1}
        ]}}})
    Index 1 Param 2=&FloatFromCurve(FloatCurve=DataInterface() {propertyValues {Curve {keys [<at 0.5 0>, {arriveTangent=3.2053, tangentMode=RCTM_User, time=0.9, value=1}]}}})
    Index 1 Param 3=&FloatFromCurve(FloatCurve=DataInterface() {propertyValues {Curve {keys [<at 0 0 RCIM_Cubic>, {arriveTangent=3.4223, interpMode=RCIM_Cubic, leaveTangent=3.4223, tangentMode=RCTM_User, time=1, value=1}]}}})
    Index 1 Param 4=&FloatFromCurve(
     FloatCurve=DataInterface() {propertyValues {Curve {keys [
         {interpMode=RCIM_Cubic, leaveTangent=0.1169, tangentMode=RCTM_Break, time=0, value=0}
         {arriveTangent=2.5708, interpMode=RCIM_Cubic, leaveTangent=2.5708, tangentMode=RCTM_User, time=1, value=1}
        ]}}}
     Scale Curve=&RandomRangeFloat(
      Minimum=1
      Maximum=&MaskFloatBySpawnGroup(
       Masked Float=15
       Passthrough Float=&MaskFloatBySpawnGroup(Masked Float=40, Passthrough Float=5, Spawn Group Mask=5)
       Spawn Group Mask=4
      )
     )
    )
    Index 2 Param 1=&FloatFromCurve(
     FloatCurve=DataInterface() {propertyValues {Curve {keys [
         {interpMode=RCIM_Cubic, leaveTangent=-0.0025, tangentMode=RCTM_Break, time=0, value=0.5}
         {arriveTangent=-0.0316, interpMode=RCIM_Cubic, tangentMode=RCTM_Break, time=1, value=-0.2}
        ]}}}
     Scale Curve=&MaskFloatBySpawnGroup(
      Masked Float=1
      Passthrough Float=&MaskFloatBySpawnGroup(Masked Float=1, Passthrough Float=-0.5, Spawn Group Mask=5)
      Spawn Group Mask=4
     )
    )
    Index 2 Param 2=&FloatFromCurve(
     FloatCurve=DataInterface() {propertyValues {Curve {keys [
         {interpMode=RCIM_Cubic, tangentMode=RCTM_User, time=0, value=0}
         {arriveTangent=2.0522, interpMode=RCIM_Cubic, leaveTangent=2.0522, tangentMode=RCTM_User, time=1, value=1}
        ]}}}
     Scale Curve=&MaskFloatBySpawnGroup(
      Masked Float=2
      Passthrough Float=&MaskFloatBySpawnGroup(Masked Float=2, Passthrough Float=1, Spawn Group Mask=5)
      Spawn Group Mask=4
     )
    )
    Module.Index0Param0Write=true
    Module.Index0Param1Write=true
    Module.Index0Param2Write=true
    Module.Index0Param3Write=true
    Module.Index1Param0Write=true
    Module.Index1Param1Write=true
    Module.Index1Param2Write=true
    Module.Index1Param3Write=true
    Module.Index2Param0Write=true
    Module.Index2Param1Write=true
    Module.Index2Param2Write=false
    Module.Index2Param3Write=false
    Module.Index3Param0Write=false
    Module.Index3Param1Write=false
    Module.Index3Param2Write=false
    Module.Index3Param3Write=true
    Write Parameter Index 1=true
    Write Parameter Index 2=true
    Write Parameter Index 3=true
   ) {module=DynamicMaterialParameters001}
  ]
  rendererClasses [
   &NiagaraSpriteRendererProperties
   &NiagaraLightRendererProperties
   &NiagaraMeshRendererProperties
  ]
  renderers [
   {rendererClass=&NiagaraSpriteRendererProperties}
   {rendererClass=&NiagaraSpriteRendererProperties, rendererIndex=1}
   {rendererClass=&NiagaraLightRendererProperties, rendererIndex=2}
   {rendererClass=&NiagaraMeshRendererProperties, rendererIndex=3}
   {rendererClass=&NiagaraMeshRendererProperties, rendererIndex=4}
   {rendererClass=&NiagaraMeshRendererProperties, rendererIndex=5}
   {rendererClass=&NiagaraSpriteRendererProperties, rendererIndex=6}
   {rendererClass=&NiagaraMeshRendererProperties, rendererIndex=7}
   {rendererClass=&NiagaraMeshRendererProperties, rendererIndex=8}
   {rendererClass=&NiagaraMeshRendererProperties, rendererIndex=9}
   {rendererClass=&NiagaraMeshRendererProperties, rendererIndex=10}
   {rendererClass=&NiagaraMeshRendererProperties, rendererIndex=11}
   {rendererClass=&NiagaraMeshRendererProperties, rendererIndex=12}
  ]
 }
 CentralBlast001 {
  bEnabled=false
  emitterUpdateScript [
   &EmitterState()
   &SpawnBurst_Instantaneous(Use Spawn Probability=false, UseLoopCountLimit=false)
   &SpawnBurst_Instantaneous(Spawn Count=25, Spawn Group=1, Use Spawn Probability=false, UseLoopCountLimit=false) {module=SpawnBurst_Instantaneous001}
   &SpawnBurst_Instantaneous(Spawn Count=2, Spawn Group=2, Use Spawn Probability=false, UseLoopCountLimit=false) {module=SpawnBurst_Instantaneous002}
   &SpawnRate(
    SpawnGroup=3
    SpawnRate=&Multiply_Float(A=&RandomRangeFloat(Minimum=5, Maximum=6), B=$System.FadeOutAura)
    Use Spawn Probability=false
   )
   &SpawnBurst_Instantaneous(
    Spawn Count=&RandomRangeInt(Minimum=4, Maximum=7)
    Spawn Group=4
    Use Spawn Probability=false
    UseLoopCountLimit=false
   ) {module=SpawnBurst_Instantaneous005}
   &SpawnBurst_Instantaneous(Spawn Count=8, Spawn Group=5, Use Spawn Probability=false, UseLoopCountLimit=false) {module=SpawnBurst_Instantaneous003}
   &SpawnRate(SpawnGroup=5, SpawnRate=&Multiply_Float(A=9, B=$System.FadeOutBottomLightning), Use Spawn Probability=false) {module=SpawnRate001}
   &SpawnBurst_Instantaneous(Spawn Count=4, Spawn Group=6, Use Spawn Probability=false, UseLoopCountLimit=false) {module=SpawnBurst_Instantaneous004}
   &SpawnRate(SpawnGroup=6, SpawnRate=&Multiply_Float(A=15, B=$System.FadeOutAura), Use Spawn Probability=false) {module=SpawnRate002}
  ]
  particleSpawnScript [
   &PartitionParticles(Number Of Partitions=&MaskIntBySpawnGroup(
     Masked Int=0
     Passthrough Int=&MaskIntBySpawnGroup(
      Masked Int=0
      Passthrough Int=&MaskIntBySpawnGroup(
       Masked Int=1
       Passthrough Int=&MaskIntBySpawnGroup(
        Masked Int=0
        Passthrough Int=&MaskIntBySpawnGroup(
         Masked Int=0
         Passthrough Int=&MaskIntBySpawnGroup(Masked Int=0, Passthrough Int=0, Spawn Group Mask=5)
         Spawn Group Mask=4
        )
        Spawn Group Mask=3
       )
       Spawn Group Mask=2
      )
      Spawn Group Mask=1
     )
     Spawn Group Mask=0
    ))
   /SetVariables(Particles.VisibilityTag=&SpawnGroup()) {isSetParameters=true, module=SetVariables_#4}
   /SetVariables(
    Particles.BeamOrientation:Quat4f=&RandomQuaternion()
    Particles.Rando:NiagaraFloat=&FloatFromCurve(
     FloatCurve=DataInterface() {propertyValues {Curve {keys [
         {interpMode=RCIM_Cubic, tangentMode=RCTM_User, time=0, value=0}
         {arriveTangent=2.3793, interpMode=RCIM_Cubic, leaveTangent=2.3793, tangentMode=RCTM_User, time=1, value=1}
        ]}}}
     CurveIndex=&RandomRangeFloat(Minimum=0, Maximum=1)
    )
   ) {isSetParameters=true, module=SetVariables_#5}
   &InitializeParticle(
    Color=&MaskLinearColorBySpawnGroup(
     Masked Linear Color=&Multiply_LinearColorByFloat(Linear Color=$System.ElimColor, Float=2)
     Passthrough Linear Color=&MaskLinearColorBySpawnGroup(
      Masked Linear Color=&Multiply_LinearColorByFloat(Linear Color=$System.ElimColor, Float=2)
      Passthrough Linear Color=&MaskLinearColorBySpawnGroup(
       Masked Linear Color=&Multiply_LinearColorByFloat(Linear Color=$System.ElimColor, Float=2)
       Passthrough Linear Color=&MaskLinearColorBySpawnGroup(
        Masked Linear Color=&Multiply_LinearColorByFloat(Linear Color=$System.ElimColor, Float=2)
        Passthrough Linear Color=&MaskLinearColorBySpawnGroup(
         Masked Linear Color=&Multiply_LinearColorByFloat(Linear Color=$System.ElimColor, Float=0.4)
         Passthrough Linear Color=&MaskLinearColorBySpawnGroup(
          Masked Linear Color=&Multiply_LinearColorByFloat(Linear Color=$System.ElimColor, Float=0.4)
          Passthrough Linear Color=&Multiply_LinearColorByFloat(Linear Color=$System.ElimColor, Float=0.4)
          Spawn Group Mask=5
         )
         Spawn Group Mask=4
        )
        Spawn Group Mask=3
       )
       Spawn Group Mask=2
      )
      Spawn Group Mask=1
     )
    )
    Lifetime=&MaskFloatBySpawnGroup(
     Masked Float=0.4
     Passthrough Float=&MaskFloatBySpawnGroup(
      Masked Float=&RandomRangeFloat(Minimum=0.35, Maximum=0.4)
      Passthrough Float=&MaskFloatBySpawnGroup(
       Masked Float=1.2
       Passthrough Float=&MaskFloatBySpawnGroup(
        Masked Float=&RandomRangeFloat(Minimum=0.4, Maximum=0.7)
        Passthrough Float=&MaskFloatBySpawnGroup(
         Masked Float=&RandomRangeFloat(Minimum=0.2, Maximum=0.3)
         Passthrough Float=&MaskFloatBySpawnGroup(
          Masked Float=&RandomRangeFloat(Minimum=0.175, Maximum=0.35)
          Passthrough Float=&RandomRangeFloat(Minimum=0.2, Maximum=0.3)
          Spawn Group Mask=5
         )
         Spawn Group Mask=4
        )
        Spawn Group Mask=3
       )
       Spawn Group Mask=2
      )
      Spawn Group Mask=1
     )
    )
    Mass Mode=::NewEnumerator3
    Mesh Scale=&MaskVectorBySpawnGroup(
     Masked Vector=<2.5 2.5 2.5>
     Passthrough Vector=&MaskVectorBySpawnGroup(
      Masked Vector=<2.5 2.5 2.5>
      Passthrough Vector=&MaskVectorBySpawnGroup(
       Masked Vector=<2.5 2.5 2.5>
       Passthrough Vector=&MaskVectorBySpawnGroup(
        Masked Vector=<2.5 2.5 2.5>
        Passthrough Vector=&MaskVectorBySpawnGroup(
         Masked Vector=&RandomRangeVector(Minimum=<0.01 0.01 0.01>, Maximum=<0.03 0.03 0.03>)
         Passthrough Vector=&MaskVectorBySpawnGroup(
          Masked Vector=&RandomRangeVector(Minimum=<1.2 1.2 1.5>, Maximum=<1.4 1.4 2.2>)
          Passthrough Vector=<0.4 0.4 0.4>
          Spawn Group Mask=5
         )
         Spawn Group Mask=4
        )
        Spawn Group Mask=3
       )
       Spawn Group Mask=2
      )
      Spawn Group Mask=1
     )
    )
    Mesh Scale Mode=::NewEnumerator4
    Position Offset=&MaskVectorBySpawnGroup(
     Masked Vector=<0 0 0>
     Passthrough Vector=&MaskVectorBySpawnGroup(
      Masked Vector=<0 0 0>
      Passthrough Vector=&MaskVectorBySpawnGroup(
       Masked Vector=<0 0 0>
       Passthrough Vector=&MaskVectorBySpawnGroup(
        Masked Vector=<0 0 0>
        Passthrough Vector=&MaskVectorBySpawnGroup(
         Masked Vector=<0 0 0>
         Passthrough Vector=&MaskVectorBySpawnGroup(
          Masked Vector=&RandomRangeVector(Minimum=<0 0 15>, Maximum=<0 0 25>)
          Passthrough Vector=<0 0 0>
          Spawn Group Mask=5
         )
         Spawn Group Mask=4
        )
        Spawn Group Mask=3
       )
       Spawn Group Mask=2
      )
      Spawn Group Mask=1
     )
    )
    Sprite Size Mode=::NewEnumerator3
    Uniform Sprite Size=250
    UsePositionOffset=true
    Write Material Random=true
   )
   &ShapeLocation(
    Box Size=&MaskVectorBySpawnGroup(
     Masked Vector=<0 0 0>
     Passthrough Vector=&MaskVectorBySpawnGroup(
      Masked Vector=<0 0 0>
      Passthrough Vector=&MaskVectorBySpawnGroup(
       Masked Vector=<0 0 0>
       Passthrough Vector=&MaskVectorBySpawnGroup(
        Masked Vector=<75 75 5>
        Passthrough Vector=&MaskVectorBySpawnGroup(
         Masked Vector=<0 0 0>
         Passthrough Vector=&MaskVectorBySpawnGroup(Masked Vector=<0 0 0>, Passthrough Vector=<75 75 5>, Spawn Group Mask=5)
         Spawn Group Mask=4
        )
        Spawn Group Mask=3
       )
       Spawn Group Mask=2
      )
      Spawn Group Mask=1
     )
    )
    MaskSpawnGroup=false
    Override Seed=false
    Shape Primitive=::NewEnumerator2
    Surface Only=false
    UseOwnerScale=true
   )
   &ShapeLocation(
    MaskSpawnGroup=false
    Override Seed=false
    Sphere Radius=&MaskFloatBySpawnGroup(
     Masked Float=0
     Passthrough Float=&MaskFloatBySpawnGroup(
      Masked Float=0
      Passthrough Float=&MaskFloatBySpawnGroup(
       Masked Float=0
       Passthrough Float=&MaskFloatBySpawnGroup(
        Masked Float=0
        Passthrough Float=&MaskFloatBySpawnGroup(
         Masked Float=10
         Passthrough Float=&MaskFloatBySpawnGroup(Masked Float=0, Passthrough Float=0, Spawn Group Mask=5)
         Spawn Group Mask=4
        )
        Spawn Group Mask=3
       )
       Spawn Group Mask=2
      )
      Spawn Group Mask=1
     )
    )
    UseOwnerScale=true
   ) {module=ShapeLocation001}
   &InitialMeshOrientation(
    Orientation Axis=&MaskVectorBySpawnGroup(
     Masked Vector=<0 0 1>
     Passthrough Vector=&MaskVectorBySpawnGroup(
      Masked Vector=<0 0 1>
      Passthrough Vector=&MaskVectorBySpawnGroup(
       Masked Vector=<0 0 1>
       Passthrough Vector=&MaskVectorBySpawnGroup(
        Masked Vector=<0 0 0>
        Passthrough Vector=&MaskVectorBySpawnGroup(
         Masked Vector=<1 0 0>
         Passthrough Vector=&MaskVectorBySpawnGroup(Masked Vector=<0 0 1>, Passthrough Vector=<1 1 0>, Spawn Group Mask=5)
         Spawn Group Mask=4
        )
        Spawn Group Mask=3
       )
       Spawn Group Mask=2
      )
      Spawn Group Mask=1
     )
    )
    Orientation Vector=&MaskVectorBySpawnGroup(
     Masked Vector=<0 0 1>
     Passthrough Vector=&MaskVectorBySpawnGroup(
      Masked Vector=<0 0 1>
      Passthrough Vector=&MaskVectorBySpawnGroup(
       Masked Vector=<0 0 1>
       Passthrough Vector=&MaskVectorBySpawnGroup(
        Masked Vector=<0 0 0>
        Passthrough Vector=&MaskVectorBySpawnGroup(
         Masked Vector=<1 0 0>
         Passthrough Vector=&MaskVectorBySpawnGroup(Masked Vector=<1 0 0>, Passthrough Vector=<1 0 0>, Spawn Group Mask=5)
         Spawn Group Mask=4
        )
        Spawn Group Mask=3
       )
       Spawn Group Mask=2
      )
      Spawn Group Mask=1
     )
    )
    Rotation=&MaskVectorBySpawnGroup(
     Masked Vector=<0 0.25 0>
     Passthrough Vector=&MaskVectorBySpawnGroup(
      Masked Vector=<0 0.25 0>
      Passthrough Vector=&MaskVectorBySpawnGroup(
       Masked Vector=<0 0.25 0>
       Passthrough Vector=&MaskVectorBySpawnGroup(
        Masked Vector=<0 0 0>
        Passthrough Vector=&MaskVectorBySpawnGroup(
         Masked Vector=&RandomRangeVector(Minimum=<-0.1 -0.1 0>, Maximum=<0.1 0.1 0>)
         Passthrough Vector=&MaskVectorBySpawnGroup(
          Masked Vector=&RandomRangeVector(Minimum=<0 0 0.1>, Maximum=<1 0 0.3>)
          Passthrough Vector=&RandomRangeVector(Minimum=<0 0 0>, Maximum=<1 1 1>)
          Spawn Group Mask=5
         )
         Spawn Group Mask=4
        )
        Spawn Group Mask=3
       )
       Spawn Group Mask=2
      )
      Spawn Group Mask=1
     )
    )
    Use Orientation Vector=true
    Use Rotation Vector=true
   )
   /SetVariables(Particles.SpriteAlignment=<1 0 0.4>) {isSetParameters=true, module=SetVariables_#6}
   /SetVariables(
    Particles.GroundPulseCO=10
    Particles.GroundPulseDMP:Vector4f=&MakeVector4(X=0.6, Y=-0.05, Z=0, W=0)
    Particles.GroundPulseOrientation:Quat4f=<0 1 0 1>
   ) {isSetParameters=true, module=SetVariables_#7}
   /SetVariables(Particles.GroundZapCO=15) {isSetParameters=true, module=SetVariables_#8}
   &CameraOffset(Camera Offset Amount=&MaskFloatBySpawnGroup(
     Masked Float=5
     Passthrough Float=&MaskFloatBySpawnGroup(Masked Float=0, Passthrough Float=10, Spawn Group Mask=5)
     Spawn Group Mask=4
    )) {module=CameraOffset001}
  ]
  particleUpdateScript [
   &ParticleState()
   &ScaleMeshSize(Scale Factor=&MaskVectorBySpawnGroup(
     Masked Vector=<1 1 1>
     Passthrough Vector=&MaskVectorBySpawnGroup(
      Masked Vector=<1 1 1>
      Passthrough Vector=&MaskVectorBySpawnGroup(
       Masked Vector=<1 1 1>
       Passthrough Vector=&MaskVectorBySpawnGroup(
        Masked Vector=<1 1 1>
        Passthrough Vector=&MaskVectorBySpawnGroup(
         Masked Vector=&MakeVector(
          X=2
          Y=8
          Z=&FloatFromCurve(FloatCurve=DataInterface() {propertyValues {Curve {keys [
               {interpMode=RCIM_Cubic, leaveTangent=0.0966, tangentMode=RCTM_Break, time=0, value=1}
               {arriveTangent=0.0951, interpMode=RCIM_Cubic, leaveTangent=-0.1745, tangentMode=RCTM_Break, time=1, value=1.07}
              ]}}})
         )
         Passthrough Vector=&MaskVectorBySpawnGroup(Masked Vector=<1 1 1>, Passthrough Vector=<1 1 1>, Spawn Group Mask=5)
         Spawn Group Mask=4
        )
        Spawn Group Mask=3
       )
       Spawn Group Mask=2
      )
      Spawn Group Mask=1
     )
    ))
   &ScaleMeshSize(Scale Factor=&MaskVectorBySpawnGroup(
     Masked Vector=<1 1 1>
     Passthrough Vector=&MaskVectorBySpawnGroup(
      Masked Vector=<1 1 1>
      Passthrough Vector=&MaskVectorBySpawnGroup(
       Masked Vector=<1 1 1>
       Passthrough Vector=&MaskVectorBySpawnGroup(
        Masked Vector=<1 1 1>
        Passthrough Vector=&MaskVectorBySpawnGroup(
         Masked Vector=&RandomRangeVector(Minimum=<3 3 30>, Maximum=<4 4 60>)
         Passthrough Vector=&MaskVectorBySpawnGroup(
          Masked Vector=&Multiply_VectorByFloat(Vector=<1 1 1>, Float=0.22)
          Passthrough Vector=&Multiply_VectorByFloat(
           Vector=&RandomRangeVector(Minimum=<1 1 2.2>, Maximum=<2 2 2.2>)
           Float=&FloatFromCurve(
            FloatCurve=DataInterface() {propertyValues {Curve {keys [<at 0 1>,<at 1 1.5>]}}}
            Scale Curve=&RandomRangeFloat(Minimum=0.05, Maximum=0.07)
           )
          )
          Spawn Group Mask=5
         )
         Spawn Group Mask=4
        )
        Spawn Group Mask=3
       )
       Spawn Group Mask=2
      )
      Spawn Group Mask=1
     )
    )) {module=ScaleMeshSize001}
   &UpdateMeshOrientation(
    Rotation Rate=&MaskFloatBySpawnGroup(
     Masked Float=0
     Passthrough Float=&MaskFloatBySpawnGroup(
      Masked Float=0
      Passthrough Float=&MaskFloatBySpawnGroup(
       Masked Float=0
       Passthrough Float=&MaskFloatBySpawnGroup(
        Masked Float=0
        Passthrough Float=&MaskFloatBySpawnGroup(
         Masked Float=0
         Passthrough Float=&MaskFloatBySpawnGroup(Masked Float=&RandomRangeFloat(Minimum=-1, Maximum=1), Passthrough Float=0, Spawn Group Mask=5)
         Spawn Group Mask=4
        )
        Spawn Group Mask=3
       )
       Spawn Group Mask=2
      )
      Spawn Group Mask=1
     )
    )
    Rotation Vector=&MaskVectorBySpawnGroup(
     Masked Vector=<0 0 0>
     Passthrough Vector=&MaskVectorBySpawnGroup(
      Masked Vector=<0 0 0>
      Passthrough Vector=&MaskVectorBySpawnGroup(
       Masked Vector=<0 0 0>
       Passthrough Vector=&MaskVectorBySpawnGroup(
        Masked Vector=<0 0 0>
        Passthrough Vector=&MaskVectorBySpawnGroup(
         Masked Vector=<0 0 0>
         Passthrough Vector=&MaskVectorBySpawnGroup(Masked Vector=<0 0 1>, Passthrough Vector=<0 0 0>, Spawn Group Mask=5)
         Spawn Group Mask=4
        )
        Spawn Group Mask=3
       )
       Spawn Group Mask=2
      )
      Spawn Group Mask=1
     )
    )
   )
   &SpriteRotationRate(Rotation Rate=&FloatFromCurve(
     FloatCurve=DataInterface() {propertyValues {Curve {keys [
         {arriveTangent=1.4306, interpMode=RCIM_Cubic, leaveTangent=1.4306, tangentMode=RCTM_User, time=0.2, value=0}
         {arriveTangent=1.1179, interpMode=RCIM_Cubic, leaveTangent=1.1179, tangentMode=RCTM_User, time=0.4, value=0.6228}
         {arriveTangent=0.0224, interpMode=RCIM_Cubic, leaveTangent=0.0224, tangentMode=RCTM_User, time=1, value=1}
        ]}}}
     Scale Curve=0
    ))
   &ScaleColor(
    Scale Alpha=&MaskFloatBySpawnGroup(
     Masked Float=&FloatFromCurve(FloatCurve=DataInterface() {propertyValues {Curve {keys [
          {arriveTangent=-1.5276, interpMode=RCIM_Cubic, leaveTangent=-1.5276, tangentMode=RCTM_User, time=0, value=1}
          {interpMode=RCIM_Cubic, tangentMode=RCTM_User, time=1, value=0}
         ]}}})
     Passthrough Float=&MaskFloatBySpawnGroup(
      Masked Float=&FloatFromCurve(FloatCurve=DataInterface() {propertyValues {Curve {keys [
           {arriveTangent=-1.5276, interpMode=RCIM_Cubic, leaveTangent=-1.5276, tangentMode=RCTM_User, time=0, value=1}
           {interpMode=RCIM_Cubic, tangentMode=RCTM_User, time=1, value=0}
          ]}}})
      Passthrough Float=&MaskFloatBySpawnGroup(
       Masked Float=&FloatFromCurve(FloatCurve=DataInterface() {propertyValues {Curve {keys [
            {arriveTangent=-1.5276, interpMode=RCIM_Cubic, leaveTangent=-1.5276, tangentMode=RCTM_User, time=0, value=1}
            {interpMode=RCIM_Cubic, tangentMode=RCTM_User, time=1, value=0}
           ]}}})
       Passthrough Float=&MaskFloatBySpawnGroup(
        Masked Float=&FloatFromCurve(FloatCurve=DataInterface() {propertyValues {Curve {keys [
             {arriveTangent=-1.5276, interpMode=RCIM_Cubic, leaveTangent=-1.5276, tangentMode=RCTM_User, time=0, value=1}
             {interpMode=RCIM_Cubic, tangentMode=RCTM_User, time=1, value=0}
            ]}}})
        Passthrough Float=&MaskFloatBySpawnGroup(
         Masked Float=&FloatFromCurve(FloatCurve=DataInterface() {propertyValues {Curve {keys [
              {tangentMode=RCTM_User, time=0, value=1}
              {arriveTangent=-0.0405, interpMode=RCIM_Cubic, leaveTangent=-0.0405, tangentMode=RCTM_User, time=0.7009, value=1}
              {arriveTangent=-8.0089, interpMode=RCIM_Cubic, leaveTangent=-8.0089, tangentMode=RCTM_User, time=1, value=0}
             ]}}})
         Passthrough Float=&MaskFloatBySpawnGroup(
          Masked Float=&FloatFromCurve(FloatCurve=DataInterface() {propertyValues {Curve {keys [
               {tangentMode=RCTM_User, time=0, value=1}
               {arriveTangent=-0.0405, interpMode=RCIM_Cubic, leaveTangent=-0.0405, tangentMode=RCTM_User, time=0.7009, value=1}
               {arriveTangent=-8.0089, interpMode=RCIM_Cubic, leaveTangent=-8.0089, tangentMode=RCTM_User, time=1, value=0}
              ]}}})
          Passthrough Float=2
          Spawn Group Mask=5
         )
         Spawn Group Mask=4
        )
        Spawn Group Mask=3
       )
       Spawn Group Mask=2
      )
      Spawn Group Mask=1
     )
    )
    Scale RGB=&MaskVectorBySpawnGroup(
     Masked Vector=<1 1 1>
     Passthrough Vector=&MaskVectorBySpawnGroup(
      Masked Vector=<1 1 1>
      Passthrough Vector=&MaskVectorBySpawnGroup(
       Masked Vector=<1 1 1>
       Passthrough Vector=&MaskVectorBySpawnGroup(
        Masked Vector=<1 1 1>
        Passthrough Vector=&MaskVectorBySpawnGroup(
         Masked Vector=&Lerp_Vector(
          A=&VectorFromFloat(Value=&FloatFromCurve(
            FloatCurve=DataInterface() {propertyValues {Curve {keys [<at 0 1>,<at 0.1 1>,<at 0.2 2>,<at 0.3 1>,<at 0.6 1>,<at 0.7 2>,<at 0.8 1>,<at 1 1>]}}}
            Scale Curve=&MaskFloatBySpawnGroup(
             Masked Float=25
             Passthrough Float=&MaskFloatBySpawnGroup(Masked Float=10, Passthrough Float=25, Spawn Group Mask=5)
             Spawn Group Mask=4
            )
           ))
          B=&VectorFromFloat(Value=&RandomRangeFloat(Minimum=0.3, Maximum=5, Evaluation Type=::NewEnumerator1))
          Alpha=&FloatFromCurve(FloatCurve=DataInterface() {propertyValues {Curve {keys [
               {interpMode=RCIM_Cubic, leaveTangent=2.5294, tangentMode=RCTM_Break, time=0, value=0}
               {arriveTangent=-0.2166, interpMode=RCIM_Cubic, tangentMode=RCTM_Break, time=0.4, value=1}
              ]}}})
         )
         Passthrough Vector=&MaskVectorBySpawnGroup(
          Masked Vector=&Lerp_Vector(
           A=&VectorFromFloat(Value=&FloatFromCurve(
             FloatCurve=DataInterface() {propertyValues {Curve {keys [<at 0 1>,<at 0.1 1>,<at 0.2 2>,<at 0.3 1>,<at 0.6 1>,<at 0.7 2>,<at 0.8 1>,<at 1 1>]}}}
             Scale Curve=&MaskFloatBySpawnGroup(
              Masked Float=25
              Passthrough Float=&MaskFloatBySpawnGroup(Masked Float=10, Passthrough Float=25, Spawn Group Mask=5)
              Spawn Group Mask=4
             )
            ))
           B=&VectorFromFloat(Value=&RandomRangeFloat(Minimum=0.3, Maximum=5, Evaluation Type=::NewEnumerator1))
           Alpha=&FloatFromCurve(FloatCurve=DataInterface() {propertyValues {Curve {keys [
                {interpMode=RCIM_Cubic, leaveTangent=2.5294, tangentMode=RCTM_Break, time=0, value=0}
                {arriveTangent=-0.2166, interpMode=RCIM_Cubic, tangentMode=RCTM_Break, time=0.4, value=1}
               ]}}})
          )
          Passthrough Vector=&Lerp_Vector(
           A=&VectorFromFloat(Value=&FloatFromCurve(
             FloatCurve=DataInterface() {propertyValues {Curve {keys [<at 0 1>,<at 0.1 1>,<at 0.2 2>,<at 0.3 1>,<at 0.6 1>,<at 0.7 2>,<at 0.8 1>,<at 1 1>]}}}
             Scale Curve=&MaskFloatBySpawnGroup(
              Masked Float=25
              Passthrough Float=&MaskFloatBySpawnGroup(Masked Float=10, Passthrough Float=25, Spawn Group Mask=5)
              Spawn Group Mask=4
             )
            ))
           B=&VectorFromFloat(Value=&RandomRangeFloat(Minimum=0.3, Maximum=5, Evaluation Type=::NewEnumerator1))
           Alpha=&FloatFromCurve(FloatCurve=DataInterface() {propertyValues {Curve {keys [
                {interpMode=RCIM_Cubic, leaveTangent=2.5294, tangentMode=RCTM_Break, time=0, value=0}
                {arriveTangent=-0.2166, interpMode=RCIM_Cubic, tangentMode=RCTM_Break, time=0.4, value=1}
               ]}}})
          )
          Spawn Group Mask=5
         )
         Spawn Group Mask=4
        )
        Spawn Group Mask=3
       )
       Spawn Group Mask=2
      )
      Spawn Group Mask=1
     )
    )
    ScaleA=true
    ScaleRGB=true
   ) {module=ScaleColor001}
   &ScaleSpriteSize(
    Scale Sprite Size Mode=::NewEnumerator0
    Uniform Scale Factor=&FloatFromCurve(FloatCurve=DataInterface() {propertyValues {Curve {keys [
         {arriveTangent=2.278, interpMode=RCIM_Cubic, leaveTangent=2.278, tangentMode=RCTM_User, time=0, value=0}
         {interpMode=RCIM_Cubic, tangentMode=RCTM_User, time=1, value=1}
        ]}}})
   )
   /SetVariables(Particles.WaveDynMat:Vector4f=&MakeVector4(
     X=&FloatFromCurve(
      FloatCurve=DataInterface() {propertyValues {Curve {keys [{arriveTangent=3.726, interpMode=RCIM_Cubic, leaveTangent=3.726, tangentMode=RCTM_User, time=0, value=-0.3}, <at 1 1>]}}}
      Scale Curve=0.6
     )
     Y=&FloatFromCurve(FloatCurve=DataInterface() {propertyValues {Curve {keys [{arriveTangent=-0.5818, interpMode=RCIM_Cubic, leaveTangent=-0.5818, tangentMode=RCTM_User, time=0, value=1}, <at 1 0.8>]}}})
     Z=&FloatFromCurve(FloatCurve=DataInterface() {propertyValues {Curve {keys [<at 0 0>,<at 0.15 1>,<at 1 1.2>]}}}, Scale Curve=10)
     W=&FloatFromCurve(FloatCurve=DataInterface() {propertyValues {Curve {keys [<at 0 0>,<at 1 1>]}}}, Scale Curve=0)
    )) {isSetParameters=true, module=SetVariables_#9}
   /SetVariables(Particles.GlowSize=&Multiply_Vector2D(A=$Particles.SpriteSize, B=<3 3>)) {isSetParameters=true, module=SetVariables_#10}
   /SetVariables(Particles.GlowColor=&MultiplyLinearColor(Linear Color=$Particles.Color, Scale Factor=<1 1 1 0.08>)) {isSetParameters=true, module=SetVariables_#11}
   /SetVariables(
    Particles.ShockColor:LinearColor=&MultiplyLinearColor(
     Linear Color=&ColorFromCurve(ColorCurve=DataInterface() {propertyValues {
        AlphaCurve {keys [<at 0 1 RCIM_Cubic>]}
        BlueCurve {keys [<at 0 20 RCIM_Cubic>,<at 0.6348 8>]}
        GreenCurve {keys [<at 0 12.843 RCIM_Cubic>,<at 0.6348 0.4929>]}
        RedCurve {keys [<at 0 0 RCIM_Cubic>,<at 0.6348 0>]}
       }})
     Scale Factor=&MakeVector4_FromFloat(Float=&FloatFromCurve(
       FloatCurve=DataInterface() {propertyValues {Curve {keys [
           {arriveTangent=-1.5276, interpMode=RCIM_Cubic, leaveTangent=-1.5276, tangentMode=RCTM_User, time=0, value=1}
           {interpMode=RCIM_Cubic, tangentMode=RCTM_User, time=1, value=0}
          ]}}}
       Scale Curve=0.5
      ))
    )
    Particles.ShockwaveScale:Vector3f=&MaskVectorBySpawnGroup(
     Masked Vector=&Multiply_VectorByFloat(
      Vector=$Particles.Initial.Scale
      Float=&FloatFromCurve(FloatCurve=DataInterface() {propertyValues {Curve {keys [
           {arriveTangent=1.8207, interpMode=RCIM_Cubic, leaveTangent=1.8207, tangentMode=RCTM_User, time=0, value=0.2}
           {interpMode=RCIM_Cubic, tangentMode=RCTM_User, time=1, value=1.2}
          ]}}})
     )
     Passthrough Vector=$Particles.Initial.Scale
    )
   ) {isSetParameters=true, module=SetVariables_#12}
   /SetVariables(
    Particles.BeamColor:LinearColor=&MultiplyLinearColor(
     Linear Color=&LerpLinearColors(StartColor=<rgba 0.634 1.8967 3 0.3>, EndColor=<rgba 0 5.8116 6 0.4>, LerpFactor=$Particles.Rando)
     Scale Factor=&MakeVector4_FromFloat(Float=&FloatFromCurve(
       FloatCurve=DataInterface() {propertyValues {Curve {keys [
           {arriveTangentWeight=0.0275, interpMode=RCIM_Cubic, leaveTangentWeight=0.3161, tangentMode=RCTM_User, time=0, value=1}
           {arriveTangent=-2.6497, interpMode=RCIM_Cubic, leaveTangent=-2.6497, tangentMode=RCTM_User, time=0.25, value=1}
           {arriveTangent=-0.0405, interpMode=RCIM_Cubic, leaveTangent=-0.0405, tangentMode=RCTM_User, time=1, value=0}
          ]}}}
       Scale Curve=0.25
      ))
    )
    Particles.BeamDMP:Vector4f=&MakeVector4(
     X=0.1
     Y=&FloatFromCurve(
      FloatCurve=DataInterface() {propertyValues {Curve {keys [
          {arriveTangent=5.1604, interpMode=RCIM_Cubic, leaveTangent=5.1604, tangentMode=RCTM_User, time=0, value=0}
          {interpMode=RCIM_Cubic, tangentMode=RCTM_User, time=0.5, value=1}
          <at 1 0 RCIM_Cubic>
         ]}}}
      Scale Curve=&Lerp_Float(A=1.75, B=1.5, Alpha=$Particles.Rando)
     )
     Z=0
     W=0
    )
    Particles.BeamScale:Vector3f=&Multiply_VectorByFloat(
     Vector=&VectorFromFloat(Value=&Lerp_Float(A=1.5, B=2.5, Alpha=$Particles.Rando))
     Float=&FloatFromCurve(
      FloatCurve=DataInterface() {propertyValues {Curve {keys [
          {arriveTangent=3.3092, interpMode=RCIM_Cubic, leaveTangent=3.3092, tangentMode=RCTM_User, time=0, value=0.4}
          {arriveTangent=0.048, interpMode=RCIM_Cubic, leaveTangent=0.048, tangentMode=RCTM_User, time=0.5, value=1}
          {arriveTangent=-2.6178, interpMode=RCIM_Cubic, leaveTangent=-2.6178, tangentMode=RCTM_User, time=1, value=0.5}
         ]}}}
      Scale Curve=0.75
     )
    )
   ) {isSetParameters=true, module=SetVariables_#13}
   /SetVariables(
    Particles.GroundPulseColor:LinearColor=&MultiplyLinearColor(
     Linear Color=&ColorFromCurve(ColorCurve=DataInterface() {propertyValues {
        AlphaCurve {keys [<at 0 0.2 RCIM_Cubic>,<at 1 0 RCIM_Cubic>]}
        BlueCurve {keys [<at 0 0.3289 RCIM_Cubic>,<at 0.1757 1>,<at 1 1 RCIM_Cubic>]}
        GreenCurve {keys [<at 0 1 RCIM_Cubic>,<at 0.1757 0.8834>,<at 1 0.676 RCIM_Cubic>]}
        RedCurve {keys [<at 0 0 RCIM_Cubic>,<at 0.1757 0>,<at 1 0 RCIM_Cubic>]}
       }})
     Scale Factor=&MakeVector4_FromFloat(Float=&FloatFromCurve(
       FloatCurve=DataInterface() {propertyValues {Curve {keys [
           {arriveTangent=7.3302, interpMode=RCIM_Cubic, leaveTangent=7.3302, tangentMode=RCTM_User, time=0, value=0}
           {arriveTangent=122.0485, interpMode=RCIM_Cubic, leaveTangent=-11.619, tangentMode=RCTM_Break, time=0.05, value=5}
           {arriveTangent=-0.6259, tangentMode=RCTM_User, time=1, value=0}
          ]}}}
       Scale Curve=0.75
      ))
    )
    Particles.GroundPulseScale:Vector3f=&Multiply_VectorByFloat(
     Vector=<0.02 0.6 0.6>
     Float=&FloatFromCurve(
      FloatCurve=DataInterface() {propertyValues {Curve {keys [
          {arriveTangent=4.1785, interpMode=RCIM_Cubic, leaveTangent=4.1785, tangentMode=RCTM_User, time=0, value=0}
          {arriveTangent=1.1179, interpMode=RCIM_Cubic, leaveTangent=1.1179, tangentMode=RCTM_User, time=0.2042, value=0.6228}
          {arriveTangent=-0.2023, interpMode=RCIM_Cubic, leaveTangent=-0.2023, tangentMode=RCTM_User, time=0.9586, value=1}
          {arriveTangent=0.0224, interpMode=RCIM_Cubic, leaveTangent=0.0224, tangentMode=RCTM_User, time=1, value=0}
         ]}}}
      Scale Curve=1.2
     )
    )
   ) {isSetParameters=true, module=SetVariables_#14}
   /SetVariables(
    Particles.GroundZapColor:LinearColor=&Multiply_LinearColorByFloat(
     Linear Color=$System.ElimColor
     Float=&FloatFromCurve(
      FloatCurve=DataInterface() {propertyValues {Curve {keys [
          {tangentMode=RCTM_User, time=0, value=0}
          <at 0.1 1.2>
          <at 0.2 0.7>
          <at 0.6 0.6>
          <at 0.7 0.875>
          <at 0.8 0.35>
          {tangentMode=RCTM_User, time=1, value=0}
         ]}}}
      Scale Curve=&RandomRangeFloat(Minimum=0.2, Maximum=0.4)
     )
    )
    Particles.GroundZapSize:Vector2f=&Multiply_Vector2D_ByFloat(
     Float=&FloatFromCurve(FloatCurve=DataInterface() {propertyValues {Curve {keys [
          <at 0 0 RCIM_Cubic>
          <at 0.112 1.0173>
          {arriveTangent=-1.142, interpMode=RCIM_Cubic, leaveTangent=-1.142, time=0.7001, value=0.872}
          <at 1.0011 0.002>
         ]}}})
     Vector2D=<275 125>
    )
   ) {isSetParameters=true, module=SetVariables_#15}
   &FN_Light_Attributes(
    Color=<rgba 0 0.5043 1 0.7>
    Create Color Attribute=true
    Create Exponent Attribute=true
    Create Radius Attribute=true
    Exponent=1
    Radius=280
    Scale Color=&FloatFromCurve(FloatCurve=DataInterface() {propertyValues {Curve {keys [<at 0 1>,<at 1 0>]}}}, Scale Curve=20)
   )
   &CameraOffset(Camera Offset Amount=&MaskFloatBySpawnGroup(
     Masked Float=&Lerp_Float(
      A=30
      B=-5
      Alpha=&FloatFromCurve(FloatCurve=DataInterface() {propertyValues {Curve {keys [<at 0 0>,<at 0.25 1>]}}}, CurveIndex=$Engine.System.Age)
     )
     Passthrough Float=&MaskFloatBySpawnGroup(
      Masked Float=&Lerp_Float(
       A=30
       B=-5
       Alpha=&FloatFromCurve(FloatCurve=DataInterface() {propertyValues {Curve {keys [<at 0 0>,<at 0.25 1>]}}}, CurveIndex=$Engine.System.Age)
      )
      Passthrough Float=&MaskFloatBySpawnGroup(
       Masked Float=&Lerp_Float(
        A=30
        B=-5
        Alpha=&FloatFromCurve(FloatCurve=DataInterface() {propertyValues {Curve {keys [<at 0 0>,<at 0.25 1>]}}}, CurveIndex=$Engine.System.Age)
       )
       Passthrough Float=&MaskFloatBySpawnGroup(
        Masked Float=&Lerp_Float(
         A=30
         B=-5
         Alpha=&FloatFromCurve(FloatCurve=DataInterface() {propertyValues {Curve {keys [<at 0 0>,<at 0.25 1>]}}}, CurveIndex=$Engine.System.Age)
        )
        Passthrough Float=&MaskFloatBySpawnGroup(
         Masked Float=0
         Passthrough Float=&MaskFloatBySpawnGroup(Masked Float=0, Passthrough Float=0, Spawn Group Mask=5)
         Spawn Group Mask=4
        )
        Spawn Group Mask=3
       )
       Spawn Group Mask=2
      )
      Spawn Group Mask=1
     )
    ))
   &DynamicMaterialParameters(
    Index 0 Param 1=&MakeFloatFromVector(Vector=$Particles.Scale, Channel=::NewEnumerator0)
    Index 0 Param 2=&RandomRangeFloat(
     Minimum=0
     Maximum=&MaskFloatBySpawnGroup(
      Masked Float=10
      Passthrough Float=&MaskFloatBySpawnGroup(Masked Float=10, Passthrough Float=35, Spawn Group Mask=5)
      Spawn Group Mask=4
     )
    )
    Index 0 Param 3=&FloatFromCurve(
     FloatCurve=DataInterface() {propertyValues {Curve {keys [{interpMode=RCIM_Cubic, leaveTangent=8.1085, tangentMode=RCTM_Break, time=0.8568, value=1}, {tangentMode=RCTM_User, time=1, value=0}]}}}
     Scale Curve=0.07
    )
    Index 0 Param 4=&MaskFloatBySpawnGroup(
     Masked Float=0.8
     Passthrough Float=&MaskFloatBySpawnGroup(Masked Float=0.8, Passthrough Float=&RandomRangeFloat(Minimum=0.8, Maximum=1), Spawn Group Mask=5)
     Spawn Group Mask=4
    )
    Index 1 Param 1=&FloatFromCurve(FloatCurve=DataInterface() {propertyValues {Curve {keys [
         {interpMode=RCIM_Cubic, leaveTangent=1.1034, tangentMode=RCTM_Break, time=0, value=0}
         {arriveTangent=2.2399, tangentMode=RCTM_User, time=0.5, value=1}
        ]}}})
    Index 1 Param 2=&FloatFromCurve(FloatCurve=DataInterface() {propertyValues {Curve {keys [<at 0.5 0>, {arriveTangent=3.2053, tangentMode=RCTM_User, time=0.9, value=1}]}}})
    Index 1 Param 3=&FloatFromCurve(FloatCurve=DataInterface() {propertyValues {Curve {keys [<at 0 0 RCIM_Cubic>, {arriveTangent=3.4223, interpMode=RCIM_Cubic, leaveTangent=3.4223, tangentMode=RCTM_User, time=1, value=1}]}}})
    Index 1 Param 4=&FloatFromCurve(
     FloatCurve=DataInterface() {propertyValues {Curve {keys [
         {interpMode=RCIM_Cubic, leaveTangent=0.1169, tangentMode=RCTM_Break, time=0, value=0}
         {arriveTangent=2.5708, interpMode=RCIM_Cubic, leaveTangent=2.5708, tangentMode=RCTM_User, time=1, value=1}
        ]}}}
     Scale Curve=&RandomRangeFloat(
      Minimum=1
      Maximum=&MaskFloatBySpawnGroup(
       Masked Float=15
       Passthrough Float=&MaskFloatBySpawnGroup(Masked Float=40, Passthrough Float=5, Spawn Group Mask=5)
       Spawn Group Mask=4
      )
     )
    )
    Index 2 Param 1=&FloatFromCurve(
     FloatCurve=DataInterface() {propertyValues {Curve {keys [
         {interpMode=RCIM_Cubic, leaveTangent=-0.0025, tangentMode=RCTM_Break, time=0, value=0.5}
         {arriveTangent=-0.0316, interpMode=RCIM_Cubic, tangentMode=RCTM_Break, time=1, value=-0.2}
        ]}}}
     Scale Curve=&MaskFloatBySpawnGroup(
      Masked Float=1
      Passthrough Float=&MaskFloatBySpawnGroup(Masked Float=1, Passthrough Float=-0.5, Spawn Group Mask=5)
      Spawn Group Mask=4
     )
    )
    Index 2 Param 2=&FloatFromCurve(
     FloatCurve=DataInterface() {propertyValues {Curve {keys [
         {interpMode=RCIM_Cubic, tangentMode=RCTM_User, time=0, value=0}
         {arriveTangent=2.0522, interpMode=RCIM_Cubic, leaveTangent=2.0522, tangentMode=RCTM_User, time=1, value=1}
        ]}}}
     Scale Curve=&MaskFloatBySpawnGroup(
      Masked Float=2
      Passthrough Float=&MaskFloatBySpawnGroup(Masked Float=2, Passthrough Float=1, Spawn Group Mask=5)
      Spawn Group Mask=4
     )
    )
    Module.Index0Param0Write=true
    Module.Index0Param1Write=true
    Module.Index0Param2Write=true
    Module.Index0Param3Write=true
    Module.Index1Param0Write=true
    Module.Index1Param1Write=true
    Module.Index1Param2Write=true
    Module.Index1Param3Write=true
    Module.Index2Param0Write=true
    Module.Index2Param1Write=true
    Module.Index2Param2Write=false
    Module.Index2Param3Write=false
    Module.Index3Param0Write=false
    Module.Index3Param1Write=false
    Module.Index3Param2Write=false
    Module.Index3Param3Write=true
    Write Parameter Index 1=true
    Write Parameter Index 2=true
    Write Parameter Index 3=true
   ) {module=DynamicMaterialParameters001}
  ]
  rendererClasses [
   &NiagaraSpriteRendererProperties
   &NiagaraLightRendererProperties
   &NiagaraMeshRendererProperties
  ]
  renderers [
   {rendererClass=&NiagaraSpriteRendererProperties}
   {rendererClass=&NiagaraSpriteRendererProperties, rendererIndex=1}
   {rendererClass=&NiagaraLightRendererProperties, rendererIndex=2}
   {rendererClass=&NiagaraMeshRendererProperties, rendererIndex=3}
   {rendererClass=&NiagaraMeshRendererProperties, rendererIndex=4}
   {rendererClass=&NiagaraMeshRendererProperties, rendererIndex=5}
   {rendererClass=&NiagaraSpriteRendererProperties, rendererIndex=6}
   {rendererClass=&NiagaraMeshRendererProperties, rendererIndex=7}
   {rendererClass=&NiagaraMeshRendererProperties, rendererIndex=8}
   {rendererClass=&NiagaraMeshRendererProperties, rendererIndex=9}
   {rendererClass=&NiagaraMeshRendererProperties, rendererIndex=10}
   {rendererClass=&NiagaraMeshRendererProperties, rendererIndex=11}
   {rendererClass=&NiagaraMeshRendererProperties, rendererIndex=12}
  ]
 }
 HangingParticulates001 {
  bEnabled=false
  emitterUpdateScript [
   &EmitterState()
   &SpawnRate(
    SpawnRate=&Multiply_Float(A=&RandomRangeFloat(Minimum=7, Maximum=15), B=$System.FadeOutAura)
    Use Spawn Probability=false
   )
  ]
  particleSpawnScript [
   &InitializeParticle(
    Color=$System.ElimColor
    Lifetime Max=0.6
    Lifetime Min=0.3
    Lifetime Mode=::NewEnumerator1
    Sprite Size=<275 125>
    Sprite Size Mode=::NewEnumerator4
    Sprite UV Mode=::NewEnumerator2
    UsePositionOffset=false
    Write Material Random=true
   )
   &ShapeLocation(Box Size=<75 75 5>, MaskSpawnGroup=false, Override Seed=false, Shape Primitive=::NewEnumerator2, Surface Only=false, UseOwnerScale=true)
   &AddVelocity(Velocity=<0 0 0>)
   &CameraOffset(Camera Offset Amount=15)
  ]
  particleUpdateScript [
   &ParticleState()
   &WindForce(Frequency=6, Offset By Time:Vector3f=<0 0 -50>, Scale=4, Wind Speed=<1 0 0>)
   &AerodynamicDrag(
    Aerodynamic Drag:NiagaraFloat=&RandomRangeFloat(Minimum=0.6, Maximum=1)
    Aerodynamic Rotational Drag=0.2
    Drag Shape Curve=DataInterface() {propertyValues {Curve {keys [
        {interpMode=RCIM_Cubic, tangentMode=RCTM_User, time=0, value=0.1}
        {arriveTangent=2.0799, interpMode=RCIM_Cubic, leaveTangent=2.0799, tangentMode=RCTM_User, time=1, value=1}
       ]}}}
    Lift Shape Curve=DataInterface() {propertyValues {Curve {keys [
        {interpMode=RCIM_Cubic, tangentMode=RCTM_User, time=0, value=0}
        {interpMode=RCIM_Cubic, tangentMode=RCTM_User, time=0.5, value=1}
        {tangentMode=RCTM_User, time=1, value=0}
       ]}}}
    Use Linear Drag=true
    Use Rotational Drag=true
   )
   &ScaleSpriteSize(Scale Factor=&Vector2DFromFloat(Value=&FloatFromCurve(FloatCurve=DataInterface() {propertyValues {Curve {keys [
          <at 0 0 RCIM_Cubic>
          <at 0.112 1.0173>
          {arriveTangent=-1.142, interpMode=RCIM_Cubic, leaveTangent=-1.142, time=0.7001, value=0.872}
          <at 1.0011 0.002>
         ]}}})))
   &ScaleColor(
    Scale Alpha=&FloatFromCurve(
     FloatCurve=DataInterface() {propertyValues {Curve {keys [
         {tangentMode=RCTM_User, time=0, value=0}
         <at 0.1 1.15>
         <at 0.2 0.625>
         <at 0.6 0.375>
         <at 0.7 0.875>
         <at 0.8 0.35>
         {tangentMode=RCTM_User, time=1, value=0}
        ]}}}
     Scale Curve=&RandomRangeFloat(Minimum=0.05, Maximum=0.1)
    )
    Scale RGB=&VectorFromFloat(Value=&FloatFromCurve(FloatCurve=DataInterface() {propertyValues {Curve {keys [<at 0 1>,<at 0.1 1.55>,<at 0.2 1>,<at 0.3 1.3>,<at 0.5 1>,<at 0.6 1.2>,<at 0.7 1>,<at 1 1>]}}}))
    ScaleA=true
    ScaleRGB=true
   )
   &SolveForcesAndVelocity(Clamp Velocity=false, Limit Acceleration=false)
  ]
  rendererClasses [&NiagaraSpriteRendererProperties]
  renderers [{rendererClass=&NiagaraSpriteRendererProperties}]
 }
 Lightning {
  bEnabled=false
  emitterUpdateScript [
   &EmitterState(Enable Visibility Culling=true, Scalability Mode=::NewEnumerator1)
   &SpawnBurst_Instantaneous(
    Spawn Count=&RandomRangeInt(Minimum=8, Maximum=12)
    Spawn Group=4
    Use Spawn Probability=false
    UseLoopCountLimit=false
   )
   &SpawnBurst_Instantaneous(Spawn Count=10, Spawn Group=5, Use Spawn Probability=false, UseLoopCountLimit=false) {module=SpawnBurst_Instantaneous001}
   &SpawnRate(SpawnGroup=5, SpawnRate=&Multiply_Float(A=15, B=$System.FadeOutBottomLightning), Use Spawn Probability=false)
   &SpawnRate(SpawnGroup=6, SpawnRate=&Multiply_Float(A=25, B=$System.FadeOutAura), Use Spawn Probability=false) {module=SpawnRate001}
   &SpawnBurst_Instantaneous(Spawn Count=4, Spawn Group=6, Use Spawn Probability=false, UseLoopCountLimit=false) {module=SpawnBurst_Instantaneous002}
  ]
  particleSpawnScript [
   /SetVariables(Particles.VisibilityTag=&SpawnGroup()) {isSetParameters=true}
   &InitializeParticle(
    Color=&Multiply_LinearColorByFloat(Linear Color=$System.ElimColor, Float=0.4)
    Lifetime=&MaskFloatBySpawnGroup(
     Masked Float=&RandomRangeFloat(Minimum=0.2, Maximum=0.3)
     Passthrough Float=&MaskFloatBySpawnGroup(
      Masked Float=&RandomRangeFloat(Minimum=0.175, Maximum=0.35)
      Passthrough Float=&RandomRangeFloat(Minimum=0.2, Maximum=0.3)
      Spawn Group Mask=5
     )
     Spawn Group Mask=4
    )
    Mesh Scale Max=&MaskVectorBySpawnGroup(
     Masked Vector=&VectorFromFloat(Value=0.03)
     Passthrough Vector=&MaskVectorBySpawnGroup(Masked Vector=<1.4 1.4 2.2>, Passthrough Vector=&VectorFromFloat(Value=0.6), Spawn Group Mask=5)
     Spawn Group Mask=4
    )
    Mesh Scale Min=&MaskVectorBySpawnGroup(
     Masked Vector=&VectorFromFloat(Value=0.01)
     Passthrough Vector=&MaskVectorBySpawnGroup(Masked Vector=<1.2 1.2 1.5>, Passthrough Vector=&VectorFromFloat(Value=0.6), Spawn Group Mask=5)
     Spawn Group Mask=4
    )
    Mesh Scale Mode=::NewEnumerator2
    Position Offset=&MaskVectorBySpawnGroup(
     Masked Vector=<0 0 0>
     Passthrough Vector=&MaskVectorBySpawnGroup(
      Masked Vector=&RandomRangeVector(Minimum=<0 0 15>, Maximum=<0 0 25>)
      Passthrough Vector=<0 0 0>
      Spawn Group Mask=5
     )
     Spawn Group Mask=4
    )
    UsePositionOffset=true
    Write Material Random=true
   )
   &InitialMeshOrientation(
    Mesh Axis To Orient=&MaskVectorBySpawnGroup(
     Masked Vector=<1 0 0>
     Passthrough Vector=&MaskVectorBySpawnGroup(Masked Vector=<0 0 1>, Passthrough Vector=<1 1 0>, Spawn Group Mask=5)
     Spawn Group Mask=4
    )
    Mesh Orientation Mode=::NewEnumerator2
    Rotation=&MaskVectorBySpawnGroup(
     Masked Vector=&RandomRangeVector(Minimum=<-0.1 -0.1 0>, Maximum=<0.1 0.1 0>)
     Passthrough Vector=&MaskVectorBySpawnGroup(
      Masked Vector=&RandomRangeVector(Minimum=<0 0 0.1>, Maximum=<1 0 0.3>)
      Passthrough Vector=&RandomRangeVector(Minimum=<0 0 0>, Maximum=<1 1 1>)
      Spawn Group Mask=5
     )
     Spawn Group Mask=4
    )
    Use Rotation Vector=true
   )
   &ShapeLocation(
    MaskSpawnGroup=false
    Override Seed=false
    Sphere Radius=&MaskFloatBySpawnGroup(
     Masked Float=10
     Passthrough Float=&MaskFloatBySpawnGroup(Masked Float=0, Passthrough Float=0, Spawn Group Mask=5)
     Spawn Group Mask=4
    )
    UseOwnerScale=true
   )
   &ShapeLocation(
    Box Size=&MaskVectorBySpawnGroup(
     Masked Vector=<0 0 0>
     Passthrough Vector=&MaskVectorBySpawnGroup(Masked Vector=<0 0 0>, Passthrough Vector=<75 75 5>, Spawn Group Mask=5)
     Spawn Group Mask=4
    )
    MaskSpawnGroup=false
    Override Seed=false
    Shape Primitive=::NewEnumerator2
    Surface Only=false
    UseOwnerScale=true
   ) {module=ShapeLocation001}
   &CameraOffset(Camera Offset Amount=&MaskFloatBySpawnGroup(
     Masked Float=5
     Passthrough Float=&MaskFloatBySpawnGroup(Masked Float=0, Passthrough Float=10, Spawn Group Mask=5)
     Spawn Group Mask=4
    ))
  ]
  particleUpdateScript [
   &ParticleState()
   &UpdateMeshOrientation(
    Rotation Rate=&MaskFloatBySpawnGroup(Masked Float=0, Passthrough Float=&RandomRangeFloat(Minimum=-1, Maximum=1), Spawn Group Mask=4)
    Rotation Vector=&MaskVectorBySpawnGroup(Masked Vector=<0 0 0>, Passthrough Vector=<0 0 1>, Spawn Group Mask=4)
   )
   &ScaleMeshSize(Scale Factor=&MaskVectorBySpawnGroup(
     Masked Vector=&RandomRangeVector(Minimum=<3 3 30>, Maximum=<4 4 60>)
     Passthrough Vector=&MaskVectorBySpawnGroup(
      Masked Vector=&Multiply_VectorByFloat(Vector=<1 1 1>, Float=0.22)
      Passthrough Vector=&Multiply_VectorByFloat(
       Vector=&RandomRangeVector(Minimum=<1 1 2.2>, Maximum=<2 2 2.2>)
       Float=&FloatFromCurve(
        FloatCurve=DataInterface() {propertyValues {Curve {keys [<at 0 1>,<at 1 1.5>]}}}
        Scale Curve=&RandomRangeFloat(Minimum=0.05, Maximum=0.07)
       )
      )
      Spawn Group Mask=5
     )
     Spawn Group Mask=4
    )) {module=ScaleMeshSize001}
   &ScaleMeshSize(Scale Factor=&MaskVectorBySpawnGroup(
     Masked Vector=&MakeVector(
      X=2
      Y=8
      Z=&FloatFromCurve(FloatCurve=DataInterface() {propertyValues {Curve {keys [
           {interpMode=RCIM_Cubic, leaveTangent=0.0966, tangentMode=RCTM_Break, time=0, value=1}
           {arriveTangent=0.0951, interpMode=RCIM_Cubic, leaveTangent=-0.1745, tangentMode=RCTM_Break, time=1, value=1.07}
          ]}}})
     )
     Passthrough Vector=&MaskVectorBySpawnGroup(Masked Vector=<1 1 1>, Passthrough Vector=<1 1 1>, Spawn Group Mask=5)
     Spawn Group Mask=4
    ))
   &DynamicMaterialParameters(
    Index 0 Param 1=&MakeFloatFromVector(Vector=$Particles.Scale, Channel=::NewEnumerator0)
    Index 0 Param 2=&RandomRangeFloat(
     Minimum=0
     Maximum=&MaskFloatBySpawnGroup(
      Masked Float=10
      Passthrough Float=&MaskFloatBySpawnGroup(Masked Float=10, Passthrough Float=35, Spawn Group Mask=5)
      Spawn Group Mask=4
     )
    )
    Index 0 Param 3=&FloatFromCurve(
     FloatCurve=DataInterface() {propertyValues {Curve {keys [{interpMode=RCIM_Cubic, leaveTangent=8.1085, tangentMode=RCTM_Break, time=0.8568, value=1}, {tangentMode=RCTM_User, time=1, value=0}]}}}
     Scale Curve=0.07
    )
    Index 0 Param 4=&MaskFloatBySpawnGroup(
     Masked Float=0.8
     Passthrough Float=&MaskFloatBySpawnGroup(Masked Float=0.8, Passthrough Float=&RandomRangeFloat(Minimum=0.8, Maximum=1), Spawn Group Mask=5)
     Spawn Group Mask=4
    )
    Index 1 Param 1=&FloatFromCurve(FloatCurve=DataInterface() {propertyValues {Curve {keys [
         {interpMode=RCIM_Cubic, leaveTangent=1.1034, tangentMode=RCTM_Break, time=0, value=0}
         {arriveTangent=2.2399, tangentMode=RCTM_User, time=0.5, value=1}
        ]}}})
    Index 1 Param 2=&FloatFromCurve(FloatCurve=DataInterface() {propertyValues {Curve {keys [<at 0.5 0>, {arriveTangent=3.2053, tangentMode=RCTM_User, time=0.9, value=1}]}}})
    Index 1 Param 3=&FloatFromCurve(FloatCurve=DataInterface() {propertyValues {Curve {keys [<at 0 0 RCIM_Cubic>, {arriveTangent=3.4223, interpMode=RCIM_Cubic, leaveTangent=3.4223, tangentMode=RCTM_User, time=1, value=1}]}}})
    Index 1 Param 4=&FloatFromCurve(
     FloatCurve=DataInterface() {propertyValues {Curve {keys [
         {interpMode=RCIM_Cubic, leaveTangent=0.1169, tangentMode=RCTM_Break, time=0, value=0}
         {arriveTangent=2.5708, interpMode=RCIM_Cubic, leaveTangent=2.5708, tangentMode=RCTM_User, time=1, value=1}
        ]}}}
     Scale Curve=&RandomRangeFloat(
      Minimum=1
      Maximum=&MaskFloatBySpawnGroup(
       Masked Float=15
       Passthrough Float=&MaskFloatBySpawnGroup(Masked Float=40, Passthrough Float=5, Spawn Group Mask=5)
       Spawn Group Mask=4
      )
     )
    )
    Index 2 Param 1=&FloatFromCurve(
     FloatCurve=DataInterface() {propertyValues {Curve {keys [
         {interpMode=RCIM_Cubic, leaveTangent=-0.0025, tangentMode=RCTM_Break, time=0, value=0.5}
         {arriveTangent=-0.0316, interpMode=RCIM_Cubic, tangentMode=RCTM_Break, time=1, value=-0.2}
        ]}}}
     Scale Curve=&MaskFloatBySpawnGroup(
      Masked Float=1
      Passthrough Float=&MaskFloatBySpawnGroup(Masked Float=1, Passthrough Float=-0.5, Spawn Group Mask=5)
      Spawn Group Mask=4
     )
    )
    Index 2 Param 2=&FloatFromCurve(
     FloatCurve=DataInterface() {propertyValues {Curve {keys [
         {interpMode=RCIM_Cubic, tangentMode=RCTM_User, time=0, value=0}
         {arriveTangent=2.0522, interpMode=RCIM_Cubic, leaveTangent=2.0522, tangentMode=RCTM_User, time=1, value=1}
        ]}}}
     Scale Curve=&MaskFloatBySpawnGroup(
      Masked Float=2
      Passthrough Float=&MaskFloatBySpawnGroup(Masked Float=2, Passthrough Float=1, Spawn Group Mask=5)
      Spawn Group Mask=4
     )
    )
    Module.Index0Param0Write=true
    Module.Index0Param1Write=true
    Module.Index0Param2Write=true
    Module.Index0Param3Write=true
    Module.Index1Param0Write=true
    Module.Index1Param1Write=true
    Module.Index1Param2Write=true
    Module.Index1Param3Write=true
    Module.Index2Param0Write=true
    Module.Index2Param1Write=true
    Module.Index2Param2Write=false
    Module.Index2Param3Write=false
    Module.Index3Param0Write=false
    Module.Index3Param1Write=false
    Module.Index3Param2Write=false
    Module.Index3Param3Write=true
    Write Parameter Index 1=true
    Write Parameter Index 2=true
    Write Parameter Index 3=true
   ) {module=DynamicMaterialParameters001}
   &ScaleColor(
    Scale Alpha=&MaskFloatBySpawnGroup(
     Masked Float=&FloatFromCurve(FloatCurve=DataInterface() {propertyValues {Curve {keys [
          {tangentMode=RCTM_User, time=0, value=1}
          {arriveTangent=-0.0405, interpMode=RCIM_Cubic, leaveTangent=-0.0405, tangentMode=RCTM_User, time=0.7009, value=1}
          {arriveTangent=-8.0089, interpMode=RCIM_Cubic, leaveTangent=-8.0089, tangentMode=RCTM_User, time=1, value=0}
         ]}}})
     Passthrough Float=&MaskFloatBySpawnGroup(
      Masked Float=&FloatFromCurve(FloatCurve=DataInterface() {propertyValues {Curve {keys [
           {tangentMode=RCTM_User, time=0, value=1}
           {arriveTangent=-0.0405, interpMode=RCIM_Cubic, leaveTangent=-0.0405, tangentMode=RCTM_User, time=0.7009, value=1}
           {arriveTangent=-8.0089, interpMode=RCIM_Cubic, leaveTangent=-8.0089, tangentMode=RCTM_User, time=1, value=0}
          ]}}})
      Passthrough Float=2
      Spawn Group Mask=5
     )
     Spawn Group Mask=4
    )
    Scale RGB=&Lerp_Vector(
     A=&VectorFromFloat(Value=&FloatFromCurve(
       FloatCurve=DataInterface() {propertyValues {Curve {keys [<at 0 1>,<at 0.1 1>,<at 0.2 2>,<at 0.3 1>,<at 0.6 1>,<at 0.7 2>,<at 0.8 1>,<at 1 1>]}}}
       Scale Curve=&MaskFloatBySpawnGroup(
        Masked Float=25
        Passthrough Float=&MaskFloatBySpawnGroup(Masked Float=10, Passthrough Float=25, Spawn Group Mask=5)
        Spawn Group Mask=4
       )
      ))
     B=&VectorFromFloat(Value=&RandomRangeFloat(Minimum=0.3, Maximum=5, Evaluation Type=::NewEnumerator1))
     Alpha=&FloatFromCurve(FloatCurve=DataInterface() {propertyValues {Curve {keys [
          {interpMode=RCIM_Cubic, leaveTangent=2.5294, tangentMode=RCTM_Break, time=0, value=0}
          {arriveTangent=-0.2166, interpMode=RCIM_Cubic, tangentMode=RCTM_Break, time=0.4, value=1}
         ]}}})
    )
    ScaleA=true
    ScaleRGB=true
   )
  ]
  rendererClasses [&NiagaraMeshRendererProperties]
  renderers [
   {rendererClass=&NiagaraMeshRendererProperties, rendererIndex=0}
   {rendererClass=&NiagaraMeshRendererProperties, rendererIndex=1}
   {rendererClass=&NiagaraMeshRendererProperties, rendererIndex=2}
   {rendererClass=&NiagaraMeshRendererProperties, rendererIndex=3}
   {rendererClass=&NiagaraMeshRendererProperties, rendererIndex=4}
   {rendererClass=&NiagaraMeshRendererProperties, rendererIndex=5}
  ]
 }
 LightningSparks001 {
  bEnabled=false
  emitterUpdateScript [
   &EmitterState(Life Cycle Mode=::NewEnumerator1, Loop Behavior=::NewEnumerator1, Loop Duration=1, UseLoopDelay=false)
   &SpawnBurst_Instantaneous(
    Spawn Count=&RandomRangeInt(Minimum=15, Maximum=25)
    Use Spawn Probability=false
    UseLoopCountLimit=false
   )
  ]
  particleSpawnScript [
   &InitializeParticle(
    AdjustAlpha=false
    AdjustHue=true
    AdjustSaturation=false
    AdjustValue=true
    Color=$System.ElimColor
    Color Mode=::NewEnumerator3
    Hue Shift Range:Vector2f=<-0.01 0.01>
    Lifetime Max=1
    Lifetime Min=0.1
    Lifetime Mode=::NewEnumerator1
    Mass Max=2
    Mass Mode=::NewEnumerator3
    Sprite Size Mode=::NewEnumerator1
    Uniform Sprite Size Max=5
    Uniform Sprite Size Min=1
    UsePositionOffset=false
    Value Range:Vector2f=<0.8 25>
    Write Material Random=true
   )
   &AddVelocity(
    Cone Angle=95
    Cone Axis=<0 0 1>
    Distribution Along Cone Axis=0.2
    Module.UseConeFalloff=false
    Module.UseVelDistribution=true
    Override Seed=false
    Velocity Mode=::NewEnumerator2
    Velocity Speed=&RandomRangeFloat(Minimum=150, Maximum=750)
   )
  ]
  particleUpdateScript [
   &ParticleState()
   &GravityForce(Gravity=<0 0 -700>)
   &Drag(Drag=&RandomRangeFloat(Minimum=0.8, Maximum=1.2), Use Linear Drag=true, Use Rotational Drag=false)
   &ScaleColor(
    Scale Alpha=&FloatFromCurve(FloatCurve=DataInterface() {propertyValues {Curve {keys [
         {arriveTangent=9.0725, interpMode=RCIM_Cubic, leaveTangent=9.0725, tangentMode=RCTM_User, time=0, value=0}
         {arriveTangent=-2.5249, interpMode=RCIM_Cubic, leaveTangent=-2.5249, tangentMode=RCTM_User, time=0.1408, value=0.9847}
         {arriveTangent=-0.5654, interpMode=RCIM_Cubic, leaveTangent=-0.5654, time=0.5, value=0.35}
         {arriveTangent=-1.25, interpMode=RCIM_Cubic, leaveTangent=-1.25, time=0.6, value=0.725}
         {arriveTangent=-3.375, interpMode=RCIM_Cubic, leaveTangent=-0.1412, tangentMode=RCTM_Break, time=0.7, value=0.1}
         {arriveTangent=-0.7119, interpMode=RCIM_Cubic, leaveTangent=1.625, tangentMode=RCTM_Break, time=0.8, value=0.05}
         {arriveTangent=-0.25, interpMode=RCIM_Cubic, leaveTangent=-0.25, time=0.9, value=0.425}
         {arriveTangent=-0.0405, interpMode=RCIM_Cubic, leaveTangent=-0.0405, tangentMode=RCTM_User, time=1, value=0}
        ]}}})
    Scale RGB=&VectorFromFloat(Value=&FloatFromCurve(FloatCurve=DataInterface() {propertyValues {Curve {keys [<at 0 1>,<at 0.2 1>,<at 0.3 5>,<at 0.4 1>,<at 0.7 1>,<at 0.8 5>,<at 0.9 1>,<at 1 1>]}}}))
    ScaleA=true
    ScaleRGB=true
   ) {module=ScaleColor001}
   &SolveForcesAndVelocity(Clamp Velocity=false, Limit Acceleration=false)
   &ScaleSpriteSizeBySpeed(Max Scale Factor=<0.5 7>, Sample Scale Factor By Curve=true, Velocity Threshold=500)
  ]
  rendererClasses [&NiagaraSpriteRendererProperties]
  renderers [{rendererClass=&NiagaraSpriteRendererProperties}]
 }
 Lightning_Ground_Star001 {
  bEnabled=false
  emitterUpdateScript [
   &EmitterState()
   &SpawnBurst_Instantaneous(Spawn Count=10, Use Spawn Probability=false, UseLoopCountLimit=false)
   &SpawnRate(SpawnRate=&Multiply_Float(A=15, B=$System.FadeOutBottomLightning), Use Spawn Probability=false)
  ]
  particleSpawnScript [
   &InitializeParticle(
    Color=&Multiply_LinearColorByFloat(Linear Color=$System.ElimColor, Float=0.4)
    Lifetime Max=0.35
    Lifetime Min=0.175
    Lifetime Mode=::NewEnumerator1
    Mesh Scale Max=<1.4 1.4 2.2>
    Mesh Scale Min=<1.2 1.2 1.5>
    Mesh Scale Mode=::NewEnumerator2
    Position Offset=&RandomRangeVector(Minimum=<0 0 15>, Maximum=<0 0 25>)
    UsePositionOffset=true
    Write Material Random=true
   )
   &InitialMeshOrientation(
    Mesh Axis To Orient=<0 0 1>
    Mesh Orientation Mode=::NewEnumerator2
    Rotation=&RandomRangeVector(Minimum=<0 0 0.1>, Maximum=<1 0 0.3>)
    Use Rotation Vector=true
   )
  ]
  particleUpdateScript [
   &SolveForcesAndVelocity(Clamp Velocity=false, Limit Acceleration=false)
   &ParticleState()
   &ScaleMeshSize(Scale Factor=&Multiply_VectorByFloat(Vector=<1 1 1>, Float=0.22))
   &DynamicMaterialParameters(
    Index 0 Param 1=&MakeFloatFromVector(Vector=$Particles.Scale, Channel=::NewEnumerator0)
    Index 0 Param 2=&RandomRangeFloat(Minimum=0, Maximum=10)
    Index 0 Param 3=&FloatFromCurve(
     FloatCurve=DataInterface() {propertyValues {Curve {keys [{interpMode=RCIM_Cubic, leaveTangent=8.1085, tangentMode=RCTM_Break, time=0.8568, value=1}, {tangentMode=RCTM_User, time=1, value=0}]}}}
     Scale Curve=0.07
    )
    Index 0 Param 4=0.8
    Index 1 Param 1=&FloatFromCurve(FloatCurve=DataInterface() {propertyValues {Curve {keys [
         {interpMode=RCIM_Cubic, leaveTangent=1.1034, tangentMode=RCTM_Break, time=0, value=0}
         {arriveTangent=2.2399, tangentMode=RCTM_User, time=0.5, value=1}
        ]}}})
    Index 1 Param 2=&FloatFromCurve(FloatCurve=DataInterface() {propertyValues {Curve {keys [<at 0.5 0>, {arriveTangent=3.2053, tangentMode=RCTM_User, time=0.9, value=1}]}}})
    Index 1 Param 3=&FloatFromCurve(FloatCurve=DataInterface() {propertyValues {Curve {keys [<at 0 0 RCIM_Cubic>, {arriveTangent=3.4223, interpMode=RCIM_Cubic, leaveTangent=3.4223, tangentMode=RCTM_User, time=1, value=1}]}}})
    Index 1 Param 4=&FloatFromCurve(
     FloatCurve=DataInterface() {propertyValues {Curve {keys [
         {interpMode=RCIM_Cubic, leaveTangent=0.1169, tangentMode=RCTM_Break, time=0, value=0}
         {arriveTangent=2.5708, interpMode=RCIM_Cubic, leaveTangent=2.5708, tangentMode=RCTM_User, time=1, value=1}
        ]}}}
     Scale Curve=&RandomRangeFloat(Minimum=1, Maximum=40)
    )
    Index 2 Param 1=&FloatFromCurve(FloatCurve=DataInterface() {propertyValues {Curve {keys [
         {interpMode=RCIM_Cubic, leaveTangent=-0.0025, tangentMode=RCTM_Break, time=0, value=0.5}
         {arriveTangent=-0.0316, interpMode=RCIM_Cubic, tangentMode=RCTM_Break, time=1, value=-0.2}
        ]}}})
    Index 2 Param 2=&FloatFromCurve(
     FloatCurve=DataInterface() {propertyValues {Curve {keys [
         {interpMode=RCIM_Cubic, tangentMode=RCTM_User, time=0, value=0}
         {arriveTangent=2.0522, interpMode=RCIM_Cubic, leaveTangent=2.0522, tangentMode=RCTM_User, time=1, value=1}
        ]}}}
     Scale Curve=2
    )
    Module.Index0Param0Write=true
    Module.Index0Param1Write=true
    Module.Index0Param2Write=true
    Module.Index0Param3Write=true
    Module.Index1Param0Write=true
    Module.Index1Param1Write=true
    Module.Index1Param2Write=true
    Module.Index1Param3Write=true
    Module.Index2Param0Write=true
    Module.Index2Param1Write=true
    Module.Index2Param2Write=false
    Module.Index2Param3Write=false
    Module.Index3Param0Write=false
    Module.Index3Param1Write=false
    Module.Index3Param2Write=false
    Module.Index3Param3Write=true
    Write Parameter Index 1=true
    Write Parameter Index 2=true
    Write Parameter Index 3=true
   )
   &UpdateMeshOrientation(Rotation Rate=&RandomRangeFloat(Minimum=-1, Maximum=1), Rotation Vector=<0 0 1>)
   &ScaleColor(
    Scale Alpha=&FloatFromCurve(FloatCurve=DataInterface() {propertyValues {Curve {keys [
         {tangentMode=RCTM_User, time=0, value=1}
         {arriveTangent=-0.0405, interpMode=RCIM_Cubic, leaveTangent=-0.0405, tangentMode=RCTM_User, time=0.7009, value=1}
         {arriveTangent=-8.0089, interpMode=RCIM_Cubic, leaveTangent=-8.0089, tangentMode=RCTM_User, time=1, value=0}
        ]}}})
    Scale RGB=&Lerp_Vector(
     A=&VectorFromFloat(Value=&FloatFromCurve(
       FloatCurve=DataInterface() {propertyValues {Curve {keys [<at 0 1>,<at 0.1 1>,<at 0.2 2>,<at 0.3 1>,<at 0.6 1>,<at 0.7 2>,<at 0.8 1>,<at 1 1>]}}}
       Scale Curve=10
      ))
     B=&VectorFromFloat(Value=&RandomRangeFloat(Minimum=0.3, Maximum=5, Evaluation Type=::NewEnumerator1))
     Alpha=&FloatFromCurve(FloatCurve=DataInterface() {propertyValues {Curve {keys [
          {interpMode=RCIM_Cubic, leaveTangent=2.5294, tangentMode=RCTM_Break, time=0, value=0}
          {arriveTangent=-0.2166, interpMode=RCIM_Cubic, tangentMode=RCTM_Break, time=0.4, value=1}
         ]}}})
    )
    ScaleA=true
    ScaleRGB=true
   )
  ]
  rendererClasses [&NiagaraMeshRendererProperties]
  renderers [
   {rendererClass=&NiagaraMeshRendererProperties, rendererIndex=0}
   {rendererClass=&NiagaraMeshRendererProperties, rendererIndex=1}
  ]
 }
 NE_GroundPulse001 {
  bEnabled=false
  emitterUpdateScript [
   &EmitterState()
   &SpawnBurst_Instantaneous(Use Spawn Probability=false, UseLoopCountLimit=false)
   &SpawnRate(SpawnRate=6, Use Spawn Probability=false) {enabled=false}
  ]
  particleSpawnScript [
   &PartitionParticles() {enabled=false}
   &InitializeParticle(
    Color=<rgba 0 0.9974 1 0.2>
    Lifetime=0.75
    Mesh Scale=<0.02 0.6 0.6>
    Mesh Scale Mode=::NewEnumerator4
    Position Offset Coordinate Space=::Local
    UsePositionOffset=true
    Write Material Random=true
   )
   &InitialMeshOrientation(
    Mesh Axis To Orient=<0 0 1>
    Mesh Orientation Mode=::NewEnumerator2
    Rotation=&MakeVector(X=&RandomRangeFloat(Minimum=-1, Maximum=1), Y=0, Z=0)
    Use Rotation Vector=true
   )
   &SkeletalMeshLocation(
    Bone / Socket Index=&SelectIntFromArray(Direct Array Index=$Particles.PartitionParticles.Partition, Int Selection Array=DataInterface() {propertyValues {IntData [0,1]}})
    Bone Sampling Mode=::NewEnumerator4
    MaskSpawnGroup=false
    Orientation Sampling=::NewEnumerator1
    Sampled Position Offset=&SelectVectorFromArray(
     Direct Array Index=$Particles.PartitionParticles.Partition
     Vector Selection Array=DataInterface() {propertyValues {FloatData [<3 8 -8>, <-3 8 -12>]}}
    )
    Sampled Position Offset Coordinate Space:ENiagaraSkelMeshTransforms=::NewEnumerator2
    Skeletal Mesh=$System.SkeletalMesh
   ) {enabled=false}
   &CameraOffset(Camera Offset Amount=10)
   &DynamicMaterialParameters(
    Index 0 Param 1=0.6
    Index 0 Param 2=-0.05
    Module.Index0Float0Write=true
    Module.Index0Float1Write=true
    Module.Index0Float2Write=true
    Module.Index0Float3Write=true
   )
  ]
  particleUpdateScript [
   &ParticleState()
   &ScaleMeshSize(Scale Factor=&VectorFromFloat(Value=&FloatFromCurve(FloatCurve=DataInterface() {propertyValues {Curve {keys [
          {arriveTangent=4.1785, interpMode=RCIM_Cubic, leaveTangent=4.1785, tangentMode=RCTM_User, time=0, value=0}
          {arriveTangent=1.1179, interpMode=RCIM_Cubic, leaveTangent=1.1179, tangentMode=RCTM_User, time=0.2042, value=0.6228}
          {arriveTangent=0.0224, interpMode=RCIM_Cubic, leaveTangent=0.0224, tangentMode=RCTM_User, time=1, value=1}
         ]}}})))
   &ScaleColor(
    Scale Alpha=&Multiply_Float(
     A=&FloatFromCurve(FloatCurve=DataInterface() {propertyValues {Curve {keys [
          {arriveTangent=7.3302, interpMode=RCIM_Cubic, leaveTangent=7.3302, tangentMode=RCTM_User, time=0, value=0}
          {arriveTangent=122.0485, interpMode=RCIM_Cubic, leaveTangent=-26.0227, tangentMode=RCTM_Break, time=0.05, value=8}
          {arriveTangent=-0.6259, tangentMode=RCTM_User, time=1, value=0}
         ]}}})
     B=&Multiply_Float(A=&FNTimeOfDayScalar(Night Scale=1, Morning Scale=0.8, Evening Scale=0.8), B=1)
    )
    Scale RGB=&MakeVectorFromLinearColorRGB(Color=&MultiplyLinearColor(
      Linear Color=&ColorFromCurve(ColorCurve=DataInterface() {propertyValues {
         AlphaCurve {keys []}
         BlueCurve {keys [<at 0.0696 3>,<at 0.554 1>]}
         GreenCurve {keys [<at 0.0696 3>,<at 0.554 1>]}
         RedCurve {keys [<at 0.0696 3>,<at 0.554 1>]}
        }})
      Scale Factor=&MakeVector4_FromLinearColor(LinearColor=&SelectLinearColorFromArray(
        Array Sampling Mode=::NewEnumerator1
        Direct Array Index=$System.RankIndex
        Color Selection Array=DataInterface() {propertyValues {ColorData [
           <rgba 1 0.317 0.0677 1>
           <rgba 0.5781 0.5781 0.5781 1>
           <rgba 1 0.4465 0 1>
           <rgba 0 0.7573 1 1>
           <rgba 0 0.3392 1.5 1>
           <rgba 0.3267 0.6458 0.5478 1>
           <rgba 1 0.621 0 1>
           <rgba 0 0.4424 1.5 1>
          ]}}
        Random Seed=?
       ))
     ))
    ScaleA=true
    ScaleRGB=false
   )
  ]
  rendererClasses [&NiagaraMeshRendererProperties]
  renderers [{rendererClass=&NiagaraMeshRendererProperties, rendererIndex=0}]
 }
 NE_Lightning_Streak {
  bEnabled=false
  emitterUpdateScript [
   &EmitterState()
   &SpawnBurst_Instantaneous(
    Spawn Count=&RandomRangeInt(Minimum=8, Maximum=12)
    Use Spawn Probability=false
    UseLoopCountLimit=false
   )
  ]
  particleSpawnScript [
   &InitializeParticle(
    Color=&Multiply_LinearColorByFloat(Linear Color=$System.ElimColor, Float=0.4)
    Lifetime=&RandomRangeFloat(Minimum=0.2, Maximum=0.3)
    Mesh Scale Mode=::NewEnumerator3
    Mesh Uniform Scale=&RandomRangeFloat(Minimum=0.01, Maximum=0.03)
    UsePositionOffset=true
    Write Material Random=true
   )
   &InitialMeshOrientation(
    Mesh Orientation Mode=::NewEnumerator2
    Rotation=&RandomRangeVector(Minimum=<-0.1 -0.1 0>, Maximum=<0.1 0.1 0>)
    Use Rotation Vector=true
   )
   &ShapeLocation(MaskSpawnGroup=false, Override Seed=false, Sphere Radius=10, UseOwnerScale=true)
   &CameraOffset(Camera Offset Amount=5)
  ]
  particleUpdateScript [
   &ParticleState()
   &ScaleMeshSize(Scale Factor=&RandomRangeVector(Minimum=<3 3 30>, Maximum=<4 4 60>)) {module=ScaleMeshSize001}
   &ScaleMeshSize(Scale Factor=&MakeVector(
     X=2
     Y=8
     Z=&FloatFromCurve(FloatCurve=DataInterface() {propertyValues {Curve {keys [
          {interpMode=RCIM_Cubic, leaveTangent=0.0966, tangentMode=RCTM_Break, time=0, value=1}
          {arriveTangent=0.0951, interpMode=RCIM_Cubic, leaveTangent=-0.1745, tangentMode=RCTM_Break, time=1, value=1.07}
         ]}}})
    ))
   &DynamicMaterialParameters(
    Index 0 Param 1=&MakeFloatFromVector(Vector=$Particles.Scale, Channel=::NewEnumerator0)
    Index 0 Param 2=&RandomRangeFloat(Minimum=0, Maximum=10)
    Index 0 Param 3=&FloatFromCurve(
     FloatCurve=DataInterface() {propertyValues {Curve {keys [{interpMode=RCIM_Cubic, leaveTangent=8.1085, tangentMode=RCTM_Break, time=0.8568, value=1}, {tangentMode=RCTM_User, time=1, value=0}]}}}
     Scale Curve=0.07
    )
    Index 0 Param 4=0.8
    Index 1 Param 1=&FloatFromCurve(FloatCurve=DataInterface() {propertyValues {Curve {keys [
         {interpMode=RCIM_Cubic, leaveTangent=1.1034, tangentMode=RCTM_Break, time=0, value=0}
         {arriveTangent=2.2399, tangentMode=RCTM_User, time=0.5, value=1}
        ]}}})
    Index 1 Param 2=&FloatFromCurve(FloatCurve=DataInterface() {propertyValues {Curve {keys [<at 0.5 0>, {arriveTangent=3.2053, tangentMode=RCTM_User, time=0.9, value=1}]}}})
    Index 1 Param 3=&FloatFromCurve(FloatCurve=DataInterface() {propertyValues {Curve {keys [<at 0 0 RCIM_Cubic>, {arriveTangent=3.4223, interpMode=RCIM_Cubic, leaveTangent=3.4223, tangentMode=RCTM_User, time=1, value=1}]}}})
    Index 1 Param 4=&FloatFromCurve(
     FloatCurve=DataInterface() {propertyValues {Curve {keys [
         {interpMode=RCIM_Cubic, leaveTangent=0.1169, tangentMode=RCTM_Break, time=0, value=0}
         {arriveTangent=2.5708, interpMode=RCIM_Cubic, leaveTangent=2.5708, tangentMode=RCTM_User, time=1, value=1}
        ]}}}
     Scale Curve=&RandomRangeFloat(Minimum=1, Maximum=15)
    )
    Index 2 Param 1=&FloatFromCurve(FloatCurve=DataInterface() {propertyValues {Curve {keys [
         {interpMode=RCIM_Cubic, leaveTangent=-0.0025, tangentMode=RCTM_Break, time=0, value=0.5}
         {arriveTangent=-0.0316, interpMode=RCIM_Cubic, tangentMode=RCTM_Break, time=1, value=-0.2}
        ]}}})
    Index 2 Param 2=&FloatFromCurve(
     FloatCurve=DataInterface() {propertyValues {Curve {keys [
         {interpMode=RCIM_Cubic, tangentMode=RCTM_User, time=0, value=0}
         {arriveTangent=2.0522, interpMode=RCIM_Cubic, leaveTangent=2.0522, tangentMode=RCTM_User, time=1, value=1}
        ]}}}
     Scale Curve=2
    )
    Module.Index0Param0Write=true
    Module.Index0Param1Write=true
    Module.Index0Param2Write=true
    Module.Index0Param3Write=true
    Module.Index1Param0Write=true
    Module.Index1Param1Write=true
    Module.Index1Param2Write=true
    Module.Index1Param3Write=true
    Module.Index2Param0Write=true
    Module.Index2Param1Write=true
    Module.Index2Param2Write=false
    Module.Index2Param3Write=false
    Module.Index3Param0Write=false
    Module.Index3Param1Write=false
    Module.Index3Param2Write=false
    Module.Index3Param3Write=true
    Write Parameter Index 1=true
    Write Parameter Index 2=true
    Write Parameter Index 3=true
   ) {module=DynamicMaterialParameters001}
   &ScaleColor(
    Scale Alpha=&FloatFromCurve(FloatCurve=DataInterface() {propertyValues {Curve {keys [
         {tangentMode=RCTM_User, time=0, value=1}
         {arriveTangent=-0.0405, interpMode=RCIM_Cubic, leaveTangent=-0.0405, tangentMode=RCTM_User, time=0.7009, value=1}
         {arriveTangent=-8.0089, interpMode=RCIM_Cubic, leaveTangent=-8.0089, tangentMode=RCTM_User, time=1, value=0}
        ]}}})
    Scale RGB=&Lerp_Vector(
     A=&VectorFromFloat(Value=&FloatFromCurve(
       FloatCurve=DataInterface() {propertyValues {Curve {keys [<at 0 1>,<at 0.1 1>,<at 0.2 2>,<at 0.3 1>,<at 0.6 1>,<at 0.7 2>,<at 0.8 1>,<at 1 1>]}}}
       Scale Curve=25
      ))
     B=&VectorFromFloat(Value=&RandomRangeFloat(Minimum=0.3, Maximum=5, Evaluation Type=::NewEnumerator1))
     Alpha=&FloatFromCurve(FloatCurve=DataInterface() {propertyValues {Curve {keys [
          {interpMode=RCIM_Cubic, leaveTangent=2.5294, tangentMode=RCTM_Break, time=0, value=0}
          {arriveTangent=-0.2166, interpMode=RCIM_Cubic, tangentMode=RCTM_Break, time=0.4, value=1}
         ]}}})
    )
    ScaleA=true
    ScaleRGB=true
   )
  ]
  rendererClasses [&NiagaraMeshRendererProperties]
  renderers [
   {rendererClass=&NiagaraMeshRendererProperties, rendererIndex=0}
   {rendererClass=&NiagaraMeshRendererProperties, rendererIndex=1}
  ]
 }
 NE_LoopingZaps001 {
  bEnabled=false
  emitterUpdateScript [
   &EmitterState()
   &SpawnRate(SpawnRate=&Multiply_Float(A=15, B=$System.FadeOutAura), Use Spawn Probability=false)
   &SpawnBurst_Instantaneous(Spawn Count=4, Spawn Group=1, Use Spawn Probability=false, UseLoopCountLimit=false)
  ]
  particleSpawnScript [
   &InitializeParticle(
    Color=&Multiply_LinearColorByFloat(Linear Color=$System.ElimColor, Float=0.4)
    Lifetime=&RandomRangeFloat(Minimum=0.2, Maximum=0.3)
    Mesh Scale Mode=::NewEnumerator3
    Mesh Uniform Scale=&RandomRangeFloat(Minimum=0.6, Maximum=0.6)
    Sprite Size Mode=::NewEnumerator3
    Uniform Sprite Size=16
    UsePositionOffset=false
    Write Material Random=true
   )
   &InitialMeshOrientation(
    Mesh Axis To Orient=<1 1 0>
    Mesh Orientation Mode=::NewEnumerator2
    Rotation=&RandomRangeVector(Minimum=<0 0 0>, Maximum=<1 1 1>)
    Use Rotation Vector=true
   )
   &CameraOffset(Camera Offset Amount=10)
   &ShapeLocation(Box Size=<75 75 5>, MaskSpawnGroup=false, Override Seed=false, Shape Primitive=::NewEnumerator2, Surface Only=false, UseOwnerScale=true)
  ]
  particleUpdateScript [
   &ParticleState()
   &ScaleMeshSize(Scale Factor=&Multiply_VectorByFloat(
     Vector=&RandomRangeVector(Minimum=<1 1 2.2>, Maximum=<2 2 2.2>)
     Float=&FloatFromCurve(
      FloatCurve=DataInterface() {propertyValues {Curve {keys [<at 0 1>,<at 1 1.5>]}}}
      Scale Curve=&RandomRangeFloat(Minimum=0.05, Maximum=0.07)
     )
    )) {module=ScaleMeshSize001}
   &DynamicMaterialParameters(
    Index 0 Param 2=&RandomRangeFloat(Minimum=0, Maximum=0)
    Index 0 Param 4=3
    Index 1 Param 1=&FloatFromCurve(FloatCurve=DataInterface() {propertyValues {Curve {keys [<at 0 0.5>,<at 1 1>]}}})
    Index 1 Param 2=2
    Index 1 Param 3=&FloatFromCurve(FloatCurve=DataInterface() {propertyValues {Curve {keys [{interpMode=RCIM_Cubic, tangentMode=RCTM_Break, time=0.5, value=-0.5}, <at 1 0.1>]}}})
    Index 2 Param 1=&Multiply_Float(
     A=&FloatFromCurve(
      FloatCurve=DataInterface() {propertyValues {Curve {keys [{interpMode=RCIM_Cubic, leaveTangent=8.1085, tangentMode=RCTM_Break, time=0.8568, value=1}, {tangentMode=RCTM_User, time=1, value=0}]}}}
      Scale Curve=3.4185
     )
     B=&Divide_Float(
      A=&MakeFloatFromVector(Vector=$Particles.Scale, Channel=::NewEnumerator0)
      B=&RandomRangeFloat(Minimum=10, Maximum=15)
     )
    )
    Index 2 Param 2=&FloatFromCurve(
     FloatCurve=DataInterface() {propertyValues {Curve {keys [
         {interpMode=RCIM_Cubic, tangentMode=RCTM_Break, time=0.7, value=0}
         {arriveTangent=-0.0859, interpMode=RCIM_Cubic, tangentMode=RCTM_Break, time=1, value=-1}
        ]}}}
     Scale Curve=50
    )
    Index 2 Param 3=&Waveform(
     Global Input Position=&RandomRangeFloat(Minimum=0, Maximum=15)
     Global Phase=$Particles.Age
     "Waveform [1]"=::NewEnumerator12
     "[1] Phase Scale"=2
     "[1] Frequency"=1
     "[1] Amplitude Min/Max"=<0 1>
    )
    Index 2 Param 4=25
    Index 3 Param 1=&FloatFromCurve(
     FloatCurve=DataInterface() {propertyValues {Curve {keys [
         {interpMode=RCIM_Cubic, leaveTangent=-0.0025, tangentMode=RCTM_Break, time=0, value=0}
         {arriveTangent=1.8528, interpMode=RCIM_Cubic, tangentMode=RCTM_Break, time=1, value=1}
        ]}}}
     Scale Curve=&RandomRangeFloat(Minimum=0, Maximum=0)
    )
    Index 3 Param 2=&FloatFromCurve(
     FloatCurve=DataInterface() {propertyValues {Curve {keys [
         {interpMode=RCIM_Cubic, leaveTangent=-0.0025, tangentMode=RCTM_Break, time=0, value=0.5}
         {arriveTangent=-0.0316, interpMode=RCIM_Cubic, tangentMode=RCTM_Break, time=1, value=-0.2}
        ]}}}
     Scale Curve=0.1
    )
    Index 3 Param 3=&FloatFromCurve(FloatCurve=DataInterface() {propertyValues {Curve {keys [
         {interpMode=RCIM_Cubic, leaveTangent=1.1034, tangentMode=RCTM_Break, time=0, value=-1}
         {arriveTangent=2.2399, tangentMode=RCTM_User, time=0.7, value=0.125}
        ]}}})
    Index 3 Param 4=&FloatFromCurve(FloatCurve=DataInterface() {propertyValues {Curve {keys [
         {interpMode=RCIM_Cubic, leaveTangent=3.0601, tangentMode=RCTM_Break, time=0.5, value=0}
         {arriveTangent=3.2053, tangentMode=RCTM_User, time=1, value=1}
        ]}}})
    Module.Index0Param0Write=true
    Module.Index0Param1Write=true
    Module.Index0Param2Write=true
    Module.Index0Param3Write=true
    Module.Index1Param0Write=true
    Module.Index1Param1Write=true
    Module.Index1Param2Write=true
    Module.Index1Param3Write=true
    Module.Index2Param0Write=true
    Module.Index2Param1Write=true
    Module.Index2Param2Write=true
    Module.Index2Param3Write=true
    Module.Index3Param0Write=true
    Module.Index3Param1Write=true
    Module.Index3Param2Write=true
    Module.Index3Param3Write=true
    Write Parameter Index 1=true
    Write Parameter Index 2=true
    Write Parameter Index 3=true
   ) {enabled=false}
   &DynamicMaterialParameters(
    Index 0 Param 1=&MakeFloatFromVector(Vector=$Particles.Scale, Channel=::NewEnumerator0)
    Index 0 Param 2=&RandomRangeFloat(Minimum=0, Maximum=35)
    Index 0 Param 3=&FloatFromCurve(
     FloatCurve=DataInterface() {propertyValues {Curve {keys [{interpMode=RCIM_Cubic, leaveTangent=8.1085, tangentMode=RCTM_Break, time=0.8568, value=1}, {tangentMode=RCTM_User, time=1, value=0}]}}}
     Scale Curve=0.07
    )
    Index 0 Param 4=&RandomRangeFloat(Minimum=0.8, Maximum=1)
    Index 1 Param 1=&FloatFromCurve(FloatCurve=DataInterface() {propertyValues {Curve {keys [
         {interpMode=RCIM_Cubic, leaveTangent=1.1034, tangentMode=RCTM_Break, time=0, value=0}
         {arriveTangent=2.2399, tangentMode=RCTM_User, time=0.5, value=1}
        ]}}})
    Index 1 Param 2=&FloatFromCurve(FloatCurve=DataInterface() {propertyValues {Curve {keys [{tangentMode=RCTM_User, time=0, value=0}, {tangentMode=RCTM_User, time=1, value=1}]}}})
    Index 1 Param 3=&FloatFromCurve(FloatCurve=DataInterface() {propertyValues {Curve {keys [
         {interpMode=RCIM_Cubic, tangentMode=RCTM_User, time=0, value=0}
         {arriveTangent=2.7163, interpMode=RCIM_Cubic, leaveTangent=2.7163, tangentMode=RCTM_User, time=1, value=1}
        ]}}})
    Index 1 Param 4=&FloatFromCurve(
     FloatCurve=DataInterface() {propertyValues {Curve {keys [
         {interpMode=RCIM_Cubic, leaveTangent=0.1169, tangentMode=RCTM_Break, time=0, value=0}
         {arriveTangent=2.5708, interpMode=RCIM_Cubic, leaveTangent=2.5708, tangentMode=RCTM_User, time=1, value=1}
        ]}}}
     Scale Curve=&RandomRangeFloat(Minimum=1, Maximum=5)
    )
    Index 2 Param 1=&FloatFromCurve(
     FloatCurve=DataInterface() {propertyValues {Curve {keys [
         {interpMode=RCIM_Cubic, leaveTangent=-0.0025, tangentMode=RCTM_Break, time=0, value=0.5}
         {arriveTangent=-0.0316, interpMode=RCIM_Cubic, tangentMode=RCTM_Break, time=1, value=-0.2}
        ]}}}
     Scale Curve=-0.5
    )
    Index 2 Param 2=&FloatFromCurve(FloatCurve=DataInterface() {propertyValues {Curve {keys [
         {interpMode=RCIM_Cubic, tangentMode=RCTM_User, time=0, value=0}
         {arriveTangent=2.0522, interpMode=RCIM_Cubic, leaveTangent=2.0522, tangentMode=RCTM_User, time=1, value=1}
        ]}}})
    Module.Index0Param0Write=true
    Module.Index0Param1Write=true
    Module.Index0Param2Write=true
    Module.Index0Param3Write=true
    Module.Index1Param0Write=true
    Module.Index1Param1Write=true
    Module.Index1Param2Write=true
    Module.Index1Param3Write=true
    Module.Index2Param0Write=true
    Module.Index2Param1Write=true
    Module.Index2Param2Write=false
    Module.Index2Param3Write=false
    Module.Index3Param0Write=false
    Module.Index3Param1Write=false
    Module.Index3Param2Write=false
    Module.Index3Param3Write=true
    Write Parameter Index 1=true
    Write Parameter Index 2=true
    Write Parameter Index 3=true
   ) {module=DynamicMaterialParameters001}
   &ScaleColor(
    Scale Alpha=2
    Scale RGB=&Lerp_Vector(
     A=&VectorFromFloat(Value=&FloatFromCurve(
       FloatCurve=DataInterface() {propertyValues {Curve {keys [<at 0 1>,<at 0.1 1>,<at 0.2 2>,<at 0.3 1>,<at 0.6 1>,<at 0.7 2>,<at 0.8 1>,<at 1 1>]}}}
       Scale Curve=25
      ))
     B=&VectorFromFloat(Value=&RandomRangeFloat(Minimum=0.3, Maximum=5, Evaluation Type=::NewEnumerator1))
     Alpha=&FloatFromCurve(FloatCurve=DataInterface() {propertyValues {Curve {keys [
          {interpMode=RCIM_Cubic, leaveTangent=2.5294, tangentMode=RCTM_Break, time=0, value=0}
          {arriveTangent=-0.2166, interpMode=RCIM_Cubic, tangentMode=RCTM_Break, time=0.4, value=1}
         ]}}})
    )
    ScaleA=true
    ScaleRGB=true
   )
   &SolveForcesAndVelocity(Clamp Velocity=false, Limit Acceleration=false)
  ]
  rendererClasses [&NiagaraMeshRendererProperties]
  renderers [
   {rendererClass=&NiagaraMeshRendererProperties, rendererIndex=0}
   {rendererClass=&NiagaraMeshRendererProperties, rendererIndex=1}
  ]
 }
 NE_StarGlow {
  bEnabled=false
  emitterUpdateScript [
   &EmitterState()
   &SpawnBurst_Instantaneous(Use Spawn Probability=false, UseLoopCountLimit=false)
  ]
  particleSpawnScript [
   &InitializeParticle(
    Color=&Multiply_LinearColorByFloat(Linear Color=$System.ElimColor, Float=2)
    Lifetime=0.4
    Mass Mode=::NewEnumerator3
    Mesh Scale Mode=::NewEnumerator3
    Mesh Uniform Scale=2.5
    Sprite Size Mode=::NewEnumerator3
    Uniform Sprite Size=250
    UsePositionOffset=true
    Write Material Random=true
   )
   &ShapeLocation(MaskSpawnGroup=false, Non Uniform Scale=<0.5 1 1>, Override Seed=false, Sphere Radius=0, UseOwnerScale=true)
   /SetVariables(Particles.SpriteAlignment=<1 0 0.4>) {isSetParameters=true, module=SetVariables_#6}
   &InitialMeshOrientation(Orientation Axis=<0 0 1>, Orientation Vector=<0 0 1>, Rotation=<0 0.25 0>, Use Orientation Vector=true, Use Rotation Vector=true)
   &InitialMeshOrientation(
    Orientation Axis=<0 0 0>
    Orientation Vector=<0 0 0>
    Rotation=&RandomRangeVector(Minimum=<0 0 0>, Maximum=<1 0 0>)
    Use Orientation Vector=false
    Use Rotation Vector=true
   ) {module=InitialMeshOrientation001}
  ]
  particleUpdateScript [
   &ParticleState()
   &SpriteRotationRate(Rotation Rate=&FloatFromCurve(
     FloatCurve=DataInterface() {propertyValues {Curve {keys [
         {arriveTangent=1.4306, interpMode=RCIM_Cubic, leaveTangent=1.4306, tangentMode=RCTM_User, time=0.2, value=0}
         {arriveTangent=1.1179, interpMode=RCIM_Cubic, leaveTangent=1.1179, tangentMode=RCTM_User, time=0.4, value=0.6228}
         {arriveTangent=0.0224, interpMode=RCIM_Cubic, leaveTangent=0.0224, tangentMode=RCTM_User, time=1, value=1}
        ]}}}
     Scale Curve=0
    ))
   &ScaleColor(
    Scale Alpha=&FloatFromCurve(FloatCurve=DataInterface() {propertyValues {Curve {keys [
         {arriveTangent=-1.5276, interpMode=RCIM_Cubic, leaveTangent=-1.5276, tangentMode=RCTM_User, time=0, value=1}
         {interpMode=RCIM_Cubic, tangentMode=RCTM_User, time=1, value=0}
        ]}}})
    Scale RGB=&Multiply_Vector(
     A=&MakeVectorFromLinearColorRGB(Color=&ColorFromCurve(ColorCurve=DataInterface() {propertyValues {
         AlphaCurve {keys [<at 0 1 RCIM_Cubic>]}
         BlueCurve {keys [<at 0 20 RCIM_Cubic>,<at 0.6348 8>]}
         GreenCurve {keys [<at 0 20 RCIM_Cubic>,<at 0.6348 8>]}
         RedCurve {keys [<at 0 20 RCIM_Cubic>,<at 0.6348 8>]}
        }}))
     B=&MakeVectorFromLinearColorRGB(Color=&SelectLinearColorFromArray(
       Array Sampling Mode=::NewEnumerator1
       Direct Array Index=$System.RankIndex
       Color Selection Array=DataInterface() {propertyValues {ColorData [
          <rgba 0.2552 0.0535 0.0023 1>
          <rgba 0.0729 0.0729 0.0729 1>
          <rgba 0.2396 0.0932 0 1>
          <rgba 0 0.2282 0.276 1>
          <rgba 0 0.1126 1 1>
          <rgba 0.1152 0.2292 0.192 1>
          <rgba 0.8229 0.0443 0 1>
          <rgba 0 0.3966 1 1>
         ]}}
       Random Seed=?
      ))
    )
    ScaleA=true
    ScaleRGB=false
   ) {module=ScaleColor001}
   &ScaleSpriteSize(
    Scale Sprite Size Mode=::NewEnumerator0
    Uniform Scale Factor=&FloatFromCurve(FloatCurve=DataInterface() {propertyValues {Curve {keys [
         {arriveTangent=2.278, interpMode=RCIM_Cubic, leaveTangent=2.278, tangentMode=RCTM_User, time=0, value=0}
         {interpMode=RCIM_Cubic, tangentMode=RCTM_User, time=1, value=1}
        ]}}})
   )
   /SetVariables(Particles.GlowSize=&Multiply_Vector2D(A=$Particles.SpriteSize, B=<3 3>)) {isSetParameters=true, module=SetVariables_#10}
   /SetVariables(Particles.GlowColor=&MultiplyLinearColor(Linear Color=$Particles.Color, Scale Factor=<1 1 1 0.08>)) {isSetParameters=true, module=SetVariables_#11}
   /SetVariables(Particles.FlatGlowColor:LinearColor=&MultiplyLinearColor(Linear Color=$Particles.Color, Scale Factor=<1 1 1 0.2>)) {isSetParameters=true}
   /SetVariables(Particles.FlatGlowSize:Vector2f=&Multiply_Vector2D(A=$Particles.SpriteSize, B=<0.4 3.5>)) {isSetParameters=true}
   /SetVariables(Particles.ShockColor:LinearColor=&MultiplyLinearColor(
     Linear Color=&ColorFromCurve(ColorCurve=DataInterface() {propertyValues {
        AlphaCurve {keys [<at 0 1 RCIM_Cubic>]}
        BlueCurve {keys [<at 0 20 RCIM_Cubic>,<at 0.6348 8>]}
        GreenCurve {keys [<at 0 12.843 RCIM_Cubic>,<at 0.6348 0.4929>]}
        RedCurve {keys [<at 0 0 RCIM_Cubic>,<at 0.6348 0>]}
       }})
     Scale Factor=&MakeVector4_FromFloat(Float=&FloatFromCurve(
       FloatCurve=DataInterface() {propertyValues {Curve {keys [
           {arriveTangent=-1.5276, interpMode=RCIM_Cubic, leaveTangent=-1.5276, tangentMode=RCTM_User, time=0, value=1}
           {interpMode=RCIM_Cubic, tangentMode=RCTM_User, time=1, value=0}
          ]}}}
       Scale Curve=0.5
      ))
    )) {isSetParameters=true, module=SetVariables_#12}
   &ScaleMeshSize(Scale Factor=&VectorFromFloat(Value=&FloatFromCurve(FloatCurve=DataInterface() {propertyValues {Curve {keys [
          {arriveTangent=1.8207, interpMode=RCIM_Cubic, leaveTangent=1.8207, tangentMode=RCTM_User, time=0, value=0.2}
          {interpMode=RCIM_Cubic, tangentMode=RCTM_User, time=1, value=1.2}
         ]}}})))
   /SetVariables(Particles.WaveDynMat:Vector4f=&MakeVector4(
     X=&FloatFromCurve(
      FloatCurve=DataInterface() {propertyValues {Curve {keys [{arriveTangent=3.726, interpMode=RCIM_Cubic, leaveTangent=3.726, tangentMode=RCTM_User, time=0, value=-0.3}, <at 1 1>]}}}
      Scale Curve=0.6
     )
     Y=&FloatFromCurve(FloatCurve=DataInterface() {propertyValues {Curve {keys [{arriveTangent=-0.5818, interpMode=RCIM_Cubic, leaveTangent=-0.5818, tangentMode=RCTM_User, time=0, value=1}, <at 1 0.8>]}}})
     Z=&FloatFromCurve(FloatCurve=DataInterface() {propertyValues {Curve {keys [<at 0 0>,<at 0.15 1>,<at 1 1.2>]}}}, Scale Curve=10)
     W=&FloatFromCurve(FloatCurve=DataInterface() {propertyValues {Curve {keys [<at 0 0>,<at 1 1>]}}}, Scale Curve=0)
    )) {isSetParameters=true, module=SetVariables_#9}
   &FN_Light_Attributes(
    Color=<rgba 0 0.5043 1 0.7>
    Create Color Attribute=true
    Create Exponent Attribute=true
    Create Radius Attribute=true
    Exponent=1
    Radius=280
    Scale Color=&FloatFromCurve(FloatCurve=DataInterface() {propertyValues {Curve {keys [<at 0 1>,<at 1 0>]}}}, Scale Curve=20)
   )
   &CameraOffset(Camera Offset Amount=&Lerp_Float(
     A=30
     B=-5
     Alpha=&FloatFromCurve(FloatCurve=DataInterface() {propertyValues {Curve {keys [<at 0 0>,<at 0.25 1>]}}}, CurveIndex=$Engine.System.Age)
    ))
  ]
  rendererClasses [
   &NiagaraSpriteRendererProperties
   &NiagaraLightRendererProperties
   &NiagaraMeshRendererProperties
  ]
  renderers [
   {rendererClass=&NiagaraSpriteRendererProperties}
   {rendererClass=&NiagaraSpriteRendererProperties, rendererIndex=1}
   {rendererClass=&NiagaraSpriteRendererProperties, rendererIndex=2}
   {rendererClass=&NiagaraLightRendererProperties, rendererIndex=3}
   {rendererClass=&NiagaraMeshRendererProperties, rendererIndex=4}
  ]
 }
 Particulate_Burst_with_Smoke {
  simTarget=GPUComputeSim
  emitterUpdateScript [
   &EmitterState(Life Cycle Mode=::NewEnumerator1, Loop Behavior=::NewEnumerator1, Loop Delay=0.03, Loop Duration=1, UseLoopDelay=true)
   &SpawnBurst_Instantaneous(Spawn Count=5, Spawn Group=3, Use Spawn Probability=false) {module=SpawnBurst_Instantaneous003}
   &SpawnBurst_Instantaneous(
    Spawn Count=&RandomRangeInt(Minimum=7, Maximum=8)
    Use Spawn Probability=false
    UseLoopCountLimit=false
   )
   &SpawnBurst_Instantaneous(
    Spawn Count=&RandomRangeInt(Minimum=15, Maximum=25)
    Spawn Group=1
    Use Spawn Probability=false
    UseLoopCountLimit=false
   ) {module=SpawnBurst_Instantaneous001}
   &SpawnBurst_Instantaneous(Spawn Count=10, Spawn Group=2, Use Spawn Probability=false, UseLoopCountLimit=false) {module=SpawnBurst_Instantaneous002}
   /SetVariables() {isSetParameters=true, module=SetVariables_#16}
   /SetVariables(Emitter.DepthFade=100) {isSetParameters=true, module=SetVariables_#17}
  ]
  particleSpawnScript [
   /SetVariables(Particles.VisibilityTag=&SpawnGroup()) {isSetParameters=true, module=SetVariables_#18}
   &InitializeParticle(
    Color=&MaskLinearColorBySpawnGroup(
     Masked Linear Color=<rgba 1 1 1 1>
     Passthrough Linear Color=&MaskLinearColorBySpawnGroup(
      Masked Linear Color=$System.ElimColor
      Passthrough Linear Color=&MaskLinearColorBySpawnGroup(
       Masked Linear Color=&RandomRangeLinearColor(Minimum=<rgba 0.0729 0.0681 0.063 0.8>, Maximum=<rgba 0.2708 0.2506 0.2327 1>)
       Passthrough Linear Color=<rgba 0.224 0.224 0.224 0.8>
       Spawn Group Mask=2
      )
      Spawn Group Mask=1
     )
    )
    Lifetime=&MaskFloatBySpawnGroup(
     Masked Float=&RandomRangeFloat(Minimum=0.6, Maximum=0.8)
     Passthrough Float=&MaskFloatBySpawnGroup(
      Masked Float=&RandomRangeFloat(Minimum=0.1, Maximum=1)
      Passthrough Float=&MaskFloatBySpawnGroup(
       Masked Float=&RandomRangeFloat(Minimum=0.2, Maximum=0.25)
       Passthrough Float=&RandomRangeFloat(Minimum=0.5, Maximum=1.5)
       Spawn Group Mask=2
      )
      Spawn Group Mask=1
     )
    )
    Mesh Scale Mode=::NewEnumerator3
    Mesh Uniform Scale=&Multiply_Float(
     A=&Lerp_Float(
      A=0.0285
      B=0.06
      Alpha=&FloatFromCurve(
       FloatCurve=DataInterface() {propertyValues {Curve {keys [<at 0 0 RCIM_Constant>, {arriveTangent=7.6187, interpMode=RCIM_Constant, leaveTangent=7.6187, time=0.65, value=1}]}}}
       CurveIndex=&RandomRangeFloat(Minimum=0, Maximum=1)
      )
     )
     B=1.4
    )
    Position Offset=&MaskVectorBySpawnGroup(
     Masked Vector=<0 0 0>
     Passthrough Vector=&MaskVectorBySpawnGroup(
      Masked Vector=<0 0 0>
      Passthrough Vector=&MaskVectorBySpawnGroup(Masked Vector=<0 0 -10>, Passthrough Vector=<0 0 0>, Spawn Group Mask=2)
      Spawn Group Mask=1
     )
    )
    Sprite Rotation Mode=::NewEnumerator2
    Sprite Size Max=&MaskVector2dBySpawnGroup(
     Masked Vector2D=<0 0>
     Passthrough Vector2D=&MaskVector2dBySpawnGroup(
      Masked Vector2D=<5 5>
      Passthrough Vector2D=&MaskVector2dBySpawnGroup(Masked Vector2D=<75 88>, Passthrough Vector2D=<125 125>, Spawn Group Mask=2)
      Spawn Group Mask=1
     )
     Spawn Group Mask=0
    )
    Sprite Size Min=&MaskVector2dBySpawnGroup(
     Masked Vector2D=<0 0>
     Passthrough Vector2D=&MaskVector2dBySpawnGroup(
      Masked Vector2D=<1 1>
      Passthrough Vector2D=&MaskVector2dBySpawnGroup(Masked Vector2D=<65 75>, Passthrough Vector2D=<75 75>, Spawn Group Mask=2)
      Spawn Group Mask=1
     )
     Spawn Group Mask=0
    )
    Sprite Size Mode=::NewEnumerator2
    UsePositionOffset=true
    Write Material Random=true
   )
   &ShapeLocation(
    Handle Radius:NiagaraFloat=&MaskFloatBySpawnGroup(
     Masked Float=0
     Passthrough Float=&MaskFloatBySpawnGroup(
      Masked Float=0
      Passthrough Float=&MaskFloatBySpawnGroup(Masked Float=3, Passthrough Float=3, Spawn Group Mask=2)
      Spawn Group Mask=1
     )
    )
    Large Radius=&MaskFloatBySpawnGroup(
     Masked Float=2
     Passthrough Float=&MaskFloatBySpawnGroup(
      Masked Float=2
      Passthrough Float=&MaskFloatBySpawnGroup(Masked Float=10, Passthrough Float=5, Spawn Group Mask=2)
      Spawn Group Mask=1
     )
    )
    MaskSpawnGroup=false
    Non Uniform Scale=&MaskVectorBySpawnGroup(
     Masked Vector=<1 1 1>
     Passthrough Vector=&MaskVectorBySpawnGroup(Masked Vector=<1 1 1>, Passthrough Vector=<0.3 0.3 0.3>, Spawn Group Mask=1)
    )
    Offset=&MaskVectorBySpawnGroup(
     Masked Vector=<0 0 2>
     Passthrough Vector=&MaskVectorBySpawnGroup(Masked Vector=<0 0 2>, Passthrough Vector=<0 0 0>, Spawn Group Mask=1)
    )
    Offset Mode=::NewEnumerator0
    Override Seed=false
    Shape Primitive=::NewEnumerator3
    UseOwnerScale=true
   )
   &SubUVAnimation(End Frame=3, SubUV Animation Mode:ENiagara_SubUVLookupMode=::NewEnumerator2)
   &InitialMeshOrientation(Use Rotation Vector=false)
   &AddRotationalVelocity(Rotation Rate=&MaskVectorBySpawnGroup(
     Masked Vector=&Multiply_VectorByFloat(
      Vector=&RandomVector(Vector Scale=2)
      Float=&RandomRangeFloat(Minimum=0.5, Maximum=1)
     )
     Passthrough Vector=&MaskVectorBySpawnGroup(
      Masked Vector=&Multiply_VectorByFloat(
       Vector=&RandomVector(Vector Scale=2)
       Float=&RandomRangeFloat(Minimum=0.5, Maximum=1)
      )
      Passthrough Vector=<0 0 0>
      Spawn Group Mask=1
     )
    ))
   &AddVelocity(
    Module.Use Falloff Distance=false
    Origin Offset=&MaskVectorBySpawnGroup(
     Masked Vector=<0 0 0>
     Passthrough Vector=&MaskVectorBySpawnGroup(
      Masked Vector=<0 0 0>
      Passthrough Vector=&MaskVectorBySpawnGroup(Masked Vector=<0 0 -15>, Passthrough Vector=<0 0 -1>, Spawn Group Mask=2)
      Spawn Group Mask=1
     )
    )
    Override Seed=false
    Velocity Mode=::NewEnumerator1
    Velocity Speed=&MaskFloatBySpawnGroup(
     Masked Float=&RandomRangeFloat(Minimum=200, Maximum=550)
     Passthrough Float=&MaskFloatBySpawnGroup(
      Masked Float=1
      Passthrough Float=&MaskFloatBySpawnGroup(Masked Float=150, Passthrough Float=&RandomRangeFloat(Minimum=700, Maximum=1000), Spawn Group Mask=2)
      Spawn Group Mask=1
     )
    )
   )
   &AddVelocity(
    Cone Angle=&MaskFloatBySpawnGroup(
     Masked Float=35
     Passthrough Float=&MaskFloatBySpawnGroup(Masked Float=95, Passthrough Float=0, Spawn Group Mask=1)
    )
    Cone Axis=&MaskVectorBySpawnGroup(
     Masked Vector=&RotateVectorByQuaternion(
      Vector To Rotate=<1 0 0>
      Quaternion=&MakeQuatFromAxisAngle(Axis=<0 0 1>, AngleInDegrees=-145, NormalizeAxis=true)
     )
     Passthrough Vector=&MaskVectorBySpawnGroup(Masked Vector=<0 0 1>, Passthrough Vector=<0 0 1>, Spawn Group Mask=1)
    )
    Distribution Along Cone Axis:NiagaraFloat=&MaskFloatBySpawnGroup(
     Masked Float=0
     Passthrough Float=&MaskFloatBySpawnGroup(Masked Float=0.2, Passthrough Float=0, Spawn Group Mask=1)
    )
    Module.UseConeFalloff=false
    Module.UseVelDistribution=true
    Override Seed=false
    Velocity Mode=::NewEnumerator2
    Velocity Speed=&MaskFloatBySpawnGroup(
     Masked Float=&RandomRangeFloat(Minimum=70, Maximum=350)
     Passthrough Float=&MaskFloatBySpawnGroup(
      Masked Float=&RandomRangeFloat(Minimum=50, Maximum=650)
      Passthrough Float=&MaskFloatBySpawnGroup(Masked Float=0, Passthrough Float=0, Spawn Group Mask=2)
      Spawn Group Mask=1
     )
    )
   ) {module=AddVelocity001}
   &ApplyInitialForces()
   /SetVariables(Particles.MeshIndex=&RandomRangeInt(Minimum=0, Maximum=2)) {isSetParameters=true, module=SetVariables_#19}
   /SetVariables(Particles.InitSpriteRotRate=&UniformRangedFloat(Minimum=-3, Maximum=3)) {isSetParameters=true, module=SetVariables_#20}
   /SetVariables(Particles.RockDMP:Vector4f=&MakeVector4(X=0, Y=1, Z=0, W=0)) {isSetParameters=true, module=SetVariables_#21}
  ]
  particleUpdateScript [
   &ParticleState()
   &SpriteRotationRate(Rotation Rate=&MaskFloatBySpawnGroup(
     Masked Float=0
     Passthrough Float=&MaskFloatBySpawnGroup(
      Masked Float=0
      Passthrough Float=&MaskFloatBySpawnGroup(
       Masked Float=0
       Passthrough Float=&Multiply_Float(
        A=$Particles.InitSpriteRotRate
        B=&FloatFromCurve(
         FloatCurve=DataInterface() {propertyValues {Curve {keys [
             {arriveTangent=-4.0107, interpMode=RCIM_Cubic, leaveTangent=-4.0107, tangentMode=RCTM_User, time=-0.0046, value=1.0113}
             {arriveTangent=-0.2496, interpMode=RCIM_Cubic, leaveTangent=-0.2496, tangentMode=RCTM_User, time=0.5, value=0.16}
            ]}}}
         CurveIndex=$Emitter.Age
         Scale Curve=150
        )
       )
       Spawn Group Mask=2
      )
      Spawn Group Mask=1
     )
    ))
   &GravityForce(Gravity=&MaskVectorBySpawnGroup(
     Masked Vector=&RandomRangeVector(Minimum=<0 0 -750>, Maximum=<0 0 -1500>)
     Passthrough Vector=&MaskVectorBySpawnGroup(
      Masked Vector=<0 0 -700>
      Passthrough Vector=&MaskVectorBySpawnGroup(Masked Vector=<0 0 0>, Passthrough Vector=<0 0 0>, Spawn Group Mask=2)
      Spawn Group Mask=1
     )
    ))
   /SetVariables(Particles.Color=&MaskLinearColorBySpawnGroup(
     Masked Linear Color=$Particles.Initial.Color
     Passthrough Linear Color=&MaskLinearColorBySpawnGroup(
      Masked Linear Color=&Multiply_LinearColorByFloat(
       Linear Color=&MultiplyLinearColor(
        Linear Color=&Multiply_LinearColorByFloat(Linear Color=$System.ElimColor, Float=&RandomRangeFloat(Minimum=0.8, Maximum=25))
        Scale Factor=/Niagara/DynamicInputs/Multiply/Multiply_Vector4_ByFloat(
         Vector4:Vector4f=<1 1 1 1>
         Float:NiagaraFloat=&FloatFromCurve(FloatCurve=DataInterface() {propertyValues {Curve {keys [<at 0 1>,<at 0.2 1>,<at 0.3 5>,<at 0.4 1>,<at 0.7 1>,<at 0.8 5>,<at 0.9 1>,<at 1 1>]}}})
        )
       )
       Float=&FloatFromCurve(FloatCurve=DataInterface() {propertyValues {Curve {keys [
            {arriveTangent=9.0725, interpMode=RCIM_Cubic, leaveTangent=9.0725, tangentMode=RCTM_User, time=0, value=0}
            {arriveTangent=-2.5249, interpMode=RCIM_Cubic, leaveTangent=-2.5249, tangentMode=RCTM_User, time=0.1408, value=0.9847}
            {arriveTangent=-0.5654, interpMode=RCIM_Cubic, leaveTangent=-0.5654, time=0.5, value=0.35}
            {arriveTangent=-1.25, interpMode=RCIM_Cubic, leaveTangent=-1.25, time=0.6, value=0.725}
            {arriveTangent=-3.375, interpMode=RCIM_Cubic, leaveTangent=-0.1412, tangentMode=RCTM_Break, time=0.7, value=0.1}
            {arriveTangent=-0.7119, interpMode=RCIM_Cubic, leaveTangent=1.625, tangentMode=RCTM_Break, time=0.8, value=0.05}
            {arriveTangent=-0.25, interpMode=RCIM_Cubic, leaveTangent=-0.25, time=0.9, value=0.425}
            {arriveTangent=-0.0405, interpMode=RCIM_Cubic, leaveTangent=-0.0405, tangentMode=RCTM_User, time=1, value=0}
           ]}}})
      )
      Passthrough Linear Color=&MaskLinearColorBySpawnGroup(
       Masked Linear Color=&Multiply_LinearColorByFloat(
        Linear Color=&ColorFromCurve(ColorCurve=DataInterface() {propertyValues {
           AlphaCurve {keys [<at 0 1 RCIM_Cubic>,<at 1 1 RCIM_Cubic>]}
           BlueCurve {keys [<at 0.2382 0>,<at 0.383 0.004 RCIM_Cubic>,<at 1 0 RCIM_Cubic>]}
           GreenCurve {keys [<at 0.2382 0.0124>, {arriveTangent=-0.0163, interpMode=RCIM_Cubic, leaveTangent=-0.0163, time=0.383, value=0.0174}, <at 1 0 RCIM_Cubic>]}
           RedCurve {keys [<at 0.2382 0.4167>, {arriveTangent=-0.547, interpMode=RCIM_Cubic, leaveTangent=-0.547, time=0.383, value=0.1667}, <at 1 0 RCIM_Cubic>]}
          }})
        Float=&Multiply_Float(
         A=&FloatFromCurve(
          FloatCurve=DataInterface() {propertyValues {Curve {keys [
              {interpMode=RCIM_Cubic, leaveTangentWeight=0.0962, tangentMode=RCTM_Break, time=0, value=0}
              {interpMode=RCIM_Cubic, tangentMode=RCTM_Break, time=0.2, value=0.7}
              {interpMode=RCIM_Cubic, tangentMode=RCTM_Break, time=1, value=0}
             ]}}}
          Scale Curve=&RandomRangeFloat(Minimum=0.8, Maximum=1)
         )
         B=&Multiply_Float(A=&FNTimeOfDayScalar(Night Scale=1, Morning Scale=0.8, Evening Scale=0.8), B=1)
        )
       )
       Passthrough Linear Color=$Particles.Initial.Color
       Spawn Group Mask=2
      )
      Spawn Group Mask=1
     )
    )) {isSetParameters=true, module=SetVariables_#22}
   /Niagara/Modules/Update/Forces/CurlNoiseForce(
    Noise Frequency=0.0012
    Noise Strength=&MaskFloatBySpawnGroup(
     Masked Float=0
     Passthrough Float=&MaskFloatBySpawnGroup(
      Masked Float=0
      Passthrough Float=&MaskFloatBySpawnGroup(Masked Float=0, Passthrough Float=800, Spawn Group Mask=2)
      Spawn Group Mask=1
     )
    )
    Pan Noise=true
    Pan Noise Field=<0 0 0.3>
   )
   &Drag(
    Drag=&MaskFloatBySpawnGroup(
     Masked Float=1.5
     Passthrough Float=&MaskFloatBySpawnGroup(
      Masked Float=&RandomRangeFloat(Minimum=0.8, Maximum=1.2)
      Passthrough Float=&MaskFloatBySpawnGroup(Masked Float=0, Passthrough Float=20, Spawn Group Mask=2)
      Spawn Group Mask=1
     )
    )
    Rotational Drag:NiagaraFloat=&MaskFloatBySpawnGroup(
     Masked Float=&FloatFromCurve(FloatCurve=DataInterface() {propertyValues {Curve {keys [<at 0 0 RCIM_Cubic>, {arriveTangent=0.1846, interpMode=RCIM_Cubic, leaveTangent=0.1846, tangentMode=RCTM_User, time=1, value=0.1}]}}})
     Passthrough Float=&MaskFloatBySpawnGroup(Masked Float=0, Passthrough Float=0, Spawn Group Mask=1)
    )
    Use Linear Drag=true
    Use Rotational Drag=true
   )
   &SolveForcesAndVelocity(Clamp Velocity=false, Limit Acceleration=false)
   &ScaleSpriteSize(
    Scale Factor=&MaskVector2dBySpawnGroup(
     Masked Vector2D=<0 0>
     Passthrough Vector2D=&MaskVector2dBySpawnGroup(
      Masked Vector2D=<1 1>
      Passthrough Vector2D=&MaskVector2dBySpawnGroup(
       Masked Vector2D=&Vector2DFromCurve(
        Vector2Curve=DataInterface() {propertyValues {XCurve {keys [<at 0.0036 1.6897>,<at 0.9967 0.0582>]}, YCurve {keys [<at 0 1>,<at 1.002 3.7227>]}}}
        ScaleCurve=<0.4 0.4>
       )
       Passthrough Vector2D=&Vector2DFromFloat(Value=&FloatFromCurve(FloatCurve=DataInterface() {propertyValues {Curve {keys [
             {arriveTangent=6.2202, interpMode=RCIM_Cubic, leaveTangent=6.2202, tangentMode=RCTM_User, time=0, value=0.15}
             {arriveTangent=1.0025, interpMode=RCIM_Cubic, leaveTangent=1.0025, tangentMode=RCTM_User, time=0.2, value=0.8}
             {arriveTangent=0.0143, interpMode=RCIM_Cubic, leaveTangent=0.0143, tangentMode=RCTM_User, time=1.0463, value=1}
            ]}}}))
       Spawn Group Mask=2
      )
      Spawn Group Mask=1
     )
     Spawn Group Mask=0
    )
    Scale Sprite Size Mode=::NewEnumerator2
   )
   &ScaleSpriteSizeBySpeed(
    Max Scale Factor=&MaskVector2dBySpawnGroup(
     Masked Vector2D=<1 1>
     Passthrough Vector2D=&MaskVector2dBySpawnGroup(
      Masked Vector2D=<0.5 7>
      Passthrough Vector2D=&MaskVector2dBySpawnGroup(Masked Vector2D=<1 1>, Passthrough Vector2D=<1 1>, Spawn Group Mask=2)
      Spawn Group Mask=1
     )
     Spawn Group Mask=0
    )
    Sample Scale Factor By Curve=true
    Velocity Threshold=500
   )
   &ScaleMeshSize(Scale Factor=&VectorFromFloat(Value=&FloatFromCurve(FloatCurve=DataInterface() {propertyValues {Curve {keys [{arriveTangent=-0.0035, interpMode=RCIM_Cubic, leaveTangent=-0.0035, tangentMode=RCTM_User, time=0.6, value=1}, <at 1 0>]}}})))
   /SetVariables(Particles.SandDMP:Vector4f=&MakeVector4(
     X=0
     Y=1
     Z=1
     W=&FloatFromCurve(FloatCurve=DataInterface() {propertyValues {Curve {keys [{interpMode=RCIM_Cubic, tangentMode=RCTM_Break, time=0.3, value=1}, {interpMode=RCIM_Cubic, tangentMode=RCTM_Break, time=1, value=0.25}]}}})
    )) {isSetParameters=true, module=SetVariables_#23}
   &DynamicMaterialParameters(
    Index 0 Param 1=&FloatFromCurve(
     FloatCurve=DataInterface() {propertyValues {Curve {keys [
         {arriveTangent=2.7533, interpMode=RCIM_Cubic, leaveTangent=1.8958, tangentMode=RCTM_User, time=0, value=1}
         {arriveTangent=0.8292, interpMode=RCIM_Cubic, leaveTangent=0.8292, tangentMode=RCTM_User, time=0.321, value=1.3453}
         {arriveTangent=0.0529, interpMode=RCIM_Cubic, leaveTangent=0.2021, tangentMode=RCTM_User, time=1, value=1.5}
        ]}}}
     CurveIndex=&Add_Float(A=&Multiply_Float(A=$Particles.Initial.MaterialRandom, B=0.5), B=$Particles.NormalizedAge)
    )
    Index 0 Param 2=&FloatFromCurve(FloatCurve=DataInterface() {propertyValues {Curve {keys [
         {arriveTangent=0.0453, interpMode=RCIM_Cubic, tangentMode=RCTM_Break, time=0, value=-1}
         {arriveTangent=0.0604, interpMode=RCIM_Cubic, leaveTangent=0.0604, tangentMode=RCTM_User, time=1, value=-1}
        ]}}})
    Index 0 Param 3=&FloatFromCurve(
     FloatCurve=DataInterface() {propertyValues {Curve {keys [
         {arriveTangent=-53.9199, interpMode=RCIM_Cubic, leaveTangent=-8.9727, tangentMode=RCTM_User, time=0, value=1}
         {arriveTangent=0.2347, interpMode=RCIM_Cubic, leaveTangent=0.2723, tangentMode=RCTM_Break, time=0.25, value=0}
        ]}}}
     Scale Curve=0.01
    )
    Index 0 Param 4=&FloatFromCurve(FloatCurve=DataInterface() {propertyValues {Curve {keys [<at 0.5 1>,<at 1 0>]}}})
    Index 1 Param 1=&FloatFromCurve(
     FloatCurve=DataInterface() {propertyValues {Curve {keys [
         {interpMode=RCIM_Cubic, leaveTangent=4.7088, tangentMode=RCTM_Break, time=0, value=0}
         {arriveTangent=4.2257, interpMode=RCIM_Cubic, leaveTangent=4.2257, tangentMode=RCTM_User, time=0.25, value=1}
        ]}}}
     Scale Curve=&RandomRangeFloat(Minimum=1.5, Maximum=1.5)
    )
    Index 1 Param 2=&FloatFromCurve(
     FloatCurve=DataInterface() {propertyValues {Curve {keys [<at 0 0.2>, {leaveTangent=0.9066, time=1, value=2}]}}}
     Scale Curve=&RandomRangeFloat(Minimum=0.5, Maximum=0.6)
    )
    Index 1 Param 3=&FloatFromCurve(FloatCurve=DataInterface() {propertyValues {Curve {keys [
         {interpMode=RCIM_Cubic, leaveTangent=6.3917, tangentMode=RCTM_Break, time=0, value=0.2}
         {arriveTangent=1.7549, interpMode=RCIM_Cubic, leaveTangent=10.6767, tangentMode=RCTM_User, time=0.25, value=1}
        ]}}})
    Index 1 Param 4=&FloatFromCurve(
     FloatCurve=DataInterface() {propertyValues {Curve {keys [{arriveTangent=-0.136, interpMode=RCIM_Cubic, leaveTangent=-0.136, tangentMode=RCTM_User, time=0, value=1}, <at 1 0.95>]}}}
     Scale Curve=&RandomRangeFloat(Minimum=1.1, Maximum=1.4)
    )
    Index 2 Param 1=&FloatFromCurve(FloatCurve=DataInterface() {propertyValues {Curve {keys [
         {interpMode=RCIM_Cubic, leaveTangent=4.2485, tangentMode=RCTM_Break, time=0, value=0}
         {arriveTangent=1.0548, interpMode=RCIM_Cubic, leaveTangent=1.0548, tangentMode=RCTM_User, time=0.17, value=0.5715}
         <at 1 1>
        ]}}})
    Index 2 Param 2=&FloatFromCurve(
     FloatCurve=DataInterface() {propertyValues {Curve {keys [
         {arriveTangent=1.9056, interpMode=RCIM_Cubic, leaveTangent=1.9056, tangentMode=RCTM_User, time=0, value=0.2}
         {arriveTangent=0.0259, interpMode=RCIM_Cubic, leaveTangent=0.0259, tangentMode=RCTM_User, time=1, value=1.4}
        ]}}}
     Scale Curve=&RandomRangeFloat(Minimum=0.2, Maximum=0.7)
    )
    Index 2 Param 3=&FloatFromCurve(FloatCurve=DataInterface() {propertyValues {Curve {keys [
         {interpMode=RCIM_Cubic, leaveTangent=2.3064, tangentMode=RCTM_Break, time=0, value=1}
         {interpMode=RCIM_Cubic, leaveTangent=11.1233, tangentMode=RCTM_Break, time=1, value=2}
        ]}}})
    Index 2 Param 4=&FloatFromCurve(
     FloatCurve=DataInterface() {propertyValues {Curve {keys [
         {arriveTangent=-0.6867, interpMode=RCIM_Cubic, leaveTangent=-0.6867, tangentMode=RCTM_User, time=0, value=1}
         {interpMode=RCIM_Cubic, leaveTangent=-1.2956, tangentMode=RCTM_Break, time=1, value=0.75}
        ]}}}
     Scale Curve=1.6
    )
    Module.Param0WriteEnabled=true
    Module.Param0WriteEnabled1=true
    Module.Param0WriteEnabled2=true
    Module.Param1WriteEnabled=false
    Module.Param1WriteEnabled1=true
    Module.Param1WriteEnabled2=true
    Module.Param2WriteEnabled=true
    Module.Param2WriteEnabled1=true
    Module.Param2WriteEnabled2=true
    Module.Param3WriteEnabled=true
    Module.Param3WriteEnabled1=true
    Module.Param3WriteEnabled2=true
    Write Parameter Index 1=true
    Write Parameter Index 2=true
   ) {module=DynamicMaterialParameters001}
   &ScaleColor(
    Scale Alpha=&MaskFloatBySpawnGroup(
     Masked Float=1
     Passthrough Float=&MaskFloatBySpawnGroup(
      Masked Float=1
      Passthrough Float=&MaskFloatBySpawnGroup(
       Masked Float=1
       Passthrough Float=&Multiply_Float(
        A=&FloatFromCurve(
         FloatCurve=DataInterface() {propertyValues {Curve {keys [
             <at 0.1 0.75>
             {interpMode=RCIM_Cubic, tangentMode=RCTM_User, time=0.2, value=2.25}
             {interpMode=RCIM_Cubic, tangentMode=RCTM_User, time=0.7612, value=1.5}
             <at 1 0>
            ]}}}
         Scale Curve=&RandomRangeFloat(Minimum=1, Maximum=1.5)
        )
        B=&Multiply_Float(A=&FNTimeOfDayScalar(Night Scale=1, Morning Scale=0.7, Evening Scale=0.7), B=1.5)
       )
       Spawn Group Mask=2
      )
      Spawn Group Mask=1
     )
    )
    Scale RGB=&MaskVectorBySpawnGroup(
     Masked Vector=<1 1 1>
     Passthrough Vector=&MaskVectorBySpawnGroup(
      Masked Vector=<1 1 1>
      Passthrough Vector=&MaskVectorBySpawnGroup(
       Masked Vector=<1 1 1>
       Passthrough Vector=&VectorFromFloat(Value=&FloatFromCurve(FloatCurve=DataInterface() {propertyValues {Curve {keys [{arriveTangent=1.8145, interpMode=RCIM_Cubic, leaveTangent=0.4403, tangentMode=RCTM_Break, time=0, value=0.4}, <at 0.4 0.625>, <at 1 1>]}}}))
       Spawn Group Mask=2
      )
      Spawn Group Mask=1
     )
    )
    ScaleA=true
    ScaleRGB=true
   )
   &CameraOffset(Camera Offset Amount=&MaskFloatBySpawnGroup(
     Masked Float=0
     Passthrough Float=&MaskFloatBySpawnGroup(
      Masked Float=0
      Passthrough Float=&MaskFloatBySpawnGroup(Masked Float=0, Passthrough Float=10, Spawn Group Mask=2)
      Spawn Group Mask=1
     )
    ))
  ]
  rendererClasses [&NiagaraMeshRendererProperties, &NiagaraSpriteRendererProperties]
  renderers [
   {rendererClass=&NiagaraMeshRendererProperties, rendererIndex=0}
   {rendererClass=&NiagaraSpriteRendererProperties, rendererIndex=1}
   {rendererClass=&NiagaraSpriteRendererProperties, rendererIndex=2}
   {rendererClass=&NiagaraSpriteRendererProperties, rendererIndex=3}
   {rendererClass=&NiagaraSpriteRendererProperties, rendererIndex=4}
  ]
 }
 Particulate_Burst_with_Smoke001 {
  bEnabled=false
  simTarget=GPUComputeSim
  emitterUpdateScript [
   &EmitterState(Life Cycle Mode=::NewEnumerator1, Loop Behavior=::NewEnumerator1, Loop Delay=0.03, Loop Duration=1, UseLoopDelay=true)
   &SpawnBurst_Instantaneous(Spawn Count=5, Spawn Group=3, Use Spawn Probability=false) {module=SpawnBurst_Instantaneous003}
   &SpawnBurst_Instantaneous(
    Spawn Count=&RandomRangeInt(Minimum=7, Maximum=8)
    Use Spawn Probability=false
    UseLoopCountLimit=false
   )
   &SpawnBurst_Instantaneous(
    Spawn Count=&RandomRangeInt(Minimum=15, Maximum=25)
    Spawn Group=1
    Use Spawn Probability=false
    UseLoopCountLimit=false
   ) {module=SpawnBurst_Instantaneous001}
   &SpawnBurst_Instantaneous(Spawn Count=10, Spawn Group=2, Use Spawn Probability=false, UseLoopCountLimit=false) {module=SpawnBurst_Instantaneous002}
   /SetVariables() {isSetParameters=true, module=SetVariables_#16}
   /SetVariables(Emitter.DepthFade=100) {isSetParameters=true, module=SetVariables_#17}
  ]
  particleSpawnScript [
   /SetVariables(Particles.VisibilityTag=&SpawnGroup()) {isSetParameters=true, module=SetVariables_#18}
   &InitializeParticle(
    Color=&MaskLinearColorBySpawnGroup(
     Masked Linear Color=<rgba 1 1 1 1>
     Passthrough Linear Color=&MaskLinearColorBySpawnGroup(
      Masked Linear Color=$System.ElimColor
      Passthrough Linear Color=&MaskLinearColorBySpawnGroup(
       Masked Linear Color=&RandomRangeLinearColor(Minimum=<rgba 0.0729 0.0681 0.063 0.8>, Maximum=<rgba 0.2708 0.2506 0.2327 1>)
       Passthrough Linear Color=<rgba 0.224 0.224 0.224 0.8>
       Spawn Group Mask=2
      )
      Spawn Group Mask=1
     )
    )
    Lifetime=&MaskFloatBySpawnGroup(
     Masked Float=&RandomRangeFloat(Minimum=0.6, Maximum=0.8)
     Passthrough Float=&MaskFloatBySpawnGroup(
      Masked Float=&RandomRangeFloat(Minimum=0.1, Maximum=1)
      Passthrough Float=&MaskFloatBySpawnGroup(
       Masked Float=&RandomRangeFloat(Minimum=0.2, Maximum=0.25)
       Passthrough Float=&RandomRangeFloat(Minimum=0.5, Maximum=1.5)
       Spawn Group Mask=2
      )
      Spawn Group Mask=1
     )
    )
    Mesh Scale Mode=::NewEnumerator3
    Mesh Uniform Scale=&Multiply_Float(
     A=&Lerp_Float(
      A=0.0285
      B=0.06
      Alpha=&FloatFromCurve(
       FloatCurve=DataInterface() {propertyValues {Curve {keys [<at 0 0 RCIM_Constant>, {arriveTangent=7.6187, interpMode=RCIM_Constant, leaveTangent=7.6187, time=0.65, value=1}]}}}
       CurveIndex=&RandomRangeFloat(Minimum=0, Maximum=1)
      )
     )
     B=1.4
    )
    Position Offset=&MaskVectorBySpawnGroup(
     Masked Vector=<0 0 0>
     Passthrough Vector=&MaskVectorBySpawnGroup(
      Masked Vector=<0 0 0>
      Passthrough Vector=&MaskVectorBySpawnGroup(Masked Vector=<0 0 -10>, Passthrough Vector=<0 0 0>, Spawn Group Mask=2)
      Spawn Group Mask=1
     )
    )
    Sprite Rotation Mode=::NewEnumerator2
    Sprite Size Max=&MaskVector2dBySpawnGroup(
     Masked Vector2D=<0 0>
     Passthrough Vector2D=&MaskVector2dBySpawnGroup(
      Masked Vector2D=<5 5>
      Passthrough Vector2D=&MaskVector2dBySpawnGroup(Masked Vector2D=<75 88>, Passthrough Vector2D=<125 125>, Spawn Group Mask=2)
      Spawn Group Mask=1
     )
     Spawn Group Mask=0
    )
    Sprite Size Min=&MaskVector2dBySpawnGroup(
     Masked Vector2D=<0 0>
     Passthrough Vector2D=&MaskVector2dBySpawnGroup(
      Masked Vector2D=<1 1>
      Passthrough Vector2D=&MaskVector2dBySpawnGroup(Masked Vector2D=<65 75>, Passthrough Vector2D=<75 75>, Spawn Group Mask=2)
      Spawn Group Mask=1
     )
     Spawn Group Mask=0
    )
    Sprite Size Mode=::NewEnumerator2
    UsePositionOffset=true
    Write Material Random=true
   )
   &ShapeLocation(
    Handle Radius:NiagaraFloat=&MaskFloatBySpawnGroup(
     Masked Float=0
     Passthrough Float=&MaskFloatBySpawnGroup(
      Masked Float=0
      Passthrough Float=&MaskFloatBySpawnGroup(Masked Float=3, Passthrough Float=3, Spawn Group Mask=2)
      Spawn Group Mask=1
     )
    )
    Large Radius=&MaskFloatBySpawnGroup(
     Masked Float=2
     Passthrough Float=&MaskFloatBySpawnGroup(
      Masked Float=2
      Passthrough Float=&MaskFloatBySpawnGroup(Masked Float=10, Passthrough Float=5, Spawn Group Mask=2)
      Spawn Group Mask=1
     )
    )
    MaskSpawnGroup=false
    Non Uniform Scale=&MaskVectorBySpawnGroup(
     Masked Vector=<1 1 1>
     Passthrough Vector=&MaskVectorBySpawnGroup(Masked Vector=<1 1 1>, Passthrough Vector=<0.3 0.3 0.3>, Spawn Group Mask=1)
    )
    Offset=&MaskVectorBySpawnGroup(
     Masked Vector=<0 0 2>
     Passthrough Vector=&MaskVectorBySpawnGroup(Masked Vector=<0 0 2>, Passthrough Vector=<0 0 0>, Spawn Group Mask=1)
    )
    Offset Mode=::NewEnumerator0
    Override Seed=false
    Shape Primitive=::NewEnumerator3
    UseOwnerScale=true
   )
   &SubUVAnimation(End Frame=3, SubUV Animation Mode:ENiagara_SubUVLookupMode=::NewEnumerator2)
   &InitialMeshOrientation(Use Rotation Vector=false)
   &AddRotationalVelocity(Rotation Rate=&MaskVectorBySpawnGroup(
     Masked Vector=&Multiply_VectorByFloat(
      Vector=&RandomVector(Vector Scale=2)
      Float=&RandomRangeFloat(Minimum=0.5, Maximum=1)
     )
     Passthrough Vector=&MaskVectorBySpawnGroup(
      Masked Vector=&Multiply_VectorByFloat(
       Vector=&RandomVector(Vector Scale=2)
       Float=&RandomRangeFloat(Minimum=0.5, Maximum=1)
      )
      Passthrough Vector=<0 0 0>
      Spawn Group Mask=1
     )
    ))
   &AddVelocity(
    Module.Use Falloff Distance=false
    Origin Offset=&MaskVectorBySpawnGroup(
     Masked Vector=<0 0 0>
     Passthrough Vector=&MaskVectorBySpawnGroup(
      Masked Vector=<0 0 0>
      Passthrough Vector=&MaskVectorBySpawnGroup(Masked Vector=<0 0 -15>, Passthrough Vector=<0 0 -1>, Spawn Group Mask=2)
      Spawn Group Mask=1
     )
    )
    Override Seed=false
    Velocity Mode=::NewEnumerator1
    Velocity Speed=&MaskFloatBySpawnGroup(
     Masked Float=&RandomRangeFloat(Minimum=200, Maximum=550)
     Passthrough Float=&MaskFloatBySpawnGroup(
      Masked Float=1
      Passthrough Float=&MaskFloatBySpawnGroup(Masked Float=150, Passthrough Float=&RandomRangeFloat(Minimum=700, Maximum=1000), Spawn Group Mask=2)
      Spawn Group Mask=1
     )
    )
   )
   &AddVelocity(
    Cone Angle=&MaskFloatBySpawnGroup(
     Masked Float=35
     Passthrough Float=&MaskFloatBySpawnGroup(Masked Float=95, Passthrough Float=0, Spawn Group Mask=1)
    )
    Cone Axis=&MaskVectorBySpawnGroup(
     Masked Vector=&RotateVectorByQuaternion(
      Vector To Rotate=<1 0 0>
      Quaternion=&MakeQuatFromAxisAngle(Axis=<0 0 1>, AngleInDegrees=-145, NormalizeAxis=true)
     )
     Passthrough Vector=&MaskVectorBySpawnGroup(Masked Vector=<0 0 1>, Passthrough Vector=<0 0 1>, Spawn Group Mask=1)
    )
    Distribution Along Cone Axis:NiagaraFloat=&MaskFloatBySpawnGroup(
     Masked Float=0
     Passthrough Float=&MaskFloatBySpawnGroup(Masked Float=0.2, Passthrough Float=0, Spawn Group Mask=1)
    )
    Module.UseConeFalloff=false
    Module.UseVelDistribution=true
    Override Seed=false
    Velocity Mode=::NewEnumerator2
    Velocity Speed=&MaskFloatBySpawnGroup(
     Masked Float=&RandomRangeFloat(Minimum=70, Maximum=350)
     Passthrough Float=&MaskFloatBySpawnGroup(
      Masked Float=&RandomRangeFloat(Minimum=50, Maximum=650)
      Passthrough Float=&MaskFloatBySpawnGroup(Masked Float=0, Passthrough Float=0, Spawn Group Mask=2)
      Spawn Group Mask=1
     )
    )
   ) {module=AddVelocity001}
   &ApplyInitialForces()
   /SetVariables(Particles.MeshIndex=&RandomRangeInt(Minimum=0, Maximum=2)) {isSetParameters=true, module=SetVariables_#19}
   /SetVariables(Particles.InitSpriteRotRate=&UniformRangedFloat(Minimum=-3, Maximum=3)) {isSetParameters=true, module=SetVariables_#20}
   /SetVariables(Particles.RockDMP:Vector4f=&MakeVector4(X=0, Y=1, Z=0, W=0)) {isSetParameters=true, module=SetVariables_#21}
  ]
  particleUpdateScript [
   &ParticleState()
   &SpriteRotationRate(Rotation Rate=&MaskFloatBySpawnGroup(
     Masked Float=0
     Passthrough Float=&MaskFloatBySpawnGroup(
      Masked Float=0
      Passthrough Float=&MaskFloatBySpawnGroup(
       Masked Float=0
       Passthrough Float=&Multiply_Float(
        A=$Particles.InitSpriteRotRate
        B=&FloatFromCurve(
         FloatCurve=DataInterface() {propertyValues {Curve {keys [
             {arriveTangent=-4.0107, interpMode=RCIM_Cubic, leaveTangent=-4.0107, tangentMode=RCTM_User, time=-0.0046, value=1.0113}
             {arriveTangent=-0.2496, interpMode=RCIM_Cubic, leaveTangent=-0.2496, tangentMode=RCTM_User, time=0.5, value=0.16}
            ]}}}
         CurveIndex=$Emitter.Age
         Scale Curve=150
        )
       )
       Spawn Group Mask=2
      )
      Spawn Group Mask=1
     )
    ))
   &GravityForce(Gravity=&MaskVectorBySpawnGroup(
     Masked Vector=&RandomRangeVector(Minimum=<0 0 -750>, Maximum=<0 0 -1500>)
     Passthrough Vector=&MaskVectorBySpawnGroup(
      Masked Vector=<0 0 -700>
      Passthrough Vector=&MaskVectorBySpawnGroup(Masked Vector=<0 0 0>, Passthrough Vector=<0 0 0>, Spawn Group Mask=2)
      Spawn Group Mask=1
     )
    ))
   /SetVariables(Particles.Color=&MaskLinearColorBySpawnGroup(
     Masked Linear Color=$Particles.Initial.Color
     Passthrough Linear Color=&MaskLinearColorBySpawnGroup(
      Masked Linear Color=&Multiply_LinearColorByFloat(
       Linear Color=&MultiplyLinearColor(
        Linear Color=&Multiply_LinearColorByFloat(Linear Color=$System.ElimColor, Float=&RandomRangeFloat(Minimum=0.8, Maximum=25))
        Scale Factor=&Multiply_Vector4_ByFloat(
         Vector4:Vector4f=<1 1 1 1>
         Float:NiagaraFloat=&FloatFromCurve(FloatCurve=DataInterface() {propertyValues {Curve {keys [<at 0 1>,<at 0.2 1>,<at 0.3 5>,<at 0.4 1>,<at 0.7 1>,<at 0.8 5>,<at 0.9 1>,<at 1 1>]}}})
        )
       )
       Float=&FloatFromCurve(FloatCurve=DataInterface() {propertyValues {Curve {keys [
            {arriveTangent=9.0725, interpMode=RCIM_Cubic, leaveTangent=9.0725, tangentMode=RCTM_User, time=0, value=0}
            {arriveTangent=-2.5249, interpMode=RCIM_Cubic, leaveTangent=-2.5249, tangentMode=RCTM_User, time=0.1408, value=0.9847}
            {arriveTangent=-0.5654, interpMode=RCIM_Cubic, leaveTangent=-0.5654, time=0.5, value=0.35}
            {arriveTangent=-1.25, interpMode=RCIM_Cubic, leaveTangent=-1.25, time=0.6, value=0.725}
            {arriveTangent=-3.375, interpMode=RCIM_Cubic, leaveTangent=-0.1412, tangentMode=RCTM_Break, time=0.7, value=0.1}
            {arriveTangent=-0.7119, interpMode=RCIM_Cubic, leaveTangent=1.625, tangentMode=RCTM_Break, time=0.8, value=0.05}
            {arriveTangent=-0.25, interpMode=RCIM_Cubic, leaveTangent=-0.25, time=0.9, value=0.425}
            {arriveTangent=-0.0405, interpMode=RCIM_Cubic, leaveTangent=-0.0405, tangentMode=RCTM_User, time=1, value=0}
           ]}}})
      )
      Passthrough Linear Color=&MaskLinearColorBySpawnGroup(
       Masked Linear Color=&Multiply_LinearColorByFloat(
        Linear Color=&ColorFromCurve(ColorCurve=DataInterface() {propertyValues {
           AlphaCurve {keys [<at 0 1 RCIM_Cubic>,<at 1 1 RCIM_Cubic>]}
           BlueCurve {keys [<at 0.2382 0.4167>, {arriveTangent=-0.547, interpMode=RCIM_Cubic, leaveTangent=-0.547, time=0.383, value=0.1667}, <at 1 0 RCIM_Cubic>]}
           GreenCurve {keys [<at 0.2382 0.2077>, {arriveTangent=-0.2726, interpMode=RCIM_Cubic, leaveTangent=-0.2726, time=0.383, value=0.0946}, <at 1 0 RCIM_Cubic>]}
           RedCurve {keys [<at 0.2382 0>,<at 0.383 0.0721 RCIM_Cubic>,<at 1 0 RCIM_Cubic>]}
          }})
        Float=&Multiply_Float(
         A=&FloatFromCurve(
          FloatCurve=DataInterface() {propertyValues {Curve {keys [
              {interpMode=RCIM_Cubic, leaveTangentWeight=0.0962, tangentMode=RCTM_Break, time=0, value=0}
              {interpMode=RCIM_Cubic, tangentMode=RCTM_Break, time=0.2, value=0.7}
              {interpMode=RCIM_Cubic, tangentMode=RCTM_Break, time=1, value=0}
             ]}}}
          Scale Curve=&RandomRangeFloat(Minimum=0.8, Maximum=1)
         )
         B=&Multiply_Float(A=&FNTimeOfDayScalar(Night Scale=1, Morning Scale=0.8, Evening Scale=0.8), B=1)
        )
       )
       Passthrough Linear Color=$Particles.Initial.Color
       Spawn Group Mask=2
      )
      Spawn Group Mask=1
     )
    )) {isSetParameters=true, module=SetVariables_#22}
   &CurlNoiseForce(
    Noise Frequency=0.0012
    Noise Strength=&MaskFloatBySpawnGroup(
     Masked Float=0
     Passthrough Float=&MaskFloatBySpawnGroup(
      Masked Float=0
      Passthrough Float=&MaskFloatBySpawnGroup(Masked Float=0, Passthrough Float=800, Spawn Group Mask=2)
      Spawn Group Mask=1
     )
    )
    Pan Noise=true
    Pan Noise Field=<0 0 0.3>
   )
   &Drag(
    Drag=&MaskFloatBySpawnGroup(
     Masked Float=1.5
     Passthrough Float=&MaskFloatBySpawnGroup(
      Masked Float=&RandomRangeFloat(Minimum=0.8, Maximum=1.2)
      Passthrough Float=&MaskFloatBySpawnGroup(Masked Float=0, Passthrough Float=20, Spawn Group Mask=2)
      Spawn Group Mask=1
     )
    )
    Rotational Drag:NiagaraFloat=&MaskFloatBySpawnGroup(
     Masked Float=&FloatFromCurve(FloatCurve=DataInterface() {propertyValues {Curve {keys [<at 0 0 RCIM_Cubic>, {arriveTangent=0.1846, interpMode=RCIM_Cubic, leaveTangent=0.1846, tangentMode=RCTM_User, time=1, value=0.1}]}}})
     Passthrough Float=&MaskFloatBySpawnGroup(Masked Float=0, Passthrough Float=0, Spawn Group Mask=1)
    )
    Use Linear Drag=true
    Use Rotational Drag=true
   )
   &SolveForcesAndVelocity(Clamp Velocity=false, Limit Acceleration=false)
   &ScaleSpriteSize(
    Scale Factor=&MaskVector2dBySpawnGroup(
     Masked Vector2D=<0 0>
     Passthrough Vector2D=&MaskVector2dBySpawnGroup(
      Masked Vector2D=<1 1>
      Passthrough Vector2D=&MaskVector2dBySpawnGroup(
       Masked Vector2D=&Vector2DFromCurve(
        Vector2Curve=DataInterface() {propertyValues {XCurve {keys [<at 0.0036 1.6897>,<at 0.9967 0.0582>]}, YCurve {keys [<at 0 1>,<at 1.002 3.7227>]}}}
        ScaleCurve=<0.4 0.4>
       )
       Passthrough Vector2D=&Vector2DFromFloat(Value=&FloatFromCurve(FloatCurve=DataInterface() {propertyValues {Curve {keys [
             {arriveTangent=6.2202, interpMode=RCIM_Cubic, leaveTangent=6.2202, tangentMode=RCTM_User, time=0, value=0.15}
             {arriveTangent=1.0025, interpMode=RCIM_Cubic, leaveTangent=1.0025, tangentMode=RCTM_User, time=0.2, value=0.8}
             {arriveTangent=0.0143, interpMode=RCIM_Cubic, leaveTangent=0.0143, tangentMode=RCTM_User, time=1.0463, value=1}
            ]}}}))
       Spawn Group Mask=2
      )
      Spawn Group Mask=1
     )
     Spawn Group Mask=0
    )
    Scale Sprite Size Mode=::NewEnumerator2
   )
   &ScaleSpriteSizeBySpeed(
    Max Scale Factor=&MaskVector2dBySpawnGroup(
     Masked Vector2D=<1 1>
     Passthrough Vector2D=&MaskVector2dBySpawnGroup(
      Masked Vector2D=<0.5 7>
      Passthrough Vector2D=&MaskVector2dBySpawnGroup(Masked Vector2D=<1 1>, Passthrough Vector2D=<1 1>, Spawn Group Mask=2)
      Spawn Group Mask=1
     )
     Spawn Group Mask=0
    )
    Sample Scale Factor By Curve=true
    Velocity Threshold=500
   )
   &ScaleMeshSize(Scale Factor=&VectorFromFloat(Value=&FloatFromCurve(FloatCurve=DataInterface() {propertyValues {Curve {keys [{arriveTangent=-0.0035, interpMode=RCIM_Cubic, leaveTangent=-0.0035, tangentMode=RCTM_User, time=0.6, value=1}, <at 1 0>]}}})))
   /SetVariables(Particles.SandDMP:Vector4f=&MakeVector4(
     X=0
     Y=1
     Z=1
     W=&FloatFromCurve(FloatCurve=DataInterface() {propertyValues {Curve {keys [{interpMode=RCIM_Cubic, tangentMode=RCTM_Break, time=0.3, value=1}, {interpMode=RCIM_Cubic, tangentMode=RCTM_Break, time=1, value=0.25}]}}})
    )) {isSetParameters=true, module=SetVariables_#23}
   &DynamicMaterialParameters(
    Index 0 Param 1=&FloatFromCurve(
     FloatCurve=DataInterface() {propertyValues {Curve {keys [
         {arriveTangent=2.7533, interpMode=RCIM_Cubic, leaveTangent=1.8958, tangentMode=RCTM_User, time=0, value=1}
         {arriveTangent=0.8292, interpMode=RCIM_Cubic, leaveTangent=0.8292, tangentMode=RCTM_User, time=0.321, value=1.3453}
         {arriveTangent=0.0529, interpMode=RCIM_Cubic, leaveTangent=0.2021, tangentMode=RCTM_User, time=1, value=1.5}
        ]}}}
     CurveIndex=&Add_Float(A=&Multiply_Float(A=$Particles.Initial.MaterialRandom, B=0.5), B=$Particles.NormalizedAge)
    )
    Index 0 Param 2=&FloatFromCurve(FloatCurve=DataInterface() {propertyValues {Curve {keys [
         {arriveTangent=0.0453, interpMode=RCIM_Cubic, tangentMode=RCTM_Break, time=0, value=-1}
         {arriveTangent=0.0604, interpMode=RCIM_Cubic, leaveTangent=0.0604, tangentMode=RCTM_User, time=1, value=-1}
        ]}}})
    Index 0 Param 3=&FloatFromCurve(FloatCurve=DataInterface() {propertyValues {Curve {keys [
         {arriveTangent=-53.9199, interpMode=RCIM_Cubic, leaveTangent=-8.9727, tangentMode=RCTM_User, time=0, value=1}
         {arriveTangent=0.2347, interpMode=RCIM_Cubic, leaveTangent=0.2723, tangentMode=RCTM_Break, time=0.25, value=0}
        ]}}})
    Index 0 Param 4=&FloatFromCurve(FloatCurve=DataInterface() {propertyValues {Curve {keys [<at 0.5 1>,<at 1 0>]}}})
    Index 1 Param 1=&FloatFromCurve(
     FloatCurve=DataInterface() {propertyValues {Curve {keys [
         {interpMode=RCIM_Cubic, leaveTangent=4.7088, tangentMode=RCTM_Break, time=0, value=0}
         {arriveTangent=4.2257, interpMode=RCIM_Cubic, leaveTangent=4.2257, tangentMode=RCTM_User, time=0.25, value=1}
        ]}}}
     Scale Curve=&RandomRangeFloat(Minimum=1.5, Maximum=1.5)
    )
    Index 1 Param 2=&FloatFromCurve(
     FloatCurve=DataInterface() {propertyValues {Curve {keys [<at 0 0.2>, {leaveTangent=0.9066, time=1, value=2}]}}}
     Scale Curve=&RandomRangeFloat(Minimum=0.5, Maximum=0.6)
    )
    Index 1 Param 3=&FloatFromCurve(FloatCurve=DataInterface() {propertyValues {Curve {keys [
         {interpMode=RCIM_Cubic, leaveTangent=6.3917, tangentMode=RCTM_Break, time=0, value=0.2}
         {arriveTangent=1.7549, interpMode=RCIM_Cubic, leaveTangent=10.6767, tangentMode=RCTM_User, time=0.25, value=1}
        ]}}})
    Index 1 Param 4=&FloatFromCurve(
     FloatCurve=DataInterface() {propertyValues {Curve {keys [{arriveTangent=-0.136, interpMode=RCIM_Cubic, leaveTangent=-0.136, tangentMode=RCTM_User, time=0, value=1}, <at 1 0.95>]}}}
     Scale Curve=&RandomRangeFloat(Minimum=1.1, Maximum=1.4)
    )
    Index 2 Param 1=&FloatFromCurve(FloatCurve=DataInterface() {propertyValues {Curve {keys [
         {interpMode=RCIM_Cubic, leaveTangent=4.2485, tangentMode=RCTM_Break, time=0, value=0}
         {arriveTangent=1.0548, interpMode=RCIM_Cubic, leaveTangent=1.0548, tangentMode=RCTM_User, time=0.17, value=0.5715}
         <at 1 1>
        ]}}})
    Index 2 Param 2=&FloatFromCurve(
     FloatCurve=DataInterface() {propertyValues {Curve {keys [
         {arriveTangent=1.9056, interpMode=RCIM_Cubic, leaveTangent=1.9056, tangentMode=RCTM_User, time=0, value=0.2}
         {arriveTangent=0.0259, interpMode=RCIM_Cubic, leaveTangent=0.0259, tangentMode=RCTM_User, time=1, value=1.4}
        ]}}}
     Scale Curve=&RandomRangeFloat(Minimum=0.2, Maximum=0.7)
    )
    Index 2 Param 3=&FloatFromCurve(FloatCurve=DataInterface() {propertyValues {Curve {keys [
         {interpMode=RCIM_Cubic, leaveTangent=2.3064, tangentMode=RCTM_Break, time=0, value=1}
         {interpMode=RCIM_Cubic, leaveTangent=11.1233, tangentMode=RCTM_Break, time=1, value=2}
        ]}}})
    Index 2 Param 4=&FloatFromCurve(
     FloatCurve=DataInterface() {propertyValues {Curve {keys [
         {arriveTangent=-0.6867, interpMode=RCIM_Cubic, leaveTangent=-0.6867, tangentMode=RCTM_User, time=0, value=1}
         {interpMode=RCIM_Cubic, leaveTangent=-1.2956, tangentMode=RCTM_Break, time=1, value=0.75}
        ]}}}
     Scale Curve=1.6
    )
    Module.Param0WriteEnabled=true
    Module.Param0WriteEnabled1=true
    Module.Param0WriteEnabled2=true
    Module.Param1WriteEnabled=false
    Module.Param1WriteEnabled1=true
    Module.Param1WriteEnabled2=true
    Module.Param2WriteEnabled=true
    Module.Param2WriteEnabled1=true
    Module.Param2WriteEnabled2=true
    Module.Param3WriteEnabled=true
    Module.Param3WriteEnabled1=true
    Module.Param3WriteEnabled2=true
    Write Parameter Index 1=true
    Write Parameter Index 2=true
   ) {module=DynamicMaterialParameters001}
   &ScaleColor(
    Scale Alpha=&MaskFloatBySpawnGroup(
     Masked Float=1
     Passthrough Float=&MaskFloatBySpawnGroup(
      Masked Float=1
      Passthrough Float=&MaskFloatBySpawnGroup(
       Masked Float=1
       Passthrough Float=&Multiply_Float(
        A=&FloatFromCurve(
         FloatCurve=DataInterface() {propertyValues {Curve {keys [
             <at 0.1 0.75>
             {interpMode=RCIM_Cubic, tangentMode=RCTM_User, time=0.2, value=2.25}
             {interpMode=RCIM_Cubic, tangentMode=RCTM_User, time=0.7612, value=1.5}
             <at 1 0>
            ]}}}
         Scale Curve=&RandomRangeFloat(Minimum=1, Maximum=1.5)
        )
        B=&Multiply_Float(A=&FNTimeOfDayScalar(Night Scale=1, Morning Scale=0.7, Evening Scale=0.7), B=1.5)
       )
       Spawn Group Mask=2
      )
      Spawn Group Mask=1
     )
    )
    Scale RGB=&MaskVectorBySpawnGroup(
     Masked Vector=<1 1 1>
     Passthrough Vector=&MaskVectorBySpawnGroup(
      Masked Vector=<1 1 1>
      Passthrough Vector=&MaskVectorBySpawnGroup(
       Masked Vector=<1 1 1>
       Passthrough Vector=&VectorFromFloat(Value=&FloatFromCurve(FloatCurve=DataInterface() {propertyValues {Curve {keys [{arriveTangent=1.8145, interpMode=RCIM_Cubic, leaveTangent=0.4403, tangentMode=RCTM_Break, time=0, value=0.4}, <at 0.4 0.625>, <at 1 1>]}}}))
       Spawn Group Mask=2
      )
      Spawn Group Mask=1
     )
    )
    ScaleA=true
    ScaleRGB=true
   )
   &CameraOffset(Camera Offset Amount=&MaskFloatBySpawnGroup(
     Masked Float=0
     Passthrough Float=&MaskFloatBySpawnGroup(
      Masked Float=0
      Passthrough Float=&MaskFloatBySpawnGroup(Masked Float=0, Passthrough Float=10, Spawn Group Mask=2)
      Spawn Group Mask=1
     )
    ))
  ]
  rendererClasses [&NiagaraMeshRendererProperties, &NiagaraSpriteRendererProperties]
  renderers [
   {rendererClass=&NiagaraMeshRendererProperties, rendererIndex=0}
   {rendererClass=&NiagaraSpriteRendererProperties, rendererIndex=1}
   {rendererClass=&NiagaraSpriteRendererProperties, rendererIndex=2}
   {rendererClass=&NiagaraSpriteRendererProperties, rendererIndex=3}
   {rendererClass=&NiagaraSpriteRendererProperties, rendererIndex=4}
  ]
 }
 Rocks_Burst {
  bEnabled=false
  emitterUpdateScript [
   &EmitterState(Life Cycle Mode=::NewEnumerator1, Loop Behavior=::NewEnumerator1, Loop Duration=1, UseLoopDelay=true)
   &SpawnBurst_Instantaneous(
    Spawn Count=&RandomRangeInt(Minimum=7, Maximum=8)
    Use Spawn Probability=false
    UseLoopCountLimit=false
   )
   /SetVariables() {isSetParameters=true, module=SetVariables_#16}
  ]
  particleSpawnScript [
   &InitializeParticle(
    Lifetime Max=0.8
    Lifetime Min=0.6
    Lifetime Mode=::NewEnumerator1
    Mesh Scale Mode=::NewEnumerator3
    Mesh Uniform Scale=&Multiply_Float(
     A=&Lerp_Float(
      A=0.0285
      B=0.06
      Alpha=&FloatFromCurve(
       FloatCurve=DataInterface() {propertyValues {Curve {keys [<at 0 0 RCIM_Constant>, {arriveTangent=7.6187, interpMode=RCIM_Constant, leaveTangent=7.6187, time=0.65, value=1}]}}}
       CurveIndex=&RandomRangeFloat(Minimum=0, Maximum=1)
      )
     )
     B=1.4
    )
    UsePositionOffset=false
    Write Material Random=true
   )
   &ShapeLocation(
    Handle Radius=0
    Large Radius=2
    MaskSpawnGroup=false
    Offset=<0 0 2>
    Offset Mode=::NewEnumerator0
    Override Seed=false
    Shape Primitive=::NewEnumerator3
    UseOwnerScale=true
   )
   &InitialMeshOrientation(Use Rotation Vector=false)
   &AddRotationalVelocity(Rotation Rate=&Multiply_VectorByFloat(
     Vector=&RandomVector(Vector Scale=2)
     Float=&RandomRangeFloat(Minimum=0.5, Maximum=1)
    ))
   &DynamicMaterialParameters(
    Index 0 Param 2=1
    Module.Param0WriteEnabled=false
    Module.Param1WriteEnabled=true
    Module.Param2WriteEnabled=false
    Module.Param3WriteEnabled=false
   )
   &AddVelocity(
    Module.Use Falloff Distance=false
    Override Seed=false
    Velocity Mode=::NewEnumerator1
    Velocity Speed=&RandomRangeFloat(Minimum=200, Maximum=550)
   )
   &AddVelocity(
    Cone Angle=35
    Cone Axis=&RotateVectorByQuaternion(
     Vector To Rotate=<1 0 0>
     Quaternion=&MakeQuatFromAxisAngle(Axis=<0 0 1>, AngleInDegrees=-145, NormalizeAxis=true)
    )
    Module.UseConeFalloff=false
    Module.UseVelDistribution=false
    Override Seed=false
    Velocity Mode=::NewEnumerator2
    Velocity Speed=&RandomRangeFloat(Minimum=70, Maximum=350)
   ) {module=AddVelocity001}
   &ApplyInitialForces()
   /SetVariables(Particles.MeshIndex=&RandomRangeInt(Minimum=0, Maximum=2)) {isSetParameters=true, module=SetVariables_#19}
  ]
  particleUpdateScript [
   &ParticleState()
   &GravityForce(Gravity=&RandomRangeVector(Minimum=<0 0 -750>, Maximum=<0 0 -1500>))
   &Collision(
    Age Colliding Particles=false
    Module.Control Rest State Via Surface Normal=false
    Module.EnableMaxCollisionCount=false
    Module.Randomize Collision Normal Vector=false
    Restitution=0.2
   )
   &Drag(
    Drag=1.5
    Rotational Drag:NiagaraFloat=&FloatFromCurve(FloatCurve=DataInterface() {propertyValues {Curve {keys [<at 0 0 RCIM_Cubic>, {arriveTangent=0.1846, interpMode=RCIM_Cubic, leaveTangent=0.1846, tangentMode=RCTM_User, time=1, value=0.1}]}}})
    Use Linear Drag=true
    Use Rotational Drag=true
   )
   &SolveForcesAndVelocity(Clamp Velocity=false, Limit Acceleration=false)
   &ScaleMeshSize(Scale Factor=&VectorFromFloat(Value=&FloatFromCurve(FloatCurve=DataInterface() {propertyValues {Curve {keys [{arriveTangent=-0.0035, interpMode=RCIM_Cubic, leaveTangent=-0.0035, tangentMode=RCTM_User, time=0.6, value=1}, <at 1 0>]}}})))
   &SolveRotationalForcesAndVelocity()
  ]
  rendererClasses [&NiagaraMeshRendererProperties]
  renderers [{rendererClass=&NiagaraMeshRendererProperties, rendererIndex=0}]
 }
 Sand001 {
  bEnabled=false
  emitterUpdateScript [
   &EmitterState()
   &SpawnBurst_Instantaneous(Spawn Count=10, Use Spawn Probability=false, UseLoopCountLimit=false)
   /SetVariables(Emitter.DepthFade=100) {isSetParameters=true}
  ]
  particleSpawnScript [
   &InitializeParticle(
    Color Maximum=<rgba 0.2708 0.2506 0.2327 1>
    Color Minimum=<rgba 0.0729 0.0681 0.063 0.8>
    Color Mode=::NewEnumerator2
    Lifetime Max=0.25
    Lifetime Min=0.2
    Lifetime Mode=::NewEnumerator1
    Position Offset=<0 0 -10>
    Sprite Rotation Mode=::NewEnumerator2
    Sprite Size Max=<75 88>
    Sprite Size Min=<65 75>
    Sprite Size Mode=::NewEnumerator2
    UsePositionOffset=true
    Write Material Random=true
   )
   &ShapeLocation(
    Handle Radius=3
    Large Radius=10
    MaskSpawnGroup=false
    Non Uniform Scale=<0.3 0.3 0.3>
    Override Seed=false
    Shape Origin=$Particles.Position
    Shape Primitive=::NewEnumerator3
    UseOwnerScale=true
   )
   &SubUVAnimation(End Frame=3, SubUV Animation Mode:ENiagara_SubUVLookupMode=::NewEnumerator2)
   &AddVelocity(Module.Use Falloff Distance=false, Origin Offset=<0 0 -15>, Override Seed=false, Velocity Mode=::NewEnumerator1, Velocity Speed=150)
   &ApplyInitialForces()
  ]
  particleUpdateScript [
   &ParticleState()
   &SolveForcesAndVelocity(Clamp Velocity=false, Limit Acceleration=false)
   &ScaleSpriteSize(Scale Factor=&Vector2DFromCurve(
     Vector2Curve=DataInterface() {propertyValues {XCurve {keys [<at 0.0036 1.6897>,<at 0.9967 0.0582>]}, YCurve {keys [<at 0 1>,<at 1.002 3.7227>]}}}
     ScaleCurve=<0.4 0.4>
    ))
   &ScaleColor(
    Scale Alpha=&Multiply_Float(
     A=&FloatFromCurve(
      FloatCurve=DataInterface() {propertyValues {Curve {keys [
          {interpMode=RCIM_Cubic, leaveTangentWeight=0.0962, tangentMode=RCTM_Break, time=0, value=0}
          {interpMode=RCIM_Cubic, tangentMode=RCTM_Break, time=0.2, value=1}
          {interpMode=RCIM_Cubic, tangentMode=RCTM_Break, time=1, value=0}
         ]}}}
      Scale Curve=&RandomRangeFloat(Minimum=0.8, Maximum=1)
     )
     B=&Multiply_Float(A=&FNTimeOfDayScalar(Night Scale=1, Morning Scale=0.8, Evening Scale=0.8), B=1)
    )
    Scale RGB=&MakeVectorFromLinearColorRGB(Color=&ColorFromCurve(ColorCurve=DataInterface() {propertyValues {
        AlphaCurve {keys [<at 0 1 RCIM_Cubic>,<at 1 1 RCIM_Cubic>]}
        BlueCurve {keys [<at 0.2382 2>, {arriveTangent=-1.3127, interpMode=RCIM_Cubic, leaveTangent=-1.3127, time=0.383, value=1}, <at 1 1 RCIM_Cubic>]}
        GreenCurve {keys [<at 0.2382 0.9969>, {arriveTangent=0.0041, interpMode=RCIM_Cubic, leaveTangent=0.0041, time=0.383, value=0.5676}, <at 1 1 RCIM_Cubic>]}
        RedCurve {keys [<at 0.2382 0>, {arriveTangent=1.3127, interpMode=RCIM_Cubic, leaveTangent=1.3127, time=0.383, value=0.4328}, <at 1 1 RCIM_Cubic>]}
       }}))
    ScaleA=true
    ScaleRGB=true
   )
   &DynamicMaterialParameters(
    Index 0 Param 2=1
    Index 0 Param 3=1
    Index 0 Param 4=&FloatFromCurve(FloatCurve=DataInterface() {propertyValues {Curve {keys [{interpMode=RCIM_Cubic, tangentMode=RCTM_Break, time=0.3, value=1}, {interpMode=RCIM_Cubic, tangentMode=RCTM_Break, time=1, value=0.25}]}}})
    Module.Param0WriteEnabled=true
    Module.Param1WriteEnabled=true
    Module.Param2WriteEnabled=true
    Module.Param3WriteEnabled=true
   )
  ]
  rendererClasses [&NiagaraSpriteRendererProperties]
  renderers [{rendererClass=&NiagaraSpriteRendererProperties}]
 }
 Smoke_Sphere001 {
  bEnabled=false
  emitterUpdateScript [
   &EmitterState(Life Cycle Mode=::NewEnumerator1, Loop Behavior=::NewEnumerator1, Loop Delay=0.05, Loop Duration=0.1, UseLoopDelay=true)
   &SpawnBurst_Instantaneous(Spawn Count=5, Use Spawn Probability=false)
  ]
  particleSpawnScript [
   &InitializeParticle(
    Color=<rgba 0.224 0.224 0.224 0.8>
    Lifetime=&RandomRangeFloat(Minimum=0.5, Maximum=1.5)
    Sprite Size Mode=::NewEnumerator1
    Uniform Sprite Size Max=125
    Uniform Sprite Size Min=75
    UsePositionOffset=false
    Write Material Random=true
   )
   /SetVariables(Particles.SpriteSize=&Multiply_Vector2D_ByFloat(
     Float=&FN_LOD_Distance_Remap(Start Distance=5000, End Distance=25000, Minimum Output=1, Maximum Output=3)
     Vector2D=&RandomRangeVector2D(
      Minimum=&Vector2DFromFloat(Value=600)
      Maximum=&Vector2DFromFloat(Value=1000)
     )
    )) {enabled=false, isSetParameters=true}
   /SetVariables(Particles.InitSpriteRotRate=&UniformRangedFloat(Minimum=-3, Maximum=3)) {isSetParameters=true}
   &SphereLocation(
    MaskSpawnGroup=false
    Non Uniform Scale=<1 1 0.25>
    Override Seed=true
    Random Seed=20
    Sphere Radius=7
    Surface Only=false
    UseNonUniformScale=true
    UseOffset=false
    UseSphereOrientation=false
   ) {enabled=false}
   &ShapeLocation(Handle Radius=3, Large Radius=5, MaskSpawnGroup=false, Override Seed=false, Shape Primitive=::NewEnumerator3, UseOwnerScale=true)
   /SetVariables(Particles.Position=&MaskVectorBySpawnGroup(
     Masked Vector=&Add_Vector(A=Linked() {linkedVariable {name=Engine.Owner.Position, type=NiagaraPosition}}, B=<0 0 10>)
     Passthrough Vector=Linked() {linkedVariable {name=Particles.Position, type=NiagaraPosition}}
     Spawn Group Mask=1
    )) {enabled=false, isSetParameters=true}
   &AddVelocityFromPoint(
    Module.Use Falloff Distance=false
    Origin Offset=<0 0 -1>
    Velocity Strength=&RandomRangeFloat(Minimum=700, Maximum=1000)
   )
  ]
  particleUpdateScript [
   &ParticleState()
   &SpriteRotationRate(Rotation Rate=&Multiply_Float(
     A=$Particles.InitSpriteRotRate
     B=&FloatFromCurve(
      FloatCurve=DataInterface() {propertyValues {Curve {keys [
          {arriveTangent=-4.0107, interpMode=RCIM_Cubic, leaveTangent=-4.0107, tangentMode=RCTM_User, time=-0.0046, value=1.0113}
          {arriveTangent=-0.2496, interpMode=RCIM_Cubic, leaveTangent=-0.2496, tangentMode=RCTM_User, time=0.5, value=0.16}
         ]}}}
      CurveIndex=$Emitter.Age
      Scale Curve=150
     )
    ))
   &ScaleColor(
    Scale Alpha=&Multiply_Float(
     A=&FloatFromCurve(
      FloatCurve=DataInterface() {propertyValues {Curve {keys [
          <at 0.1 0.75>
          {interpMode=RCIM_Cubic, tangentMode=RCTM_User, time=0.2, value=2.25}
          {interpMode=RCIM_Cubic, tangentMode=RCTM_User, time=0.7612, value=1.5}
          <at 1 0>
         ]}}}
      Scale Curve=&RandomRangeFloat(Minimum=0.6, Maximum=0.9)
     )
     B=&Multiply_Float(A=&FNTimeOfDayScalar(Night Scale=1, Morning Scale=0.7, Evening Scale=0.7), B=1.5)
    )
    Scale RGB=&VectorFromFloat(Value=&FloatFromCurve(FloatCurve=DataInterface() {propertyValues {Curve {keys [{arriveTangent=1.8145, interpMode=RCIM_Cubic, leaveTangent=0.4403, tangentMode=RCTM_Break, time=0, value=0.4}, <at 0.4 0.625>, <at 1 1>]}}}))
    ScaleA=true
    ScaleRGB=true
   )
   &CurlNoiseForce(Noise Frequency=0.0012, Noise Strength=800, Pan Noise=true, Pan Noise Field=<0 0 0.3>)
   &Drag(Drag=20, Use Linear Drag=true, Use Rotational Drag=false)
   &SolveForcesAndVelocity(Clamp Velocity=false, Limit Acceleration=false)
   &ScaleSpriteSize(Scale Factor=&Vector2DFromFloat(Value=&FloatFromCurve(FloatCurve=DataInterface() {propertyValues {Curve {keys [
          {arriveTangent=6.2202, interpMode=RCIM_Cubic, leaveTangent=6.2202, tangentMode=RCTM_User, time=0, value=0.15}
          {arriveTangent=1.0025, interpMode=RCIM_Cubic, leaveTangent=1.0025, tangentMode=RCTM_User, time=0.2, value=0.8}
          {arriveTangent=0.0143, interpMode=RCIM_Cubic, leaveTangent=0.0143, tangentMode=RCTM_User, time=1.0463, value=1}
         ]}}})))
   &DynamicMaterialParameters(
    Index 0 Param 1=&Add_Float(
     A=$Particles.RandomFloat
     B=&FloatFromCurve(
      FloatCurve=DataInterface() {propertyValues {Curve {keys [
          {arriveTangent=2.7533, interpMode=RCIM_Cubic, leaveTangent=1.8958, tangentMode=RCTM_User, time=0, value=1}
          {arriveTangent=0.8292, interpMode=RCIM_Cubic, leaveTangent=0.8292, tangentMode=RCTM_User, time=0.321, value=1.3453}
          {arriveTangent=0.0529, interpMode=RCIM_Cubic, leaveTangent=0.2021, tangentMode=RCTM_User, time=1, value=1.5}
         ]}}}
      CurveIndex=&Add_Float(A=&Multiply_Float(A=$Particles.Initial.MaterialRandom, B=0.5), B=$Particles.NormalizedAge)
     )
    )
    Index 0 Param 2=&FloatFromCurve(FloatCurve=DataInterface() {propertyValues {Curve {keys [
         {arriveTangent=0.0453, interpMode=RCIM_Cubic, tangentMode=RCTM_Break, time=0, value=-1}
         {arriveTangent=0.0604, interpMode=RCIM_Cubic, leaveTangent=0.0604, tangentMode=RCTM_User, time=1, value=-1}
        ]}}})
    Index 0 Param 3=&FloatFromCurve(FloatCurve=DataInterface() {propertyValues {Curve {keys [
         {arriveTangent=-53.9199, interpMode=RCIM_Cubic, leaveTangent=-8.9727, tangentMode=RCTM_User, time=0, value=1}
         {arriveTangent=0.2347, interpMode=RCIM_Cubic, leaveTangent=0.2723, tangentMode=RCTM_Break, time=0.25, value=0}
        ]}}})
    Index 0 Param 4=&FloatFromCurve(FloatCurve=DataInterface() {propertyValues {Curve {keys [<at 0.5 1>,<at 1 0>]}}})
    Index 1 Param 1=&FloatFromCurve(
     FloatCurve=DataInterface() {propertyValues {Curve {keys [
         {interpMode=RCIM_Cubic, leaveTangent=4.7088, tangentMode=RCTM_Break, time=0, value=0}
         {arriveTangent=4.2257, interpMode=RCIM_Cubic, leaveTangent=4.2257, tangentMode=RCTM_User, time=0.25, value=1}
        ]}}}
     Scale Curve=&RandomRangeFloat(Minimum=1.5, Maximum=1.5)
    )
    Index 1 Param 2=&FloatFromCurve(
     FloatCurve=DataInterface() {propertyValues {Curve {keys [<at 0 0.2>, {leaveTangent=0.9066, time=1, value=2}]}}}
     Scale Curve=&RandomRangeFloat(Minimum=0.5, Maximum=0.6)
    )
    Index 1 Param 3=&FloatFromCurve(FloatCurve=DataInterface() {propertyValues {Curve {keys [
         {interpMode=RCIM_Cubic, leaveTangent=6.3917, tangentMode=RCTM_Break, time=0, value=0.2}
         {arriveTangent=1.7549, interpMode=RCIM_Cubic, leaveTangent=10.6767, tangentMode=RCTM_User, time=0.25, value=1}
        ]}}})
    Index 1 Param 4=&FloatFromCurve(
     FloatCurve=DataInterface() {propertyValues {Curve {keys [{arriveTangent=-0.136, interpMode=RCIM_Cubic, leaveTangent=-0.136, tangentMode=RCTM_User, time=0, value=1}, <at 1 0.95>]}}}
     Scale Curve=&RandomRangeFloat(Minimum=1.1, Maximum=1.4)
    )
    Index 2 Param 1=&FloatFromCurve(FloatCurve=DataInterface() {propertyValues {Curve {keys [
         {interpMode=RCIM_Cubic, leaveTangent=4.2485, tangentMode=RCTM_Break, time=0, value=0}
         {arriveTangent=1.0548, interpMode=RCIM_Cubic, leaveTangent=1.0548, tangentMode=RCTM_User, time=0.17, value=0.5715}
         <at 1 1>
        ]}}})
    Index 2 Param 2=&FloatFromCurve(
     FloatCurve=DataInterface() {propertyValues {Curve {keys [
         {arriveTangent=1.9056, interpMode=RCIM_Cubic, leaveTangent=1.9056, tangentMode=RCTM_User, time=0, value=0.2}
         {arriveTangent=0.0259, interpMode=RCIM_Cubic, leaveTangent=0.0259, tangentMode=RCTM_User, time=1, value=1.4}
        ]}}}
     Scale Curve=&RandomRangeFloat(Minimum=0.2, Maximum=0.7)
    )
    Index 2 Param 3=&FloatFromCurve(FloatCurve=DataInterface() {propertyValues {Curve {keys [
         {interpMode=RCIM_Cubic, leaveTangent=2.3064, tangentMode=RCTM_Break, time=0, value=1}
         {interpMode=RCIM_Cubic, leaveTangent=11.1233, tangentMode=RCTM_Break, time=1, value=2}
        ]}}})
    Index 2 Param 4=&FloatFromCurve(
     FloatCurve=DataInterface() {propertyValues {Curve {keys [
         {arriveTangent=-0.6867, interpMode=RCIM_Cubic, leaveTangent=-0.6867, tangentMode=RCTM_User, time=0, value=1}
         {interpMode=RCIM_Cubic, leaveTangent=-1.2956, tangentMode=RCTM_Break, time=1, value=0.75}
        ]}}}
     Scale Curve=1.6
    )
    Module.Param0WriteEnabled=true
    Module.Param0WriteEnabled1=true
    Module.Param0WriteEnabled2=true
    Module.Param1WriteEnabled=false
    Module.Param1WriteEnabled1=true
    Module.Param1WriteEnabled2=true
    Module.Param2WriteEnabled=true
    Module.Param2WriteEnabled1=true
    Module.Param2WriteEnabled2=true
    Module.Param3WriteEnabled=true
    Module.Param3WriteEnabled1=true
    Module.Param3WriteEnabled2=true
    Write Parameter Index 1=true
    Write Parameter Index 2=true
   )
   &CameraOffset(Camera Offset Amount=35)
  ]
  rendererClasses [&NiagaraSpriteRendererProperties]
  renderers [
   {rendererClass=&NiagaraSpriteRendererProperties}
   {rendererClass=&NiagaraSpriteRendererProperties, rendererIndex=1}
  ]
 }
}
summary {
 systemName=NS_Emote_GlowFang_GroundImpact
 emitters [
  {
   emitterName=Particulate_Burst_with_Smoke
   simTarget=GPUComputeSim
   rendererClasses [&NiagaraMeshRendererProperties, &NiagaraSpriteRendererProperties]
  }
  {
   emitterName=CentralBlast
   rendererClasses [
    &NiagaraSpriteRendererProperties
    &NiagaraLightRendererProperties
    &NiagaraMeshRendererProperties
   ]
  }
  {bEnabled=false, emitterName=Lightning, rendererClasses [&NiagaraMeshRendererProperties]}
  {bEnabled=false, emitterName=NE_GroundPulse001, rendererClasses [&NiagaraMeshRendererProperties]}
  {bEnabled=false, emitterName=Rocks_Burst, rendererClasses [&NiagaraMeshRendererProperties]}
  {
   bEnabled=false
   emitterName=NE_StarGlow
   rendererClasses [
    &NiagaraSpriteRendererProperties
    &NiagaraLightRendererProperties
    &NiagaraMeshRendererProperties
   ]
  }
  {bEnabled=false, emitterName=Beams001, rendererClasses [&NiagaraMeshRendererProperties]}
  {bEnabled=false, emitterName=Sand001, rendererClasses [&NiagaraSpriteRendererProperties]}
  {bEnabled=false, emitterName=LightningSparks001, rendererClasses [&NiagaraSpriteRendererProperties]}
  {bEnabled=false, emitterName=Smoke_Sphere001, rendererClasses [&NiagaraSpriteRendererProperties]}
  {bEnabled=false, emitterName=NE_Lightning_Streak, rendererClasses [&NiagaraMeshRendererProperties]}
  {bEnabled=false, emitterName=Lightning_Ground_Star001, rendererClasses [&NiagaraMeshRendererProperties]}
  {bEnabled=false, emitterName=NE_LoopingZaps001, rendererClasses [&NiagaraMeshRendererProperties]}
  {bEnabled=false, emitterName=HangingParticulates001, rendererClasses [&NiagaraSpriteRendererProperties]}
  {
   bEnabled=false
   emitterName=Particulate_Burst_with_Smoke001
   simTarget=GPUComputeSim
   rendererClasses [&NiagaraMeshRendererProperties, &NiagaraSpriteRendererProperties]
  }
  {
   bEnabled=false
   emitterName=CentralBlast001
   rendererClasses [
    &NiagaraSpriteRendererProperties
    &NiagaraLightRendererProperties
    &NiagaraMeshRendererProperties
   ]
  }
 ]
 userVariables [
  User.bIsFrontEnd {type=NiagaraBool}
  User.bIsFrontEndPreview {type=NiagaraBool}
  User.bKillStage {type=NiagaraInt32}
  User.NormalizedNotifyProgress {type=NiagaraFloat, defaultValue=0, description=""}
 ]
}
