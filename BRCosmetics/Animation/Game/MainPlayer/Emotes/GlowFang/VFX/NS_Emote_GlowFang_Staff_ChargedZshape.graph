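# Niagara effect for the GlowFang emote that draws charged lightning on the staff, presumably forming a Z shape from three socket-anchored segments.
# One emitter runs three spawn groups whose rates fade over the loop, each picking a skeletal mesh socket with its own cylinder offset and orientation.
# Short-lived particles take System.ElimColor, flicker through color curves and dynamic material parameters, and feed mesh and sprite renderers via a visibility tag, plus strobing red light attributes.
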NiagaraSystem /BRCosmetics/Animation/Game/MainPlayer/Emotes/GlowFang/VFX/NS_Emote_GlowFang_Staff_ChargedZshape
dependencies {
 usedDataInterfaces [&NiagaraDataInterfaceCurve, &NiagaraDataInterfaceSkeletalMesh]
 usedDynamicInputs [
  &MakeCustomVectorFromBool
  &SetBoolByFloatComparison
  &MergeHelper_FloatSwitch
  &FNTimeOfDayScalar
  &InvertBool
 ]
 usedModules [
  &self:SystemSpawnScript.SystemScriptSource.SystemScriptGraph.NiagaraNodeAssignment_4.SetVariables_#1
  &SystemState
  &self:SystemSpawnScript.SystemScriptSource.SystemScriptGraph.NiagaraNodeAssignment_0.SetVariables_#2
  &self:SystemSpawnScript.SystemScriptSource.SystemScriptGraph.NiagaraNodeAssignment_2.SetVariables_#3
  &self:SystemSpawnScript.SystemScriptSource.SystemScriptGraph.NiagaraNodeAssignment_3.SetVariables_#4
 ]
}
emitters {StaffLightning {
  emitterUpdateScript [
   &EmitterState(Inactive Response=::NewEnumerator1, Life Cycle Mode=::NewEnumerator1, Loop Behavior=::NewEnumerator1, Loop Duration=4, UseLoopDelay=true)
   /SetVariables(Emitter.FadeOutZShape:NiagaraFloat=&FloatFromCurve(
     FloatCurve=DataInterface() {propertyValues {Curve {keys [
         {interpMode=RCIM_Cubic, tangentMode=RCTM_User, time=0, value=1}
         {arriveTangent=-1.0679, interpMode=RCIM_Cubic, leaveTangent=-1.0679, tangentMode=RCTM_User, time=0.75, value=0.8}
         {arriveTangent=-0.0857, interpMode=RCIM_Cubic, leaveTangent=-0.0857, tangentMode=RCTM_User, time=1, value=0}
        ]}}}
     CurveIndex=$System.NormalizedLoopAge
    )) {isSetParameters=true}
   /Game/Effects/Niagara/Modules/BlendSpace/BlendSpaceHelper()
   &SpawnRate(SpawnGroup=2, SpawnRate=500, Use Spawn Probability=false) {enabled=false, module=SpawnRate003}
   &SpawnRate(SpawnRate=&Multiply_Float(A=12, B=$Emitter.FadeOutZShape), Use Spawn Probability=false)
   &SpawnRate(SpawnGroup=1, SpawnRate=&Multiply_Float(A=12, B=$Emitter.FadeOutZShape), Use Spawn Probability=false) {module=SpawnRate001}
   &SpawnRate(SpawnGroup=2, SpawnRate=&Multiply_Float(A=12, B=$Emitter.FadeOutZShape), Use Spawn Probability=false) {module=SpawnRate002}
  ]
  particleSpawnScript [
   /SetVariables(Particles.VisibilityTag=&MakeCustomIntFromBool(
     Bool=&SetBoolByIntComparison(A=&SpawnGroup(), B=2, Comparison Type=::NewEnumerator7)
     True INT=&RandomRangeInt(Minimum=0, Maximum=5)
     False INT=6
    )) {isSetParameters=true}
   /SetVariables(Particles.RandomSocket:NiagaraInt32=&SpawnGroup()) {isSetParameters=true}
   &InitializeParticle(
    Color=&Multiply_LinearColorByFloat(Linear Color=$System.ElimColor, Float=0.4)
    Lifetime=&MaskFloatBySpawnGroup(Masked Float=3, Passthrough Float=&RandomRangeFloat(Minimum=0.175, Maximum=0.35), Spawn Group Mask=6)
    Mesh Scale Max=<2.1 2.1 3.2>
    Mesh Scale Min=<1 1 3>
    Mesh Scale Mode=::NewEnumerator2
    Sprite Rotation Mode=::NewEnumerator2
    Sprite Size Max=<30 30>
    Sprite Size Min=<20 20>
    Sprite Size Mode=::NewEnumerator2
    UsePositionOffset=false
    Write Material Random=true
   )
   &SkeletalMeshLocation(
    Bone / Socket Index=$Particles.RandomSocket
    MaskSpawnGroup=false
    Mesh Sampling Type=::NewEnumerator5
    Skeletal Mesh=$System.Skeletal Mesh
    Socket Sampling Mode=::NewEnumerator5
   )
   &ShapeLocation(
    Cylinder Height=5
    Cylinder Height Midpoint=1
    Cylinder Radius=2.5
    MaskSpawnGroup=false
    Offset=&MergeHelper_VectorSwitch(0=<-20 0 25>, 1=<-15 0 18>, 2=<-18 0 -2>, 3=?, Index=$Particles.RandomSocket, Num=::NewEnumerator1)
    Offset Mode=::NewEnumerator0
    Override Seed=false
    Rotation Mode=::NewEnumerator2
    Shape Origin=$Particles.SkeletalMeshLocation.SampledPosition
    Shape Primitive=::NewEnumerator1
    Surface Only=false
    UseOwnerScale=true
    Yaw / Pitch / Roll=&MergeHelper_VectorSwitch(0=<0 0.25 0>, 1=<0 0.35 0>, 2=<0 0.25 0>, 3=?, Index=$Particles.RandomSocket, Num=::NewEnumerator1)
   )
   &InitialMeshOrientation(
    Mesh Axis To Orient=<0 0 1>
    Mesh Orientation Mode=::NewEnumerator2
    Rotation=&MergeHelper_VectorSwitch(
     0=&RandomRangeVector(Minimum=<0 0.03 0>, Maximum=<0 0.075 0>)
     1=<0 0.15 0>
     2=&RandomRangeVector(Minimum=<0 0.03 0>, Maximum=<0 0.075 0>)
     3=?
     Index=&SpawnGroup()
     Num=::NewEnumerator1
    )
    Use Rotation Vector=true
   )
   &AddVelocity(
    Velocity=&RandomRangeVector(Minimum=<-1 -1 -1>, Maximum=<1 1 1>)
    Velocity Speed Scale=&RandomRangeFloat(
     Minimum=&MaskFloatBySpawnGroup(Masked Float=0, Passthrough Float=15, Spawn Group Mask=6)
     Maximum=&MaskFloatBySpawnGroup(Masked Float=0, Passthrough Float=5, Spawn Group Mask=6)
    )
   )
   &CameraOffset(Camera Offset Amount=15)
  ]
  particleUpdateScript [
   &SolveForcesAndVelocity(Clamp Velocity=false, Limit Acceleration=false)
   &ParticleState()
   &ScaleMeshSize(Scale Factor=&Multiply_VectorByFloat(Vector=<1 1 0.5>, Float=0.2))
   &DynamicMaterialParameters(
    Index 0 Param 1=&MakeFloatFromVector(Vector=$Particles.Scale, Channel=::NewEnumerator0)
    Index 0 Param 2=&RandomRangeFloat(Minimum=0, Maximum=10)
    Index 0 Param 3=&FloatFromCurve(
     FloatCurve=DataInterface() {propertyValues {Curve {keys [{interpMode=RCIM_Cubic, leaveTangent=8.1085, tangentMode=RCTM_Break, time=0.8568, value=1}, {tangentMode=RCTM_User, time=1, value=0}]}}}
     Scale Curve=0.07
    )
    Index 0 Param 4=0.8
    Index 1 Param 1=&FloatFromCurve(FloatCurve=DataInterface() {propertyValues {Curve {keys [
         {interpMode=RCIM_Cubic, leaveTangent=1.1034, tangentMode=RCTM_Break, time=0, value=0}
         {arriveTangent=2.2399, tangentMode=RCTM_User, time=0.5, value=1}
        ]}}})
    Index 1 Param 2=&FloatFromCurve(FloatCurve=DataInterface() {propertyValues {Curve {keys [<at 0.5 0>, {arriveTangent=3.2053, tangentMode=RCTM_User, time=0.9, value=1}]}}})
    Index 1 Param 3=&FloatFromCurve(FloatCurve=DataInterface() {propertyValues {Curve {keys [<at 0 0 RCIM_Cubic>, {arriveTangent=3.4223, interpMode=RCIM_Cubic, leaveTangent=3.4223, tangentMode=RCTM_User, time=1, value=1}]}}})
    Index 1 Param 4=&FloatFromCurve(
     FloatCurve=DataInterface() {propertyValues {Curve {keys [
         {interpMode=RCIM_Cubic, leaveTangent=0.1169, tangentMode=RCTM_Break, time=0, value=0}
         {arriveTangent=2.5708, interpMode=RCIM_Cubic, leaveTangent=2.5708, tangentMode=RCTM_User, time=1, value=1}
        ]}}}
     Scale Curve=&RandomRangeFloat(Minimum=1, Maximum=40)
    )
    Index 2 Param 1=&FloatFromCurve(
     FloatCurve=DataInterface() {propertyValues {Curve {keys [
         {interpMode=RCIM_Cubic, leaveTangent=-0.0025, tangentMode=RCTM_Break, time=0, value=0.5}
         {arriveTangent=-0.0316, interpMode=RCIM_Cubic, tangentMode=RCTM_Break, time=1, value=-0.2}
        ]}}}
     Scale Curve=1.8187
    )
    Index 2 Param 2=&FloatFromCurve(
     FloatCurve=DataInterface() {propertyValues {Curve {keys [
         {interpMode=RCIM_Cubic, tangentMode=RCTM_User, time=0, value=0}
         {arriveTangent=2.0522, interpMode=RCIM_Cubic, leaveTangent=2.0522, tangentMode=RCTM_User, time=1, value=1}
        ]}}}
     Scale Curve=2
    )
    Module.Index0Param0Write=true
    Module.Index0Param1Write=true
    Module.Index0Param2Write=true
    Module.Index0Param3Write=true
    Module.Index1Param0Write=true
    Module.Index1Param1Write=true
    Module.Index1Param2Write=true
    Module.Index1Param3Write=true
    Module.Index2Param0Write=true
    Module.Index2Param1Write=true
    Module.Index2Param2Write=false
    Module.Index2Param3Write=false
    Module.Index3Param0Write=false
    Module.Index3Param1Write=false
    Module.Index3Param2Write=false
    Module.Index3Param3Write=true
    Write Parameter Index 1=true
    Write Parameter Index 2=true
    Write Parameter Index 3=true
   )
   &UpdateMeshOrientation(Rotation Rate=&RandomRangeFloat(Minimum=-1, Maximum=1), Rotation Vector=<0 0 1>)
   &ScaleColor(
    Scale Alpha=&FloatFromCurve(FloatCurve=DataInterface() {propertyValues {Curve {keys [
         {tangentMode=RCTM_User, time=0, value=1}
         {arriveTangent=-0.0405, interpMode=RCIM_Cubic, leaveTangent=-0.0405, tangentMode=RCTM_User, time=0.7009, value=1}
         {arriveTangent=-8.0089, interpMode=RCIM_Cubic, leaveTangent=-8.0089, tangentMode=RCTM_User, time=1, value=0}
        ]}}})
    Scale RGB=&Lerp_Vector(
     A=&VectorFromFloat(Value=&FloatFromCurve(
       FloatCurve=DataInterface() {propertyValues {Curve {keys [<at 0 1>,<at 0.1 1>,<at 0.2 2>,<at 0.3 1>,<at 0.6 1>,<at 0.7 2>,<at 0.8 1>,<at 1 1>]}}}
       Scale Curve=10
      ))
     B=&VectorFromFloat(Value=&RandomRangeFloat(Minimum=0.3, Maximum=5, Evaluation Type=::NewEnumerator1))
     Alpha=&FloatFromCurve(FloatCurve=DataInterface() {propertyValues {Curve {keys [
          {interpMode=RCIM_Cubic, leaveTangent=2.5294, tangentMode=RCTM_Break, time=0, value=0}
          {arriveTangent=-0.2166, interpMode=RCIM_Cubic, tangentMode=RCTM_Break, time=0.4, value=1}
         ]}}})
    )
    ScaleA=true
    ScaleRGB=true
   )
   &FN_Light_Attributes(
    Color=<rgba 1 0.0087 0 0.7>
    Create Color Attribute=true
    Create Exponent Attribute=true
    Create Radius Attribute=true
    Exponent=1
    Radius=80
    Scale Color=&FloatFromCurve(
     FloatCurve=DataInterface() {propertyValues {Curve {keys [<at 0 1>,<at 1 0>]}}}
     CurveIndex=&Frac_Float(Float=&Multiply_Float(A=$Particles.NormalizedAge, B=50))
    )
   )
   &BlendSpace(Position Coordinate Space=2)
  ]
  rendererClasses [&NiagaraMeshRendererProperties, &NiagaraSpriteRendererProperties]
  renderers [
   {rendererClass=&NiagaraMeshRendererProperties, rendererIndex=0}
   {rendererClass=&NiagaraMeshRendererProperties, rendererIndex=1}
   {rendererClass=&NiagaraMeshRendererProperties, rendererIndex=2}
   {rendererClass=&NiagaraMeshRendererProperties, rendererIndex=3}
   {rendererClass=&NiagaraSpriteRendererProperties, rendererIndex=4}
  ]
 }}
summary {
 systemName=NS_Emote_GlowFang_Staff_ChargedZshape
 emitters [{emitterName=StaffLightning, rendererClasses [&NiagaraMeshRendererProperties, &NiagaraSpriteRendererProperties]}]
 userVariables [
  User.bIsFrontEndPreview {type=NiagaraBool}
  User.bKillStage {type=NiagaraInt32}
  User.LoopsDelay {type=NiagaraFloat, defaultValue=0, description=""}
 ]
}
