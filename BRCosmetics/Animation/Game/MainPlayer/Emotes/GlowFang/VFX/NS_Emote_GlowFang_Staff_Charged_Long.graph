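# Niagara system for the long charged-staff effect in the GlowFang emote, tinted throughout by the system ElimColor.
# StaffChargeupFirstPhase uses masked spawn groups and visibility tags to drive timed mesh and sprite layers from one emitter, some placed at a sampled skeletal mesh socket.
# StaffLightning steps short-lived mesh arcs up the staff through an offset array over four loops.
# StaffLightning001 follows later with fading arcs at three fixed upper offsets.
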NiagaraSystem /BRCosmetics/Animation/Game/MainPlayer/Emotes/GlowFang/VFX/NS_Emote_GlowFang_Staff_Charged_Long
dependencies {
 usedDataInterfaces [&NiagaraDataInterfaceCurve, &NiagaraDataInterfaceArrayFloat3, &NiagaraDataInterfaceSkeletalMesh]
 usedDynamicInputs [
  &MergeHelper_FloatSwitch
  &FNTimeOfDayScalar
  &InvertBool
 ]
 usedModules [
  &self:SystemSpawnScript.SystemScriptSource.SystemScriptGraph.NiagaraNodeAssignment_4.SetVariables_#1
  &self:SystemSpawnScript.SystemScriptSource.SystemScriptGraph.NiagaraNodeAssignment_2.SetVariables_#2
  &SystemState
  &self:SystemSpawnScript.SystemScriptSource.SystemScriptGraph.NiagaraNodeAssignment_0.SetVariables_#3
  &self:SystemSpawnScript.SystemScriptSource.SystemScriptGraph.NiagaraNodeAssignment_1.SetVariables_#4
  &self:SystemSpawnScript.SystemScriptSource.SystemScriptGraph.NiagaraNodeAssignment_6.SetVariables_#5
  &self:SystemSpawnScript.SystemScriptSource.SystemScriptGraph.NiagaraNodeAssignment_3.SetVariables_#6
 ]
}
emitters {
 StaffChargeupFirstPhase {
  emitterUpdateScript [
   &EmitterState(
    Inactive Response=::NewEnumerator1
    Life Cycle Mode=::NewEnumerator1
    Loop Behavior=::NewEnumerator1
    Loop Delay=&MaskFloat(
     Mask Parameter=$Emitter.SpawnGroup
     Mask=0
     Masked Float=0
     Passthrough Float=&MaskFloat(
      Mask Parameter=$Emitter.SpawnGroup
      Mask=1
      Masked Float=0.2
      Passthrough Float=&MaskFloat(
       Mask Parameter=$Emitter.SpawnGroup
       Mask=2
       Masked Float=0.7
       Passthrough Float=&MaskFloat(
        Mask Parameter=$Emitter.SpawnGroup
        Mask=3
        Masked Float=1
        Passthrough Float=&MaskFloat(
         Mask Parameter=$Emitter.SpawnGroup
         Mask=4
         Masked Float=2.2
         Passthrough Float=&MaskFloat(Mask Parameter=$Emitter.SpawnGroup, Mask=5, Masked Float=0.9, Passthrough Float=0.9)
        )
       )
      )
     )
    )
    Loop Duration=&MaskFloat(
     Mask Parameter=$Emitter.SpawnGroup
     Mask=0
     Masked Float=2.75
     Passthrough Float=&MaskFloat(
      Mask Parameter=$Emitter.SpawnGroup
      Mask=1
      Masked Float=2.75
      Passthrough Float=&MaskFloat(
       Mask Parameter=$Emitter.SpawnGroup
       Mask=2
       Masked Float=2.75
       Passthrough Float=&MaskFloat(
        Mask Parameter=$Emitter.SpawnGroup
        Mask=3
        Masked Float=2.75
        Passthrough Float=&MaskFloat(
         Mask Parameter=$Emitter.SpawnGroup
         Mask=4
         Masked Float=1
         Passthrough Float=&MaskFloat(Mask Parameter=$Emitter.SpawnGroup, Mask=5, Masked Float=2.75, Passthrough Float=1)
        )
       )
      )
     )
    )
    UseLoopDelay=true
   )
   &SpawnRate(SpawnGroup=5, SpawnRate=&Multiply_Float(A=5, B=$System.DelayedFadeOutFirstPhase), Use Spawn Probability=false) {module=SpawnRate001}
   &SpawnBurst_Instantaneous(
    Spawn Count=&RandomRangeInt(Minimum=9, Maximum=11)
    Spawn Group=4
    Spawn Time=1.2
    Use Spawn Probability=false
    UseLoopCountLimit=false
   ) {module=SpawnBurst_Instantaneous003}
   &SpawnBurst_Instantaneous(Spawn Group=3, Spawn Time=1, Use Spawn Probability=false, UseLoopCountLimit=false) {module=SpawnBurst_Instantaneous002}
   &SpawnRate(SpawnRate=&Multiply_Float(A=25, B=$System.FadeOutFirstPhase), Use Spawn Probability=false)
   &SpawnBurst_Instantaneous(Spawn Group=1, Spawn Time=0.2, Use Spawn Probability=false, UseLoopCountLimit=false)
   &SpawnBurst_Instantaneous(Spawn Count=1, Spawn Group=2, Spawn Time=0.7, Use Spawn Probability=false) {module=SpawnBurst_Instantaneous001}
   &SampleSkeletalMesh(
    AdditionalYawPitchRoll=false
    Mesh Sampling Type:ENiagaraSkelSamplingModeFull=::NewEnumerator5
    Sampled Position Offset:Vector3f=<0 0 20>
    Skeletal Mesh:NiagaraDataInterfaceSkeletalMesh=$System.Skeletal Mesh
    Socket Sampling Mode:ENiagaraSocketSamplingMode=::NewEnumerator5
   )
   /SetVariables(Emitter.TempsLocation:NiagaraPosition=$Emitter.SampleSkeletalMesh.SampledPosition) {isSetParameters=true}
  ]
  particleSpawnScript [
   /SetVariables(Particles.VisibilityTag=&SpawnGroup()) {isSetParameters=true}
   &InitializeParticle(
    Color=&MaskLinearColorBySpawnGroup(
     Masked Linear Color=&Multiply_LinearColorByFloat(Linear Color=$System.ElimColor, Float=0.4)
     Passthrough Linear Color=&MaskLinearColorBySpawnGroup(
      Masked Linear Color=&Multiply_LinearColorByFloat(Linear Color=$System.ElimColor, Float=0.5)
      Passthrough Linear Color=&MaskLinearColorBySpawnGroup(
       Masked Linear Color=$System.ElimColor
       Passthrough Linear Color=&MaskLinearColorBySpawnGroup(
        Masked Linear Color=$System.ElimColor
        Passthrough Linear Color=&MaskLinearColorBySpawnGroup(
         Masked Linear Color=&Multiply_LinearColorByFloat(Linear Color=$System.ElimColor, Float=&RandomRangeFloat(Minimum=0.8, Maximum=25))
         Passthrough Linear Color=&Multiply_LinearColorByFloat(Linear Color=$System.ElimColor, Float=0.4)
         Spawn Group Mask=4
        )
        Spawn Group Mask=3
       )
       Spawn Group Mask=2
      )
      Spawn Group Mask=1
     )
    )
    Lifetime=&MaskFloatBySpawnGroup(
     Masked Float=&RandomRangeFloat(Minimum=0.175, Maximum=0.35)
     Passthrough Float=&MaskFloatBySpawnGroup(
      Masked Float=1.25
      Passthrough Float=&MaskFloatBySpawnGroup(
       Masked Float=2.75
       Passthrough Float=&MaskFloatBySpawnGroup(
        Masked Float=1
        Passthrough Float=&MaskFloatBySpawnGroup(
         Masked Float=&RandomRangeFloat(Minimum=0.1, Maximum=1)
         Passthrough Float=&RandomRangeFloat(Minimum=0.175, Maximum=0.3)
         Spawn Group Mask=4
        )
        Spawn Group Mask=3
       )
       Spawn Group Mask=2
      )
      Spawn Group Mask=1
     )
    )
    Mesh Scale=&MaskVectorBySpawnGroup(
     Masked Vector=&RandomRangeVector(Minimum=<0.75 0.75 0.75>, Maximum=<2 2 2>)
     Passthrough Vector=&MaskVectorBySpawnGroup(
      Masked Vector=<0 0 0>
      Passthrough Vector=&MaskVectorBySpawnGroup(
       Masked Vector=<1.05 1.05 1.001>
       Passthrough Vector=&MaskVectorBySpawnGroup(
        Masked Vector=<0 0 0>
        Passthrough Vector=&MaskVectorBySpawnGroup(
         Masked Vector=<0 0 0>
         Passthrough Vector=&RandomRangeVector(Minimum=<0.3 0.3 0.5>, Maximum=<0.5 0.5 1>)
         Spawn Group Mask=4
        )
        Spawn Group Mask=3
       )
       Spawn Group Mask=2
      )
      Spawn Group Mask=1
     )
    )
    Mesh Scale Mode=::NewEnumerator4
    Position Offset=&MaskVectorBySpawnGroup(
     Masked Vector=<0 0 0>
     Passthrough Vector=&MaskVectorBySpawnGroup(
      Masked Vector=<0 0 25>
      Passthrough Vector=&MaskVectorBySpawnGroup(
       Masked Vector=<0 0 0>
       Passthrough Vector=&MaskVectorBySpawnGroup(Masked Vector=<0 0 0.3>, Passthrough Vector=<0 0 0>, Spawn Group Mask=3)
       Spawn Group Mask=2
      )
      Spawn Group Mask=1
     )
    )
    Sprite Size Mode=::NewEnumerator3
    Uniform Sprite Size=&MaskFloatBySpawnGroup(
     Masked Float=0
     Passthrough Float=&MaskFloatBySpawnGroup(
      Masked Float=50
      Passthrough Float=&MaskFloatBySpawnGroup(
       Masked Float=0
       Passthrough Float=&MaskFloatBySpawnGroup(Masked Float=100, Passthrough Float=&RandomRangeFloat(Minimum=1, Maximum=5), Spawn Group Mask=3)
       Spawn Group Mask=2
      )
      Spawn Group Mask=1
     )
    )
    UsePositionOffset=true
    Write Material Random=true
   )
   &InitialMeshOrientation(
    Mesh Orientation Mode=::NewEnumerator2
    Rotation=&MaskVectorBySpawnGroup(
     Masked Vector=&RandomRangeVector(Minimum=<0 0 -0.5>, Maximum=<0 0 0.5>)
     Passthrough Vector=&MaskVectorBySpawnGroup(
      Masked Vector=<0 0 0>
      Passthrough Vector=&MaskVectorBySpawnGroup(
       Masked Vector=<0 0 -0.25>
       Passthrough Vector=&MaskVectorBySpawnGroup(
        Masked Vector=<0 0 0>
        Passthrough Vector=&MaskVectorBySpawnGroup(
         Masked Vector=<0 0 0>
         Passthrough Vector=&RandomRangeVector(Minimum=<0 0 -1>, Maximum=<0.2 0.2 1>)
         Spawn Group Mask=4
        )
        Spawn Group Mask=3
       )
       Spawn Group Mask=2
      )
      Spawn Group Mask=1
     )
    )
    Use Rotation Vector=true
   )
   &ShapeLocation(
    Box Midpoint=&MaskVectorBySpawnGroup(
     Masked Vector=<0.5 0.5 1>
     Passthrough Vector=&MaskVectorBySpawnGroup(
      Masked Vector=<0.5 0.5 0.5>
      Passthrough Vector=&MaskVectorBySpawnGroup(
       Masked Vector=<0.5 0.5 0.5>
       Passthrough Vector=&MaskVectorBySpawnGroup(
        Masked Vector=<0.5 0.5 0.5>
        Passthrough Vector=&MaskVectorBySpawnGroup(Masked Vector=<0.5 0.5 0.5>, Passthrough Vector=<0.5 0.5 1>, Spawn Group Mask=4)
        Spawn Group Mask=3
       )
       Spawn Group Mask=2
      )
      Spawn Group Mask=1
     )
    )
    Box Size=&MaskVectorBySpawnGroup(
     Masked Vector=<1 1 130>
     Passthrough Vector=&MaskVectorBySpawnGroup(
      Masked Vector=<0 0 0>
      Passthrough Vector=&MaskVectorBySpawnGroup(
       Masked Vector=<0 0 0>
       Passthrough Vector=&MaskVectorBySpawnGroup(
        Masked Vector=<1 1 35>
        Passthrough Vector=&MaskVectorBySpawnGroup(Masked Vector=<0 0 0>, Passthrough Vector=<10 10 130>, Spawn Group Mask=4)
        Spawn Group Mask=3
       )
       Spawn Group Mask=2
      )
      Spawn Group Mask=1
     )
    )
    MaskSpawnGroup=false
    Offset=&MaskVectorBySpawnGroup(
     Masked Vector=<0 0 80>
     Passthrough Vector=&MaskVectorBySpawnGroup(
      Masked Vector=<0 0 110>
      Passthrough Vector=&MaskVectorBySpawnGroup(Masked Vector=<0 0 110>, Passthrough Vector=<0 0 110>, Spawn Group Mask=2)
      Spawn Group Mask=1
     )
    )
    Offset Coordinate Space=::Local
    Offset Mode=::NewEnumerator0
    Override Seed=false
    Shape Primitive=::NewEnumerator2
    Surface Only=false
    UseOwnerScale=true
   )
   &SystemLocation(
    MaskSpawnGroup=false
    Offset=&MaskVectorBySpawnGroup(
     Masked Vector=$Particles.ShapeLocation.ShapePosition
     Passthrough Vector=&MaskVectorBySpawnGroup(
      Masked Vector=<0 0 135>
      Passthrough Vector=&MaskVectorBySpawnGroup(
       Masked Vector=<-5.3 0 55>
       Passthrough Vector=&MaskVectorBySpawnGroup(
        Masked Vector=$Emitter.TempsLocation
        Passthrough Vector=&MaskVectorBySpawnGroup(Masked Vector=$Emitter.TempsLocation, Passthrough Vector=$Particles.ShapeLocation.ShapePosition, Spawn Group Mask=4)
        Spawn Group Mask=3
       )
       Spawn Group Mask=2
      )
      Spawn Group Mask=1
     )
    )
   )
   &CameraOffset(Camera Offset Amount=&MaskFloatBySpawnGroup(
     Masked Float=0
     Passthrough Float=&MaskFloatBySpawnGroup(Masked Float=15, Passthrough Float=0, Spawn Group Mask=1)
    ))
   &AddVelocity(
    Velocity=&MaskVectorBySpawnGroup(
     Masked Vector=<0 0 0>
     Passthrough Vector=&MaskVectorBySpawnGroup(
      Masked Vector=<0 0 0>
      Passthrough Vector=&MaskVectorBySpawnGroup(
       Masked Vector=<0 0 0>
       Passthrough Vector=&MaskVectorBySpawnGroup(
        Masked Vector=<-1 0 1>
        Passthrough Vector=&MaskVectorBySpawnGroup(Masked Vector=<0 0 0>, Passthrough Vector=<0 0 0>, Spawn Group Mask=4)
        Spawn Group Mask=3
       )
       Spawn Group Mask=2
      )
      Spawn Group Mask=1
     )
    )
    Velocity Speed Scale=15
   )
   &AddVelocity(
    Cone Angle=&MaskFloatBySpawnGroup(Masked Float=20, Passthrough Float=95, Spawn Group Mask=5)
    Cone Axis=<0 0 1>
    Distribution Along Cone Axis=0.2
    Module.UseConeFalloff=false
    Module.UseVelDistribution=true
    Override Seed=false
    Velocity Mode=::NewEnumerator2
    Velocity Speed=&MaskFloatBySpawnGroup(
     Masked Float=0
     Passthrough Float=&MaskFloatBySpawnGroup(
      Masked Float=0
      Passthrough Float=&MaskFloatBySpawnGroup(
       Masked Float=0
       Passthrough Float=&MaskFloatBySpawnGroup(
        Masked Float=0
        Passthrough Float=&MaskFloatBySpawnGroup(Masked Float=&RandomRangeFloat(Minimum=150, Maximum=300), Passthrough Float=0, Spawn Group Mask=4)
        Spawn Group Mask=3
       )
       Spawn Group Mask=2
      )
      Spawn Group Mask=1
     )
    )
   ) {module=AddVelocity001}
  ]
  particleUpdateScript [
   /SetVariables(Particles.Position=&MaskVectorBySpawnGroup(
     Masked Vector=$Emitter.TempsLocation
     Passthrough Vector=Linked() {linkedVariable {name=Particles.Position, type=NiagaraPosition}}
     Spawn Group Mask=1
    )) {isSetParameters=true}
   &ParticleState()
   &ScaleSpriteSize(
    Scale Sprite Size Mode=::NewEnumerator0
    Uniform Scale Factor=&MaskFloatBySpawnGroup(
     Masked Float=0
     Passthrough Float=&MaskFloatBySpawnGroup(
      Masked Float=&FloatFromCurve(
       FloatCurve=DataInterface() {propertyValues {Curve {keys [
           <at 0 0.3 RCIM_Cubic>
           {arriveTangent=5.3971, interpMode=RCIM_Cubic, leaveTangent=5.3971, tangentMode=RCTM_User, time=0.6462, value=1.5162}
           {arriveTangent=13.7263, interpMode=RCIM_Cubic, leaveTangent=13.7263, time=0.8013, value=1.0461}
           {arriveTangent=-57.6853, interpMode=RCIM_Cubic, leaveTangent=-57.6853, tangentMode=RCTM_User, time=0.9, value=5}
           <at 1 0 RCIM_Cubic>
          ]}}}
       Scale Curve=0.85
      )
      Passthrough Float=&MaskFloatBySpawnGroup(
       Masked Float=0
       Passthrough Float=&FloatFromCurve(FloatCurve=DataInterface() {propertyValues {Curve {keys [
            {arriveTangent=3.4683, interpMode=RCIM_Cubic, leaveTangent=3.4683, tangentMode=RCTM_User, time=0, value=0}
            {interpMode=RCIM_Cubic, tangentMode=RCTM_User, time=1, value=1.15}
           ]}}})
       Spawn Group Mask=2
      )
      Spawn Group Mask=1
     )
    )
   )
   &ScaleMeshSize(Scale Factor=&MaskVectorBySpawnGroup(
     Masked Vector=&Multiply_VectorByFloat(Vector=<1 1 1>, Float=0.2)
     Passthrough Vector=&MaskVectorBySpawnGroup(
      Masked Vector=<0 0 0>
      Passthrough Vector=&MaskVectorBySpawnGroup(
       Masked Vector=<1 1 1>
       Passthrough Vector=&MaskVectorBySpawnGroup(
        Masked Vector=<1 1 1>
        Passthrough Vector=&MaskVectorBySpawnGroup(
         Masked Vector=<1 1 1>
         Passthrough Vector=&Multiply_VectorByFloat(
          Vector=&MakeCustomVectorFromBool(
           Bool=&SetBoolByFloatComparison(
            A=&VectorLength(Vector=Linked() {linkedVariable {name=Particles.Position, type=NiagaraPosition}})
            B=&VectorLength(Vector=<0 0 50>)
            Comparison Type=::NewEnumerator7
           )
           True Vector=<1 1 1>
           False Vector=<0.7 0.7 0.7>
          )
          Float=&Multiply_Float(
           A=0.2
           B=&FloatFromCurve(
            FloatCurve=DataInterface() {propertyValues {Curve {keys [
                <at 0 1 RCIM_Cubic>
                {arriveTangent=-0.1283, interpMode=RCIM_Cubic, leaveTangent=0.4938, tangentMode=RCTM_Break, time=0.9, value=1}
                {arriveTangent=7.2371, interpMode=RCIM_Cubic, leaveTangent=7.2371, tangentMode=RCTM_User, time=1, value=2}
               ]}}}
            CurveIndex=&Divide_Float(A=$Engine.System.Age, B=3.5)
           )
          )
         )
         Spawn Group Mask=4
        )
        Spawn Group Mask=3
       )
       Spawn Group Mask=2
      )
      Spawn Group Mask=1
     )
    ))
   &DynamicMaterialParameters(
    Index 0 Param 1=&MakeFloatFromVector(Vector=$Particles.Scale, Channel=::NewEnumerator0)
    Index 0 Param 2=&RandomRangeFloat(Minimum=0, Maximum=10)
    Index 0 Param 3=&FloatFromCurve(
     FloatCurve=DataInterface() {propertyValues {Curve {keys [{interpMode=RCIM_Cubic, leaveTangent=8.1085, tangentMode=RCTM_Break, time=0.8568, value=1}, {tangentMode=RCTM_User, time=1, value=0}]}}}
     Scale Curve=0.07
    )
    Index 0 Param 4=0.8
    Index 1 Param 1=&FloatFromCurve(FloatCurve=DataInterface() {propertyValues {Curve {keys [
         {interpMode=RCIM_Cubic, leaveTangent=1.1034, tangentMode=RCTM_Break, time=0, value=0}
         {arriveTangent=2.2399, tangentMode=RCTM_User, time=0.5, value=1}
        ]}}})
    Index 1 Param 2=&FloatFromCurve(FloatCurve=DataInterface() {propertyValues {Curve {keys [<at 0.5 0>, {arriveTangent=3.2053, tangentMode=RCTM_User, time=0.9, value=1}]}}})
    Index 1 Param 3=&FloatFromCurve(FloatCurve=DataInterface() {propertyValues {Curve {keys [<at 0 0 RCIM_Cubic>, {arriveTangent=3.4223, interpMode=RCIM_Cubic, leaveTangent=3.4223, tangentMode=RCTM_User, time=1, value=1}]}}})
    Index 1 Param 4=&FloatFromCurve(
     FloatCurve=DataInterface() {propertyValues {Curve {keys [
         {interpMode=RCIM_Cubic, leaveTangent=0.1169, tangentMode=RCTM_Break, time=0, value=0}
         {arriveTangent=2.5708, interpMode=RCIM_Cubic, leaveTangent=2.5708, tangentMode=RCTM_User, time=1, value=1}
        ]}}}
     Scale Curve=&RandomRangeFloat(Minimum=1, Maximum=40)
    )
    Index 2 Param 1=&FloatFromCurve(FloatCurve=DataInterface() {propertyValues {Curve {keys [
         {interpMode=RCIM_Cubic, leaveTangent=-0.0025, tangentMode=RCTM_Break, time=0, value=0.5}
         {arriveTangent=-0.0316, interpMode=RCIM_Cubic, tangentMode=RCTM_Break, time=1, value=-0.2}
        ]}}})
    Index 2 Param 2=&FloatFromCurve(
     FloatCurve=DataInterface() {propertyValues {Curve {keys [
         {interpMode=RCIM_Cubic, tangentMode=RCTM_User, time=0, value=0}
         {arriveTangent=2.0522, interpMode=RCIM_Cubic, leaveTangent=2.0522, tangentMode=RCTM_User, time=1, value=1}
        ]}}}
     Scale Curve=2
    )
    Module.Index0Param0Write=true
    Module.Index0Param1Write=true
    Module.Index0Param2Write=true
    Module.Index0Param3Write=true
    Module.Index1Param0Write=true
    Module.Index1Param1Write=true
    Module.Index1Param2Write=true
    Module.Index1Param3Write=true
    Module.Index2Param0Write=true
    Module.Index2Param1Write=true
    Module.Index2Param2Write=false
    Module.Index2Param3Write=false
    Module.Index3Param0Write=false
    Module.Index3Param1Write=false
    Module.Index3Param2Write=false
    Module.Index3Param3Write=true
    Write Parameter Index 1=true
    Write Parameter Index 2=true
    Write Parameter Index 3=true
   )
   &UpdateMeshOrientation(
    Rotation Rate=&MaskFloatBySpawnGroup(
     Masked Float=&RandomRangeFloat(Minimum=-0.5, Maximum=0.5)
     Passthrough Float=&MaskFloatBySpawnGroup(Masked Float=0, Passthrough Float=0, Spawn Group Mask=1)
    )
    Rotation Vector=<0 0 1>
   )
   &ScaleColor(
    Scale Alpha=&MaskFloatBySpawnGroup(
     Masked Float=&FloatFromCurve(FloatCurve=DataInterface() {propertyValues {Curve {keys [
          {tangentMode=RCTM_User, time=0, value=1}
          {arriveTangent=-0.0405, interpMode=RCIM_Cubic, leaveTangent=-0.0405, tangentMode=RCTM_User, time=0.7009, value=1}
          {arriveTangent=-8.0089, interpMode=RCIM_Cubic, leaveTangent=-8.0089, tangentMode=RCTM_User, time=1, value=0}
         ]}}})
     Passthrough Float=&MaskFloatBySpawnGroup(
      Masked Float=&FloatFromCurve(FloatCurve=DataInterface() {propertyValues {Curve {keys [
           {interpMode=RCIM_Cubic, leaveTangent=0.2193, tangentMode=RCTM_Break, time=0, value=0}
           {arriveTangent=2.5, interpMode=RCIM_Cubic, leaveTangent=2.5, time=0.1691, value=0.1683}
           {arriveTangent=-0.2166, interpMode=RCIM_Cubic, tangentMode=RCTM_Break, time=0.4, value=1}
          ]}}})
      Passthrough Float=&MaskFloatBySpawnGroup(
       Masked Float=&FloatFromCurve(
        FloatCurve=DataInterface() {propertyValues {Curve {keys [
            {arriveTangent=-0.141, interpMode=RCIM_Cubic, leaveTangent=-0.141, tangentMode=RCTM_User, time=0, value=0}
            {arriveTangent=3.8262, interpMode=RCIM_Cubic, leaveTangent=21.5493, tangentMode=RCTM_Break, time=0.5, value=1}
            {arriveTangent=351.5977, interpMode=RCIM_Cubic, leaveTangent=-207.9236, tangentMode=RCTM_Break, time=0.5234, value=11.2794}
            {arriveTangent=-5.5932, tangentMode=RCTM_User, time=0.7, value=0}
           ]}}}
        Scale Curve=2
       )
       Passthrough Float=&MaskFloatBySpawnGroup(
        Masked Float=&FloatFromCurve(
         FloatCurve=DataInterface() {propertyValues {Curve {keys [
             {interpMode=RCIM_Cubic, tangentMode=RCTM_User, time=0, value=0}
             <at 0.7 1 RCIM_Cubic>
             {interpMode=RCIM_Cubic, tangentMode=RCTM_User, time=1, value=0}
            ]}}}
         Scale Curve=0.32
        )
        Passthrough Float=&MaskFloatBySpawnGroup(
         Masked Float=&FloatFromCurve(FloatCurve=DataInterface() {propertyValues {Curve {keys [
              {arriveTangent=9.0725, interpMode=RCIM_Cubic, leaveTangent=9.0725, tangentMode=RCTM_User, time=0, value=0}
              {arriveTangent=-2.5249, interpMode=RCIM_Cubic, leaveTangent=-2.5249, tangentMode=RCTM_User, time=0.1408, value=0.9847}
              {arriveTangent=-0.5654, interpMode=RCIM_Cubic, leaveTangent=-0.5654, time=0.5, value=0.35}
              {arriveTangent=-1.25, interpMode=RCIM_Cubic, leaveTangent=-1.25, time=0.6, value=0.725}
              {arriveTangent=-3.375, interpMode=RCIM_Cubic, leaveTangent=-0.1412, tangentMode=RCTM_Break, time=0.7, value=0.1}
              {arriveTangent=-0.7119, interpMode=RCIM_Cubic, leaveTangent=1.625, tangentMode=RCTM_Break, time=0.8, value=0.05}
              {arriveTangent=-0.25, interpMode=RCIM_Cubic, leaveTangent=-0.25, time=0.9, value=0.425}
              {arriveTangent=-0.0405, interpMode=RCIM_Cubic, leaveTangent=-0.0405, tangentMode=RCTM_User, time=1, value=0}
             ]}}})
         Passthrough Float=&FloatFromCurve(FloatCurve=DataInterface() {propertyValues {Curve {keys [
              {tangentMode=RCTM_User, time=0, value=1}
              {arriveTangent=-0.0405, interpMode=RCIM_Cubic, leaveTangent=-0.0405, tangentMode=RCTM_User, time=0.7009, value=1}
              {arriveTangent=-8.0089, interpMode=RCIM_Cubic, leaveTangent=-8.0089, tangentMode=RCTM_User, time=1, value=0}
             ]}}})
         Spawn Group Mask=4
        )
        Spawn Group Mask=3
       )
       Spawn Group Mask=2
      )
      Spawn Group Mask=1
     )
    )
    Scale RGB=&MaskVectorBySpawnGroup(
     Masked Vector=&Lerp_Vector(
      A=&VectorFromFloat(Value=&FloatFromCurve(
        FloatCurve=DataInterface() {propertyValues {Curve {keys [<at 0 1>,<at 0.1 1>,<at 0.2 2>,<at 0.3 1>,<at 0.6 1>,<at 0.7 2>,<at 0.8 1>,<at 1 1>]}}}
        Scale Curve=10
       ))
      B=&VectorFromFloat(Value=&RandomRangeFloat(Minimum=0.3, Maximum=5, Evaluation Type=::NewEnumerator1))
      Alpha=&FloatFromCurve(FloatCurve=DataInterface() {propertyValues {Curve {keys [
           {interpMode=RCIM_Cubic, leaveTangent=2.5294, tangentMode=RCTM_Break, time=0, value=0}
           {arriveTangent=-0.2166, interpMode=RCIM_Cubic, tangentMode=RCTM_Break, time=0.4, value=1}
          ]}}})
     )
     Passthrough Vector=&MaskVectorBySpawnGroup(
      Masked Vector=&Lerp_Vector(
       A=&VectorFromFloat(Value=&FloatFromCurve(
         FloatCurve=DataInterface() {propertyValues {Curve {keys [{tangentMode=RCTM_User, time=0, value=1}, {tangentMode=RCTM_User, time=1, value=1}]}}}
         Scale Curve=10
        ))
       B=&VectorFromFloat(Value=&RandomRangeFloat(Minimum=4, Maximum=5, Evaluation Type=::NewEnumerator1))
       Alpha=&FloatFromCurve(FloatCurve=DataInterface() {propertyValues {Curve {keys [
            {interpMode=RCIM_Cubic, leaveTangent=2.5294, tangentMode=RCTM_Break, time=0, value=0}
            {arriveTangent=-0.2166, interpMode=RCIM_Cubic, tangentMode=RCTM_Break, time=0.4, value=1}
           ]}}})
      )
      Passthrough Vector=&MaskVectorBySpawnGroup(
       Masked Vector=&Lerp_Vector(
        A=&VectorFromFloat(Value=&FloatFromCurve(FloatCurve=DataInterface() {propertyValues {Curve {keys [<at 0 1>,<at 0.3 1>,<at 0.7 1>,<at 0.8 4>,<at 0.9 1>,<at 1 1>]}}}, Scale Curve=10))
        B=&VectorFromFloat(Value=&RandomRangeFloat(Minimum=8.5, Maximum=10, Evaluation Type=::NewEnumerator1))
        Alpha=&FloatFromCurve(FloatCurve=DataInterface() {propertyValues {Curve {keys [
             {interpMode=RCIM_Cubic, leaveTangent=0.06, tangentMode=RCTM_Break, time=0.1, value=0}
             {arriveTangent=2.6781, interpMode=RCIM_Cubic, leaveTangent=2.6781, tangentMode=RCTM_User, time=0.8, value=0.8}
             {arriveTangent=2.5, interpMode=RCIM_Cubic, leaveTangent=-22.9319, tangentMode=RCTM_Break, time=0.9, value=1.5}
             <at 1 0 RCIM_Cubic>
            ]}}})
       )
       Passthrough Vector=&MaskVectorBySpawnGroup(
        Masked Vector=&Lerp_Vector(
         A=&VectorFromFloat(Value=&FloatFromCurve(FloatCurve=DataInterface() {propertyValues {Curve {keys [<at 0 1>,<at 0.1 1>,<at 0.2 4>,<at 0.3 1>,<at 0.6 1>,<at 0.7 2>,<at 0.8 1>,<at 1 1>]}}}))
         B=&VectorFromFloat(Value=&RandomRangeFloat(Minimum=0.5, Maximum=1, Evaluation Type=::NewEnumerator1))
         Alpha=&FloatFromCurve(FloatCurve=DataInterface() {propertyValues {Curve {keys [
              {interpMode=RCIM_Cubic, leaveTangent=2.5294, tangentMode=RCTM_Break, time=0, value=0}
              {arriveTangent=-0.2166, interpMode=RCIM_Cubic, tangentMode=RCTM_Break, time=0.4, value=1}
             ]}}})
        )
        Passthrough Vector=&MaskVectorBySpawnGroup(
         Masked Vector=&VectorFromFloat(Value=&FloatFromCurve(FloatCurve=DataInterface() {propertyValues {Curve {keys [<at 0 1>,<at 0.2 1>,<at 0.3 5>,<at 0.4 1>,<at 0.7 1>,<at 0.8 5>,<at 0.9 1>,<at 1 1>]}}}))
         Passthrough Vector=&Lerp_Vector(
          A=&VectorFromFloat(Value=&FloatFromCurve(
            FloatCurve=DataInterface() {propertyValues {Curve {keys [<at 0 1>,<at 0.1 1>,<at 0.2 2>,<at 0.3 1>,<at 0.6 1>,<at 0.7 2>,<at 0.8 1>,<at 1 1>]}}}
            Scale Curve=10
           ))
          B=&VectorFromFloat(Value=&RandomRangeFloat(Minimum=0.3, Maximum=5, Evaluation Type=::NewEnumerator1))
          Alpha=&FloatFromCurve(FloatCurve=DataInterface() {propertyValues {Curve {keys [
               {interpMode=RCIM_Cubic, leaveTangent=2.5294, tangentMode=RCTM_Break, time=0, value=0}
               {arriveTangent=-0.2166, interpMode=RCIM_Cubic, tangentMode=RCTM_Break, time=0.4, value=1}
              ]}}})
         )
         Spawn Group Mask=4
        )
        Spawn Group Mask=3
       )
       Spawn Group Mask=2
      )
      Spawn Group Mask=1
     )
    )
    ScaleA=true
    ScaleRGB=true
   )
   &Drag(
    Drag=&MaskFloatBySpawnGroup(
     Masked Float=0
     Passthrough Float=&MaskFloatBySpawnGroup(
      Masked Float=0
      Passthrough Float=&MaskFloatBySpawnGroup(
       Masked Float=0
       Passthrough Float=&MaskFloatBySpawnGroup(
        Masked Float=&FloatFromCurve(
         FloatCurve=DataInterface() {propertyValues {Curve {keys [
             {interpMode=RCIM_Cubic, tangentMode=RCTM_User, time=0, value=0}
             {arriveTangent=1.6092, interpMode=RCIM_Cubic, leaveTangent=1.6092, tangentMode=RCTM_User, time=1, value=1}
            ]}}}
         Scale Curve=7
        )
        Passthrough Float=&RandomRangeFloat(Minimum=0.8, Maximum=1.2)
        Spawn Group Mask=3
       )
       Spawn Group Mask=2
      )
      Spawn Group Mask=1
     )
    )
    Use Linear Drag=true
    Use Rotational Drag=false
   )
   &GravityForce(Gravity=&MaskVectorBySpawnGroup(
     Masked Vector=<0 0 0>
     Passthrough Vector=&MaskVectorBySpawnGroup(
      Masked Vector=<0 0 0>
      Passthrough Vector=&MaskVectorBySpawnGroup(
       Masked Vector=<0 0 0>
       Passthrough Vector=&MaskVectorBySpawnGroup(
        Masked Vector=<0 0 0>
        Passthrough Vector=&MaskVectorBySpawnGroup(Masked Vector=<0 0 -700>, Passthrough Vector=<0 0 0>, Spawn Group Mask=4)
        Spawn Group Mask=3
       )
       Spawn Group Mask=2
      )
      Spawn Group Mask=1
     )
    ))
   &SolveForcesAndVelocity(Clamp Velocity=false, Limit Acceleration=false)
   &ScaleSpriteSizeBySpeed(
    Max Scale Factor=&MaskVector2dBySpawnGroup(
     Masked Vector2D=<1 1>
     Passthrough Vector2D=&MaskVector2dBySpawnGroup(
      Masked Vector2D=<1 1>
      Passthrough Vector2D=&MaskVector2dBySpawnGroup(
       Masked Vector2D=<1 1>
       Passthrough Vector2D=&MaskVector2dBySpawnGroup(Masked Vector2D=<1 1>, Passthrough Vector2D=<0.5 7>, Spawn Group Mask=3)
       Spawn Group Mask=2
      )
      Spawn Group Mask=1
     )
     Spawn Group Mask=0
    )
    Sample Scale Factor By Curve=true
    Velocity Threshold=500
   )
   /SetVariables(Particles.StaffDMP:Vector4f=&MakeVector4(
     X=&FloatFromCurve(FloatCurve=DataInterface() {propertyValues {Curve {keys [
          {interpMode=RCIM_Cubic, leaveTangent=-79.1318, tangentMode=RCTM_Break, time=0, value=1}
          {arriveTangent=-2.6873, tangentMode=RCTM_User, time=0.5, value=-10}
          {arriveTangent=14.6846, interpMode=RCIM_Cubic, leaveTangent=14.6846, tangentMode=RCTM_User, time=0.9, value=-10}
          <at 1 1>
         ]}}})
     Y=0
     Z=0
     W=0
    )) {isSetParameters=true}
  ]
  rendererClasses [&NiagaraMeshRendererProperties, &NiagaraSpriteRendererProperties]
  renderers [
   {rendererClass=&NiagaraMeshRendererProperties, rendererIndex=0}
   {rendererClass=&NiagaraMeshRendererProperties, rendererIndex=1}
   {rendererClass=&NiagaraSpriteRendererProperties, rendererIndex=2}
   {rendererClass=&NiagaraSpriteRendererProperties, rendererIndex=3}
   {rendererClass=&NiagaraSpriteRendererProperties, rendererIndex=4}
   {rendererClass=&NiagaraSpriteRendererProperties, rendererIndex=5}
   {rendererClass=&NiagaraMeshRendererProperties, rendererIndex=6}
   {rendererClass=&NiagaraMeshRendererProperties, rendererIndex=7}
  ]
 }
 StaffLightning {
  emitterUpdateScript [
   &EmitterState(
    Delay First Loop Only=true
    Inactive Response=::NewEnumerator1
    Life Cycle Mode=::NewEnumerator1
    Loop Behavior=::NewEnumerator2
    Loop Count=4
    Loop Delay=1.1769
    Loop Duration=1.35
    UseLoopDelay=true
   )
   /SetVariables(Emitter.FadeOut:NiagaraFloat=&FloatFromCurve(
     FloatCurve=DataInterface() {propertyValues {Curve {keys [
         {interpMode=RCIM_Cubic, tangentMode=RCTM_User, time=0, value=1}
         {arriveTangent=-1.4808, interpMode=RCIM_Cubic, leaveTangent=-1.4808, tangentMode=RCTM_User, time=0.5, value=0.8}
         {arriveTangent=-0.0857, interpMode=RCIM_Cubic, leaveTangent=-0.0857, tangentMode=RCTM_User, time=1, value=0}
        ]}}}
     CurveIndex=$Emitter.NormalizedLoopAge
    )) {isSetParameters=true}
   /Game/Effects/Niagara/Modules/BlendSpace/BlendSpaceHelper()
   &SpawnRate(SpawnRate=&Multiply_Float(A=8, B=$Emitter.FadeOut), Use Spawn Probability=false)
   &SpawnRate(SpawnGroup=1, SpawnRate=&Multiply_Float(A=8, B=$Emitter.FadeOut), Use Spawn Probability=false) {module=SpawnRate001}
  ]
  particleSpawnScript [
   /SetVariables(Particles.VisibilityTag=&SpawnGroup()) {isSetParameters=true, module=SetVariables_#7}
   &InitializeParticle(
    Color=&Multiply_LinearColorByFloat(Linear Color=$System.ElimColor, Float=0.4)
    Lifetime Max=0.35
    Lifetime Min=0.175
    Lifetime Mode=::NewEnumerator1
    Mesh Scale Max=<2.1 2.1 2.5>
    Mesh Scale Min=<1 1 2.5>
    Mesh Scale Mode=::NewEnumerator2
    UsePositionOffset=false
    Write Material Random=true
   )
   &ShapeLocation(
    Box Size=&MaskVectorBySpawnGroup(Masked Vector=<5 5 15>, Passthrough Vector=<2 2 10>)
    MaskSpawnGroup=false
    Offset=&SelectVectorFromArray(
     Direct Array Index=&MakeIntFromFloat(Float=&Multiply_Float(A=$Emitter.NormalizedLoopAge, B=4))
     Interpolation=&Multiply_Float(A=$Emitter.NormalizedLoopAge, B=4)
     Direct Array Mode=::NewEnumerator1
     Vector Selection Array=DataInterface() {propertyValues {FloatData [<0 0 0>, <0 0 45>, <0 0 95>, <0 0 130>]}}
    )
    Offset Mode=::NewEnumerator0
    Override Seed=false
    Shape Primitive=::NewEnumerator2
    Surface Only=false
    UseOwnerScale=true
   )
   &InitialMeshOrientation(
    Mesh Orientation Mode=::NewEnumerator2
    Rotation=&RandomRangeVector(Minimum=<0 0 -1>, Maximum=<0 0 1>)
    Use Rotation Vector=true
   )
   &AddVelocity(
    Velocity=<0 0 1>
    Velocity Speed Scale=&RandomRangeFloat(Minimum=10, Maximum=&MaskFloatBySpawnGroup(Masked Float=5, Passthrough Float=15))
   )
  ]
  particleUpdateScript [
   &SolveForcesAndVelocity(Clamp Velocity=false, Limit Acceleration=false)
   &ParticleState()
   &ScaleMeshSize(Scale Factor=&Multiply_VectorByFloat(Vector=<1 1 1>, Float=0.2))
   &DynamicMaterialParameters(
    Index 0 Param 1=&MakeFloatFromVector(Vector=$Particles.Scale, Channel=::NewEnumerator0)
    Index 0 Param 2=&RandomRangeFloat(Minimum=0, Maximum=10)
    Index 0 Param 3=&FloatFromCurve(
     FloatCurve=DataInterface() {propertyValues {Curve {keys [{interpMode=RCIM_Cubic, leaveTangent=8.1085, tangentMode=RCTM_Break, time=0.8568, value=1}, {tangentMode=RCTM_User, time=1, value=0}]}}}
     Scale Curve=0.07
    )
    Index 0 Param 4=0.8
    Index 1 Param 1=&FloatFromCurve(FloatCurve=DataInterface() {propertyValues {Curve {keys [
         {interpMode=RCIM_Cubic, leaveTangent=1.1034, tangentMode=RCTM_Break, time=0, value=0}
         {arriveTangent=2.2399, tangentMode=RCTM_User, time=0.5, value=1}
        ]}}})
    Index 1 Param 2=&FloatFromCurve(FloatCurve=DataInterface() {propertyValues {Curve {keys [<at 0.5 0>, {arriveTangent=3.2053, tangentMode=RCTM_User, time=0.9, value=1}]}}})
    Index 1 Param 3=&FloatFromCurve(FloatCurve=DataInterface() {propertyValues {Curve {keys [<at 0 0 RCIM_Cubic>, {arriveTangent=3.4223, interpMode=RCIM_Cubic, leaveTangent=3.4223, tangentMode=RCTM_User, time=1, value=1}]}}})
    Index 1 Param 4=&FloatFromCurve(
     FloatCurve=DataInterface() {propertyValues {Curve {keys [
         {interpMode=RCIM_Cubic, leaveTangent=0.1169, tangentMode=RCTM_Break, time=0, value=0}
         {arriveTangent=2.5708, interpMode=RCIM_Cubic, leaveTangent=2.5708, tangentMode=RCTM_User, time=1, value=1}
        ]}}}
     Scale Curve=&RandomRangeFloat(Minimum=1, Maximum=40)
    )
    Index 2 Param 1=&FloatFromCurve(FloatCurve=DataInterface() {propertyValues {Curve {keys [
         {interpMode=RCIM_Cubic, leaveTangent=-0.0025, tangentMode=RCTM_Break, time=0, value=0.5}
         {arriveTangent=-0.0316, interpMode=RCIM_Cubic, tangentMode=RCTM_Break, time=1, value=-0.2}
        ]}}})
    Index 2 Param 2=&FloatFromCurve(
     FloatCurve=DataInterface() {propertyValues {Curve {keys [
         {interpMode=RCIM_Cubic, tangentMode=RCTM_User, time=0, value=0}
         {arriveTangent=2.0522, interpMode=RCIM_Cubic, leaveTangent=2.0522, tangentMode=RCTM_User, time=1, value=1}
        ]}}}
     Scale Curve=2
    )
    Module.Index0Param0Write=true
    Module.Index0Param1Write=true
    Module.Index0Param2Write=true
    Module.Index0Param3Write=true
    Module.Index1Param0Write=true
    Module.Index1Param1Write=true
    Module.Index1Param2Write=true
    Module.Index1Param3Write=true
    Module.Index2Param0Write=true
    Module.Index2Param1Write=true
    Module.Index2Param2Write=false
    Module.Index2Param3Write=false
    Module.Index3Param0Write=false
    Module.Index3Param1Write=false
    Module.Index3Param2Write=false
    Module.Index3Param3Write=true
    Write Parameter Index 1=true
    Write Parameter Index 2=true
    Write Parameter Index 3=true
   )
   &UpdateMeshOrientation(Rotation Rate=&RandomRangeFloat(Minimum=-1, Maximum=1), Rotation Vector=<0 0 1>)
   &ScaleColor(
    Scale Alpha=&FloatFromCurve(FloatCurve=DataInterface() {propertyValues {Curve {keys [
         {tangentMode=RCTM_User, time=0, value=1}
         {arriveTangent=-0.0405, interpMode=RCIM_Cubic, leaveTangent=-0.0405, tangentMode=RCTM_User, time=0.7009, value=1}
         {arriveTangent=-8.0089, interpMode=RCIM_Cubic, leaveTangent=-8.0089, tangentMode=RCTM_User, time=1, value=0}
        ]}}})
    Scale RGB=&Lerp_Vector(
     A=&VectorFromFloat(Value=&FloatFromCurve(
       FloatCurve=DataInterface() {propertyValues {Curve {keys [<at 0 1>,<at 0.1 1>,<at 0.2 2>,<at 0.3 1>,<at 0.6 1>,<at 0.7 2>,<at 0.8 1>,<at 1 1>]}}}
       Scale Curve=10
      ))
     B=&VectorFromFloat(Value=&RandomRangeFloat(Minimum=0.3, Maximum=5, Evaluation Type=::NewEnumerator1))
     Alpha=&FloatFromCurve(FloatCurve=DataInterface() {propertyValues {Curve {keys [
          {interpMode=RCIM_Cubic, leaveTangent=2.5294, tangentMode=RCTM_Break, time=0, value=0}
          {arriveTangent=-0.2166, interpMode=RCIM_Cubic, tangentMode=RCTM_Break, time=0.4, value=1}
         ]}}})
    )
    ScaleA=true
    ScaleRGB=true
   )
   &BlendSpace()
  ]
  rendererClasses [&NiagaraMeshRendererProperties]
  renderers [
   {rendererClass=&NiagaraMeshRendererProperties, rendererIndex=0}
   {rendererClass=&NiagaraMeshRendererProperties, rendererIndex=1}
   {rendererClass=&NiagaraMeshRendererProperties, rendererIndex=2}
   {rendererClass=&NiagaraMeshRendererProperties, rendererIndex=3}
  ]
 }
 StaffLightning001 {
  emitterUpdateScript [
   &EmitterState(
    Inactive Response=::NewEnumerator1
    Life Cycle Mode=::NewEnumerator1
    Loop Behavior=::NewEnumerator1
    Loop Delay=2.2883
    Loop Duration=4
    UseLoopDelay=true
   )
   /SetVariables(Emitter.FadeOutZShape:NiagaraFloat=&FloatFromCurve(
     FloatCurve=DataInterface() {propertyValues {Curve {keys [
         {interpMode=RCIM_Cubic, tangentMode=RCTM_User, time=0, value=1}
         {arriveTangent=-1.0679, interpMode=RCIM_Cubic, leaveTangent=-1.0679, tangentMode=RCTM_User, time=0.75, value=0.8}
         {arriveTangent=-0.0857, interpMode=RCIM_Cubic, leaveTangent=-0.0857, tangentMode=RCTM_User, time=1, value=0}
        ]}}}
     CurveIndex=$System.NormalizedLoopAge
    )) {isSetParameters=true}
   &BlendSpaceHelper()
   &SpawnRate(SpawnGroup=2, SpawnRate=500, Use Spawn Probability=false) {enabled=false, module=SpawnRate003}
   &SpawnRate(SpawnRate=&Multiply_Float(A=12, B=$Emitter.FadeOutZShape), Use Spawn Probability=false)
   &SpawnRate(SpawnGroup=1, SpawnRate=&Multiply_Float(A=12, B=$Emitter.FadeOutZShape), Use Spawn Probability=false) {module=SpawnRate001}
   &SpawnRate(SpawnGroup=2, SpawnRate=&Multiply_Float(A=12, B=$Emitter.FadeOutZShape), Use Spawn Probability=false) {module=SpawnRate002}
  ]
  particleSpawnScript [
   /SetVariables(Particles.VisibilityTag=&MakeCustomIntFromBool(
     Bool=&SetBoolByIntComparison(A=&SpawnGroup(), B=2, Comparison Type=::NewEnumerator7)
     True INT=&RandomRangeInt(Minimum=0, Maximum=5)
     False INT=6
    )) {isSetParameters=true, module=SetVariables_#7}
   /SetVariables(Particles.RandomSocket:NiagaraInt32=&SpawnGroup()) {isSetParameters=true}
   &InitializeParticle(
    Color=&Multiply_LinearColorByFloat(Linear Color=$System.ElimColor, Float=0.4)
    Lifetime=&MaskFloatBySpawnGroup(Masked Float=3, Passthrough Float=&RandomRangeFloat(Minimum=0.175, Maximum=0.35), Spawn Group Mask=6)
    Mesh Scale Max=<2.1 2.1 3.2>
    Mesh Scale Min=<1 1 3>
    Mesh Scale Mode=::NewEnumerator2
    Sprite Rotation Mode=::NewEnumerator2
    Sprite Size Max=<30 30>
    Sprite Size Min=<20 20>
    Sprite Size Mode=::NewEnumerator2
    UsePositionOffset=false
    Write Material Random=true
   )
   &ShapeLocation(
    Cylinder Height=5
    Cylinder Height Midpoint=1
    Cylinder Radius=2.5
    MaskSpawnGroup=false
    Offset=&MergeHelper_VectorSwitch(0=<-20 -3 165>, 1=<-15 -3 158>, 2=<-18 -3 138>, 3=?, Index=$Particles.RandomSocket, Num=::NewEnumerator1)
    Offset Mode=::NewEnumerator0
    Override Seed=false
    Rotation Mode=::NewEnumerator2
    Shape Primitive=::NewEnumerator1
    Surface Only=false
    UseOwnerScale=true
    Yaw / Pitch / Roll=&MergeHelper_VectorSwitch(0=<0 0.25 0>, 1=<0 0.35 0>, 2=<0 0.25 0>, 3=?, Index=$Particles.RandomSocket, Num=::NewEnumerator1)
   )
   &InitialMeshOrientation(
    Mesh Axis To Orient=<0 0 1>
    Mesh Orientation Mode=::NewEnumerator2
    Rotation=&MergeHelper_VectorSwitch(
     0=&RandomRangeVector(Minimum=<0 0.03 0>, Maximum=<0 0.075 0>)
     1=<0 0.15 0>
     2=&RandomRangeVector(Minimum=<0 0.03 0>, Maximum=<0 0.075 0>)
     3=?
     Index=&SpawnGroup()
     Num=::NewEnumerator1
    )
    Use Rotation Vector=true
   )
   &AddVelocity(
    Velocity=&RandomRangeVector(Minimum=<-1 -1 -1>, Maximum=<1 1 1>)
    Velocity Speed Scale=&RandomRangeFloat(
     Minimum=&MaskFloatBySpawnGroup(Masked Float=0, Passthrough Float=15, Spawn Group Mask=6)
     Maximum=&MaskFloatBySpawnGroup(Masked Float=0, Passthrough Float=5, Spawn Group Mask=6)
    )
   )
   &CameraOffset(Camera Offset Amount=15)
  ]
  particleUpdateScript [
   &SolveForcesAndVelocity(Clamp Velocity=false, Limit Acceleration=false)
   &ParticleState()
   &ScaleMeshSize(Scale Factor=&Multiply_VectorByFloat(Vector=<1 1 0.5>, Float=0.2))
   &DynamicMaterialParameters(
    Index 0 Param 1=&MakeFloatFromVector(Vector=$Particles.Scale, Channel=::NewEnumerator0)
    Index 0 Param 2=&RandomRangeFloat(Minimum=0, Maximum=10)
    Index 0 Param 3=&FloatFromCurve(
     FloatCurve=DataInterface() {propertyValues {Curve {keys [{interpMode=RCIM_Cubic, leaveTangent=8.1085, tangentMode=RCTM_Break, time=0.8568, value=1}, {tangentMode=RCTM_User, time=1, value=0}]}}}
     Scale Curve=0.07
    )
    Index 0 Param 4=0.8
    Index 1 Param 1=&FloatFromCurve(FloatCurve=DataInterface() {propertyValues {Curve {keys [
         {interpMode=RCIM_Cubic, leaveTangent=1.1034, tangentMode=RCTM_Break, time=0, value=0}
         {arriveTangent=2.2399, tangentMode=RCTM_User, time=0.5, value=1}
        ]}}})
    Index 1 Param 2=&FloatFromCurve(FloatCurve=DataInterface() {propertyValues {Curve {keys [<at 0.5 0>, {arriveTangent=3.2053, tangentMode=RCTM_User, time=0.9, value=1}]}}})
    Index 1 Param 3=&FloatFromCurve(FloatCurve=DataInterface() {propertyValues {Curve {keys [<at 0 0 RCIM_Cubic>, {arriveTangent=3.4223, interpMode=RCIM_Cubic, leaveTangent=3.4223, tangentMode=RCTM_User, time=1, value=1}]}}})
    Index 1 Param 4=&FloatFromCurve(
     FloatCurve=DataInterface() {propertyValues {Curve {keys [
         {interpMode=RCIM_Cubic, leaveTangent=0.1169, tangentMode=RCTM_Break, time=0, value=0}
         {arriveTangent=2.5708, interpMode=RCIM_Cubic, leaveTangent=2.5708, tangentMode=RCTM_User, time=1, value=1}
        ]}}}
     Scale Curve=&RandomRangeFloat(Minimum=1, Maximum=40)
    )
    Index 2 Param 1=&FloatFromCurve(
     FloatCurve=DataInterface() {propertyValues {Curve {keys [
         {interpMode=RCIM_Cubic, leaveTangent=-0.0025, tangentMode=RCTM_Break, time=0, value=0.5}
         {arriveTangent=-0.0316, interpMode=RCIM_Cubic, tangentMode=RCTM_Break, time=1, value=-0.2}
        ]}}}
     Scale Curve=1.8187
    )
    Index 2 Param 2=&FloatFromCurve(
     FloatCurve=DataInterface() {propertyValues {Curve {keys [
         {interpMode=RCIM_Cubic, tangentMode=RCTM_User, time=0, value=0}
         {arriveTangent=2.0522, interpMode=RCIM_Cubic, leaveTangent=2.0522, tangentMode=RCTM_User, time=1, value=1}
        ]}}}
     Scale Curve=2
    )
    Module.Index0Param0Write=true
    Module.Index0Param1Write=true
    Module.Index0Param2Write=true
    Module.Index0Param3Write=true
    Module.Index1Param0Write=true
    Module.Index1Param1Write=true
    Module.Index1Param2Write=true
    Module.Index1Param3Write=true
    Module.Index2Param0Write=true
    Module.Index2Param1Write=true
    Module.Index2Param2Write=false
    Module.Index2Param3Write=false
    Module.Index3Param0Write=false
    Module.Index3Param1Write=false
    Module.Index3Param2Write=false
    Module.Index3Param3Write=true
    Write Parameter Index 1=true
    Write Parameter Index 2=true
    Write Parameter Index 3=true
   )
   &UpdateMeshOrientation(Rotation Rate=&RandomRangeFloat(Minimum=-1, Maximum=1), Rotation Vector=<0 0 1>)
   &ScaleColor(
    Scale Alpha=&FloatFromCurve(FloatCurve=DataInterface() {propertyValues {Curve {keys [
         {tangentMode=RCTM_User, time=0, value=1}
         {arriveTangent=-0.0405, interpMode=RCIM_Cubic, leaveTangent=-0.0405, tangentMode=RCTM_User, time=0.7009, value=1}
         {arriveTangent=-8.0089, interpMode=RCIM_Cubic, leaveTangent=-8.0089, tangentMode=RCTM_User, time=1, value=0}
        ]}}})
    Scale RGB=&Lerp_Vector(
     A=&VectorFromFloat(Value=&FloatFromCurve(
       FloatCurve=DataInterface() {propertyValues {Curve {keys [<at 0 1>,<at 0.1 1>,<at 0.2 2>,<at 0.3 1>,<at 0.6 1>,<at 0.7 2>,<at 0.8 1>,<at 1 1>]}}}
       Scale Curve=10
      ))
     B=&VectorFromFloat(Value=&RandomRangeFloat(Minimum=0.3, Maximum=5, Evaluation Type=::NewEnumerator1))
     Alpha=&FloatFromCurve(FloatCurve=DataInterface() {propertyValues {Curve {keys [
          {interpMode=RCIM_Cubic, leaveTangent=2.5294, tangentMode=RCTM_Break, time=0, value=0}
          {arriveTangent=-0.2166, interpMode=RCIM_Cubic, tangentMode=RCTM_Break, time=0.4, value=1}
         ]}}})
    )
    ScaleA=true
    ScaleRGB=true
   )
   &FN_Light_Attributes(
    Color=<rgba 1 0.0087 0 0.7>
    Create Color Attribute=true
    Create Exponent Attribute=true
    Create Radius Attribute=true
    Exponent=1
    Radius=80
    Scale Color=&FloatFromCurve(
     FloatCurve=DataInterface() {propertyValues {Curve {keys [<at 0 1>,<at 1 0>]}}}
     CurveIndex=&Frac_Float(Float=&Multiply_Float(A=$Particles.NormalizedAge, B=50))
    )
   )
   &BlendSpace(Position Coordinate Space=2)
  ]
  rendererClasses [&NiagaraMeshRendererProperties, &NiagaraSpriteRendererProperties]
  renderers [
   {rendererClass=&NiagaraMeshRendererProperties, rendererIndex=0}
   {rendererClass=&NiagaraMeshRendererProperties, rendererIndex=1}
   {rendererClass=&NiagaraMeshRendererProperties, rendererIndex=2}
   {rendererClass=&NiagaraMeshRendererProperties, rendererIndex=3}
   {rendererClass=&NiagaraSpriteRendererProperties, rendererIndex=4}
  ]
 }
}
summary {
 systemName=NS_Emote_GlowFang_Staff_Charged_Long
 emitters [
  {
   emitterName=StaffChargeupFirstPhase
   rendererClasses [&NiagaraMeshRendererProperties, &NiagaraSpriteRendererProperties]
  }
  {emitterName=StaffLightning, rendererClasses [&NiagaraMeshRendererProperties]}
  {
   emitterName=StaffLightning001
   rendererClasses [&NiagaraMeshRendererProperties, &NiagaraSpriteRendererProperties]
  }
 ]
 userVariables [
  User.bIsFrontEndPreview {type=NiagaraBool}
  User.bKillStage {type=NiagaraInt32}
  User.LoopsDelay {type=NiagaraFloat, defaultValue=0, description=""}
 ]
}
